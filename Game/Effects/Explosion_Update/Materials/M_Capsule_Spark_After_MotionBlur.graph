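# Additive unlit Niagara sprite material that draws glowing capsule-shaped sparks for explosion effects, with responsive AA enabled.
# Sprite_Capsule supplies world position offset and UVs, sized from particle size; those UVs feed a powered sphere mask.
# The mask blends glow values tinted by particle color, with HDR boost off mobile and depth fade on the default shading path.
# A PreviousFrameSwitch subtracts particle velocity times delta time from the previous-frame offset, supplying per-particle motion vectors.
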Material /Game/Effects/Explosion_Update/Materials/M_Capsule_Spark_After_MotionBlur
graph {
 properties {
  bEnableResponsiveAA=true
  bUsedWithNiagaraSprites=true
  blendMode=BLEND_Additive
  pixelDepthOffsetMode=PDOM_Legacy
  shadingModel=MSM_Unlit
  usedShadingModels=Unlit
 }
 nodes [
  n1 = ParticleColor()
  n2 = ParticleSize()
  n3 = StaticBool()
  n4 = MaterialFunctionCall!(
   Particle Position
   Normalized Particle Direction
   Capsule Diameter=Multiply(A=ScalarParameter() {defaultValue=1, parameterName=CapsuleDiameter}, B=ComponentMask(None=@n2) {r=true})
   Capsule Pivot Alignment
   Scale Along Velocity Vector
   ---------Scale Over Velocity Parameters---------
   Particle Speed
   Minimum Speed
   Max Speed
   Minimum Scale Factor=MaterialFunctionCall(Float2=@n2) {materialFunction=&BreakOutFloat2Components}.G
   Maximum Scale Factor
   ---------- Advanced -----------
   Cap Size Multiplier=ScalarParameter() {defaultValue=0.75, parameterName=Cap Size Multiplier, sliderMax=1}
   Use Camera Offset=@n3
   Use Advanced Features=@n3
   World Position to Negate
  ) {materialFunction=/Engine/Functions/Engine_MaterialFunctions02/WorldPositionOffset/Sprite_Capsule}
  n5 = ScalarParameter() {defaultValue=0.5, parameterName=SphereMaskRadus}
  n6 = DynamicParameter() {defaultValue=<rgba 2 20 1 1>, paramNames [Power,DepthFade,HDRMult,Param4]}
  n7 = Power(Base=SphereMask(A=@n4.Texture Coordinates, B=@n5, Radius=@n5) {hardnessPercent=0}, Exp=@n6.Power)
  n8 = Multiply(
   A=@n1.RGB
   B=LinearInterpolate(A=ScalarParameter() {parameterName=Glow Dark Value}, B=ScalarParameter() {defaultValue=1, parameterName=Glow Bright Value}, Alpha=@n7)
  )
  n9 = Clamp(None=@n8)
  MaterialFunctionCall(Emissive=ShadingPathSwitch(
    Default=MaterialFunctionCall(LDR Input=@n9, HDR Tint=@n8, HDR Intensity Multiplier=@n6.HDRMult) {materialFunction=&DeriveHDRfromLDR}.RemappedRange
    Mobile=@n8
   )) {materialFunction=&MF_FX_EmissiveAdjustments}
  Multiply(A=@n1.A, B=ShadingPathSwitch(Default=DepthFade(Opacity=@n7, FadeDistance=@n6.DepthFade), Forward=@n7, Mobile=@n7))
  ParticleSize()
  n10 = Multiply(A=Multiply(A=ParticleDirection(), B=ParticleSpeed()), B=DeltaTime())
  PreviousFrameSwitch(Current Frame=@n4.World Position Offset, Previous Frame=Subtract(A=@n4.World Position Offset, B=Multiply(A=@n10)))
 ]
}
