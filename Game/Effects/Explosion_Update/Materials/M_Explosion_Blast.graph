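# Additive unlit material for an explosion blast effect on Niagara sprites and mesh particles, using a ring texture by default.
# The texture's red channel blends a dark tint toward white, multiplied by particle or parameter color, then passed through an emissive adjustment function.
# That channel also drives opacity, eroded by a Dissolve dynamic parameter and scaled by particle alpha.
# Near-camera fade and depth fade soften it, with depth fade skipped on mobile and low quality.
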Material /Game/Effects/Explosion_Update/Materials/M_Explosion_Blast
graph {
 properties {
  bUsedWithNiagaraMeshParticles=true
  bUsedWithNiagaraSprites=true
  blendMode=BLEND_Additive
  pixelDepthOffsetMode=PDOM_Legacy
  shadingModel=MSM_Unlit
  usedShadingModels=Unlit
 }
 nodes [
  n1 = ParticleColor()
  n2 = VectorParameter() {parameterName=Base Color, defaultValue=<rgba 0.6854 1.6198 2.8 1>}
  n3 = VectorParameter() {parameterName=DarkColor, defaultValue=<rgba 0.1133 0.088 0.2448 1>}
  n4 = TextureSampleParameter2D() {parameterName=Texture, samplerType=SAMPLERTYPE_Normal, texture=/Game/Effects/Explosion_Update/Textures/T_Explosion_Ring}
  MaterialFunctionCall(Emissive=Multiply(A=StaticSwitchParameter(True=@n1.RGB, False=@n2.RGB) {parameterName=Base Color}, B=LinearInterpolate(A=@n3.RGB, Alpha=@n4.R))) {materialFunction=&MF_FX_EmissiveAdjustments}
  n5 = Multiply(
   A=OneMinus(None=@n4.R)
   B=DynamicParameter() {parameterIndex=2, defaultValue=<rgba 0 1 1 1>, paramNames [Dissolve,Param2,Param3,Param4]}.Dissolve
  )
  n6 = Saturate(None=Multiply(A=Subtract(A=@n4.R, B=@n5), B=@n1.A))
  n7 = DepthFade(Opacity=@n6, FadeDistance=ScalarParameter() {defaultValue=35, parameterName=DepthFadeDistance})
  n8 = Constant() {r=1}
  n9 = QualitySwitch(Default=@n7, Low=@n8, High=@n7, Medium=@n7, Epic=@n7)
  Multiply(
   A=MaterialFunctionCall(Distance=ScalarParameter() {defaultValue=150, parameterName=NearFadeDistance}) {materialFunction=&Near_Camera_Fade}.Result
   B=Multiply(A=@n6, B=ShadingPathSwitch(Default=@n9, Mobile=@n8) {desc=Migrated from FeatureLevelSwitch})
  )
 ]
}
