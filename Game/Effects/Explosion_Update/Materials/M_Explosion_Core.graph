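# Unlit translucent Niagara mesh-particle material for the glowing core of an explosion.
# A panning, particle-randomized noise texture blends two colors into HDR emissive, tinted by particle color and optionally by time of day.
# Squared inverse Fresnel softens silhouette edges, and a steep power of it adds an intense center hotspot.
# Opacity combines that mask with particle alpha, depth fade and optional near-camera fade.
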Material /Game/Effects/Explosion_Update/Materials/M_Explosion_Core
graph {
 properties {
  bUsedWithNiagaraMeshParticles=true
  blendMode=BLEND_Translucent
  pixelDepthOffsetMode=PDOM_Legacy
  shadingModel=MSM_Unlit
  usedShadingModels=Unlit
 }
 nodes [
  n1 = MaterialFunctionCall(
   Color_Morning=VectorParameter() {parameterName=Color Morning, defaultValue=<rgba 0.2795 0.2795 0.2795 1>}.RGB
   Color_Day=VectorParameter() {parameterName=Color Day, defaultValue=<rgba 1 1 1 1>}.RGB
   Color_Evening=VectorParameter() {parameterName=Color Evening, defaultValue=<rgba 0.2726 0.2726 0.2726 1>}.RGB
   Color_Night=VectorParameter() {parameterName=Color Night, defaultValue=<rgba 0.3 0.3438 0.3438 1>}.RGB
  ) {materialFunction=&Color_Time_of_Day}
  n2 = Add(
   A=Multiply(A=ScalarParameter() {defaultValue=1, parameterName=Noise Scale}, B=TextureCoordinate() {uTiling=0.9, vTiling=0.5})
   B=ParticleRandom()
  )
  n3 = ScalarParameter() {defaultValue=0.5, parameterName=Noise Pan Speed}
  n4 = LinearInterpolate(
   A=VectorParameter() {parameterName=ColorLow, defaultValue=<rgba 0.2969 0.0734 0 1>}.RGB
   B=VectorParameter() {parameterName=ColorHigh, defaultValue=<rgba 1 0.6527 0 1>}.RGB
   Alpha=TextureSampleParameter2D(UVs=Panner(Coordinate=@n2, Speed=@n3) {speedY=1}) {parameterName=Noise Texture, texture=/Game/Global/Textures/Default/Caustics/T_Fractal_01}.R
  )
  n5 = Multiply(A=@n4, B=ScalarParameter() {defaultValue=200, parameterName=EmissiveMultiplier}) {constB=200}
  n6 = ShadingPathSwitch(
   Default=MaterialFunctionCall(LDR Input=@n4, HDR Tint=@n5) {materialFunction=&DeriveHDRfromLDR}.RemappedRange
   Mobile=@n5
  ) {desc=Migrated from FeatureLevelSwitch}
  n7 = OneMinus(None=Fresnel(BaseReflectFractionIn=ScalarParameter() {defaultValue=0.4, parameterName=BaseReflectFraction}) {baseReflectFraction=0.4, exponent=1})
  n8 = Saturate(None=Multiply(A=Multiply(A=@n7, B=@n7)) {constB=1.4})
  n9 = Multiply(B=Power(Base=@n8) {constExponent=20}) {constA=999999}
  n10 = ParticleColor()
  n11 = QualitySwitch(Default=@n6, Low=@n5)
  n12 = ScalarParameter() {defaultValue=1, parameterName=EmissivePower}
  n13 = Add(
   A=Multiply(A=Multiply(A=@n6, B=@n9), B=ScalarParameter() {defaultValue=1, parameterName=Core Intensity})
   B=Multiply(A=@n10.RGB, B=Power(Base=@n11, Exp=@n12))
  )
  MaterialFunctionCall(Emissive=StaticSwitchParameter(True=Multiply(A=@n1.Result, B=@n13), False=@n13) {defaultValue=true, parameterName=UseTimeOfDayColor}) {materialFunction=&MF_FX_EmissiveAdjustments}
  n14 = Multiply(A=@n10.A, B=Saturate(None=Add(A=@n8, B=@n9)))
  n15 = MaterialFunctionCall(Distance=ScalarParameter() {defaultValue=250, parameterName=NearFadeDistance}) {materialFunction=&Near_Camera_Fade}
  n16 = Multiply(A=@n14, B=Multiply(A=@n15.Result, B=@n15.Result))
  Multiply(
   A=DepthFade(
    Opacity=StaticSwitchParameter(True=@n16, False=@n14) {parameterName=Use_CameraFade}
    FadeDistance=ScalarParameter() {defaultValue=100, parameterName=DEpthFade}
   )
   B=ScalarParameter() {defaultValue=1, parameterName=Alpha}
  )
  ScalarParameter() {defaultValue=1, parameterName=Fresnel Expo}
 ]
}
