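# Translucent unlit material for explosion smoke and fire on Niagara mesh particles, animated through four sets of dynamic parameters.
# Distorted volume-texture Voronoi noise, in world or local space, drives vertex-normal displacement and fire patterns, with a separate swirl UV offset.
# Emissive blends fake normal-based sky-light shading, optional time-of-day tint and two fire layers.
# Opacity combines panning noise dissolve, inverted Fresnel edge softening, near-camera fade and depth fade, simplified on mobile.
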Material /Game/Effects/Explosion_Update/Materials/M_Explosion_MeshSmoke
graph {
 properties {
  bCastRayTracedShadows=false
  bUsedWithNiagaraMeshParticles=true
  bUsedWithNiagaraSprites=true
  blendMode=BLEND_Translucent
  pixelDepthOffsetMode=PDOM_Legacy
  shadingModel=MSM_Unlit
  usedShadingModels=Unlit
 }
 nodes [
  ActorPositionWS()
  CameraVectorWS()
  n1 = Multiply(A=AtmosphericLightColor(), B=ScalarParameter() {defaultValue=0.02, parameterName=Atmosphere Color Blend Rate})
  ComponentMask(None=@n1) {r=true}
  n2 = Constant3Vector() {constant=<rgba 0 0 1 0>}
  n3 = DynamicParameter() {parameterIndex=3, defaultValue=<rgba 0 0 1 0>, paramNames [Fire Pan, Dissolve Pan, Temperature, FireNoise Pan]}
  n4 = DynamicParameter() {parameterIndex=1, defaultValue=<rgba 5 0.008 0.5 1>, paramNames [Temperature Pow, Noise Scale, Luminance Adjustment, Fresnel Mul]}
  n5 = Multiply(A=@n4.Noise Scale, B=ScalarParameter() {defaultValue=1, group=Fire, parameterName=Fire Noise Scale})
  n6 = Constant3Vector() {constant=<rgba 0 0 1 0>}
  n7 = Multiply(
   A=DynamicParameter() {defaultValue=<rgba -100 1 1 -1>, paramNames [Displacement, ShadowStrength, PlayAnimTime, Dissolve]}.PlayAnimTime
   B=@n6
  )
  n8 = Multiply(A=ParticleRandom()) {constB=100}
  n9 = WorldPosition()
  n10 = TransformPosition(None=@n9.XYZ) {desc=Local, transformSourceType=TRANSFORMPOSSOURCE_World}
  n11 = Add(
   A=Multiply(A=NamedRerouteUsage()) {constA=2, constB=4}
   B=Add(A=Add(A=@n7, B=@n8), B=StaticSwitchParameter(True=@n10, False=@n9.XYZ) {parameterName=IS_Local Space})
  )
  n12 = WorldPosition()
  n13 = ParticleRandom()
  n14 = ParticlePositionWS()
  n15 = StaticSwitchParameter(
   True=TransformPosition(None=Add(A=@n12.XYZ, B=@n13)) {desc=Local, transformSourceType=TRANSFORMPOSSOURCE_World}
   False=Add(A=Add(A=@n14, B=@n12.XYZ), B=@n13)
  ) {parameterName=IS_Local Space}
  n16 = DynamicParameter() {parameterIndex=3, defaultValue=<rgba 0 0 1 0>, paramNames [Fire Pan, Dissolve Pan, Temperature, FireNoise Pan]}
  n17 = Multiply(A=@n6, B=@n16.FireNoise Pan)
  n18 = ScalarParameter() {defaultValue=0.01, group=Fire, parameterName=Volume Noise Distortion UVW Scale}
  n19 = ScalarParameter() {defaultValue=1, parameterName=Input Global Scale}
  n20 = TextureSample(UVs=Multiply(A=Add(A=@n15, B=@n17), B=Divide(A=@n18, B=@n19) {constB=1}), Tex=NamedRerouteUsage()) {samplerType=SAMPLERTYPE_LinearColor}
  n21 = Add(A=@n11, B=Multiply(A=Subtract(A=@n20.R) {constB=0.5}, B=ScalarParameter() {defaultValue=50, parameterName=Volume Distortion Scale}))
  n22 = TextureSample(UVs=Add(A=Multiply(A=@n2, B=@n3.Fire Pan), B=Multiply(A=@n5, B=@n21)), Tex=NamedRerouteUsage()) {samplerType=SAMPLERTYPE_LinearColor, texture=/Game/Effects/Explosion_Update/Textures/VT_Voronoi_Seamless}
  n23 = Multiply(A=@n22.R, B=@n22.R)
  n24 = Multiply(A=@n23, B=@n23)
  NamedRerouteDeclaration(None=Add(A=Multiply(A=@n24, B=@n24), B=@n22.R)) {name=Voronoi Noise Result Fire, nodeColor=<rgba 1 0 0.0703 1>}
  n25 = NamedRerouteUsage()
  n26 = VectorParameter() {parameterName=Volonoi Noise Scale, defaultValue=<rgba 1 1 1 0>}
  n27 = DynamicParameter() {parameterIndex=1, defaultValue=<rgba 5 0.008 0.5 1>, paramNames [Temperature Pow, Noise Scale, Luminance Adjustment, Fresnel Mul]}
  n28 = ParticleRandom()
  n29 = StaticSwitchParameter(True=@n28, False=Add(A=ParticlePositionWS(), B=@n28)) {parameterName=IS_Local Space}
  n30 = Multiply(A=Multiply(A=Multiply(A=@n26.RGB, B=@n27.Noise Scale), B=Subtract(A=@n29, B=@n11) {desc=L})) {constB=0.5}
  n31 = TextureSample(UVs=Add(A=ParticleRandom(), B=Add(A=@n25, B=@n30)), Tex=NamedRerouteUsage()) {samplerType=SAMPLERTYPE_LinearColor, texture=&VT_Voronoi_Seamless}
  n32 = Multiply(A=@n31.R, B=@n31.R)
  NamedRerouteDeclaration(None=Multiply(A=@n32, B=@n32)) {name=Voronoi Noise Result 2, nodeColor=<rgba 1 0 0.0703 1>}
  n33 = Multiply(A=VertexNormalWS(), B=NamedRerouteUsage())
  n34 = DynamicParameter() {defaultValue=<rgba -100 1 1 -1>, paramNames [Displacement, ShadowStrength, PlayAnimTime, Dissolve]}
  NamedRerouteDeclaration(None=StaticSwitch(
    True=Multiply(A=@n33, B=@n34.Displacement)
    False=Constant()
    Value=StaticBoolParameter() {defaultValue=true, parameterName=Use Displacement}
   )) {name=Displacement Result, nodeColor=<rgba 0 0.6484 1 1>}
  n35 = MaterialFunctionCall(
   Color_Morning=VectorParameter() {group=Time of Day Color, parameterName=Color Morning, defaultValue=<rgba 0.4375 0.3157 0.2716 1>}.RGB
   Color_Day=VectorParameter() {group=Time of Day Color, parameterName=Color Day, defaultValue=<rgba 0.2708 0.2275 0.1823 1>}.RGB
   Color_Evening=VectorParameter() {group=Time of Day Color, parameterName=Color Evening, defaultValue=<rgba 0.5208 0.2454 0.1356 1>}.RGB
   Color_Night=VectorParameter() {group=Time of Day Color, parameterName=Color Night, defaultValue=<rgba 0.131 0.1262 0.151 1>}.RGB
  ) {materialFunction=&Color_Time_of_Day}
  n36 = ShadingPathSwitch(
   Default=@n35.Result
   Deferred=@n35.Result
   Mobile=MaterialFunctionCall(
    Color_Morning=VectorParameter() {group=Time of Day Color, parameterName=Color Morning, defaultValue=<rgba 0.4375 0.3157 0.2716 1>}.RGB
    Color_Day=VectorParameter() {group=Time of Day Color, parameterName=Color Day, defaultValue=<rgba 0.2708 0.2275 0.1823 1>}.RGB
    Color_Evening=VectorParameter() {group=Time of Day Color, parameterName=Color Evening, defaultValue=<rgba 0.5208 0.2454 0.1356 1>}.RGB
    Color_Night=VectorParameter() {group=Time of Day Color, parameterName=Color Night, defaultValue=<rgba 0.131 0.1262 0.151 1>}.RGB
   ) {materialFunction=&Color_Time_of_Day}.Result
  ) {desc=Migrated from FeatureLevelSwitch}
  n37 = ScalarParameter() {defaultValue=0.01, desc=How much consider the sky atmosphere lighting, group=Shading, parameterName=SkyAtmosphere LightScale}
  n38 = WorldPosition()
  n39 = Abs(None=@n38.Z)
  n40 = Multiply(A=@n37, B=SkyAtmosphereLightIlluminance!(World Position=AppendVector(A=@n38.XY, B=@n39)))
  n41 = Saturate(None=Add(A=@n40, B=@n37) {constB=0.1})
  n42 = Constant()
  n43 = Constant()
  n44 = ParticleColor()
  n45 = DynamicParameter() {parameterIndex=1, defaultValue=<rgba 5 0.008 0.5 1>, paramNames [Temperature Pow, Noise Scale, Luminance Adjustment, Fresnel Mul]}
  n46 = Multiply(A=Add(A=PixelNormalWS()), B=@n45.Luminance Adjustment) {constB=0.5}
  n47 = NamedRerouteUsage()
  n48 = Add(A=Multiply(A=@n46, B=OneMinus(None=@n47)), B=@n46)
  n49 = SkyAtmosphereLightDirection()
  n50 = ScalarParameter() {defaultValue=0.333, group=Shading, parameterName=SkyAtmosphereLightDir_BlendRate}
  n51 = AppendVector(A=ComponentMask(None=Multiply(A=@n49, B=@n50) {constB=0.333}) {g=true, r=true}, B=Constant() {r=1})
  n52 = Add(A=Multiply(A=DotProduct(A=@n48, B=@n51)) {constB=0.5}) {constB=0.5}
  n53 = DynamicParameter() {parameterIndex=1, defaultValue=<rgba 5 0.008 0.5 1>, paramNames [Temperature Pow, Noise Scale, Luminance Adjustment, Fresnel Mul]}
  n54 = Constant() {r=1.5}
  n55 = NamedRerouteUsage()
  n56 = OneMinus(None=@n55)
  n57 = DynamicParameter() {defaultValue=<rgba -100 1 1 -1>, paramNames [Displacement, ShadowStrength, PlayAnimTime, Dissolve]}
  n58 = Saturate(None=Multiply(A=@n55, B=@n57.ShadowStrength))
  n59 = Multiply(
   A=ShadingPathSwitch(Default=@n52, Deferred=@n52, Mobile=Multiply(A=@n53.Luminance Adjustment, B=@n54)) {desc=Migrated from FeatureLevelSwitch}
   B=StaticSwitchParameter(True=Multiply(A=@n56, B=@n57.ShadowStrength), False=OneMinus(None=@n58)) {parameterName=IS_Lit}
  ) {constA=1}
  n60 = Multiply(A=Add(A=StaticSwitchParameter(True=@n40, False=@n43) {group=Shading, parameterName=Use SkyAtmosphereLightColor}, B=@n44.RGB), B=@n59)
  n61 = Multiply(A=@n59) {constB=0.5}
  n62 = StaticSwitchParameter(
   True=ShadingPathSwitch(Default=@n41, Deferred=@n41, Mobile=@n42) {desc=Migrated from FeatureLevelSwitch}
   False=Add(A=Multiply(A=@n42, B=@n60) {constA=1}, B=Multiply(A=@n44.RGB, B=@n61))
  ) {group=Shading, parameterName=Use SkyAtmosphereLightColor}
  n63 = DynamicParameter() {parameterIndex=3, defaultValue=<rgba 0 0 1 0>, paramNames [Fire Pan, Dissolve Pan, Temperature, FireNoise Pan]}
  n64 = Power(Base=Saturate(None=NamedRerouteUsage()), Exp=ScalarParameter() {defaultValue=1, group=Fire, parameterName=Fire Posterize Adjustment})
  n65 = DynamicParameter() {parameterIndex=1, defaultValue=<rgba 5 0.008 0.5 1>, paramNames [Temperature Pow, Noise Scale, Luminance Adjustment, Fresnel Mul]}
  n66 = Saturate(None=Subtract(A=Power(Base=@n64, Exp=@n65.Temperature Pow), B=Multiply(B=@n65.Temperature Pow) {constA=0.02}))
  n67 = ScalarParameter() {defaultValue=2, group=Fire, parameterName=Fire Contrast}
  n68 = Multiply(
   A=@n63.Temperature
   B=Multiply(A=Power(Base=@n66, Exp=@n67), B=VectorParameter() {group=Fire, parameterName=FireColor, defaultValue=<rgba 4 0.3489 0 0>}.RGB)
  )
  n69 = Constant()
  n70 = StaticBoolParameter() {defaultValue=true, parameterName=Use High Temperature}
  n71 = Power(Base=NamedRerouteUsage()) {constExponent=0.05}
  n72 = OneMinus(None=@n71)
  n73 = StaticSwitch(
   True=Power(Base=Saturate(None=@n71)) {constExponent=50}
   False=Multiply(A=@n72, B=@n72)
   Value=StaticBoolParameter() {group=Fire, parameterName=Use Inverse Tempereture}
  )
  n74 = DynamicParameter() {parameterIndex=2, defaultValue=<rgba 1 0 1 1>, paramNames [Swirl Scale, Swirl UV Pan, Fire2, Swirl UV Scale]}
  n75 = Multiply(
   A=Multiply(A=Multiply(A=@n73, B=@n74.Fire2), B=VectorParameter() {group=Fire, parameterName=FireColor2, defaultValue=<rgba 4 0.3489 0 0>}.RGB)
   B=Power(Base=NamedRerouteUsage(), Exp=ScalarParameter() {defaultValue=3, group=Fire, parameterName=Fire2NoisePow})
  )
  n76 = Add(A=Add(A=@n62, B=StaticSwitch(True=@n68, False=@n69, Value=@n70)), B=@n75) {constB=0}
  n77 = Add(A=@n62, B=@n75)
  n78 = Multiply(
   A=StaticSwitchParameter(True=@n62, False=ShadingPathSwitch(Default=@n76, Deferred=@n76, Mobile=@n77) {desc=Migrated from FeatureLevelSwitch}) {parameterName=IS_Simple Smoke}
   B=Power(Base=NamedRerouteUsage()) {constExponent=0.5}
  )
  NamedRerouteDeclaration(None=StaticSwitchParameter(True=Multiply(A=@n36, B=@n78), False=@n78) {defaultValue=true, group=Shading, parameterName=UseTimeOfDayColor}) {name=Emissive Result, nodeColor=<rgba 0 0.5781 1 1>}
  NamedRerouteDeclaration(None=Multiply(A=@n44.RGB, B=@n59)) {name=Base Color Result, nodeColor=<rgba 0.0312 1 0 1>}
  n79 = DynamicParameter() {parameterIndex=2, defaultValue=<rgba 1 0 1 1>, paramNames [Swirl Scale, Swirl UV Pan, Fire2, Swirl UV Scale]}
  n80 = WorldPosition()
  n81 = TransformPosition(None=@n80.XYZ) {desc=Local, transformSourceType=TRANSFORMPOSSOURCE_World}
  n82 = ScalarParameter() {defaultValue=0.0007, parameterName=Swirl UV Scale}
  n83 = ScalarParameter() {defaultValue=1, parameterName=Input Global Scale}
  n84 = Multiply(
   A=@n79.Swirl UV Scale
   B=Multiply(A=StaticSwitchParameter(True=@n81, False=@n80.XYZ) {parameterName=IS_Local Space}, B=Divide(A=@n82, B=@n83) {constB=1})
  )
  n85 = DynamicParameter() {parameterIndex=2, defaultValue=<rgba 1 0 1 1>, paramNames [Swirl Scale, Swirl UV Pan, Fire2, Swirl UV Scale]}
  n86 = Multiply(A=@n85.Swirl UV Pan, B=Constant3Vector() {constant=<rgba 0.1 0.1 1 0>})
  n87 = TextureSample(UVs=Add(A=ParticleRandom(), B=Add(A=@n84, B=@n86)), Tex=NamedRerouteUsage()) {samplerType=SAMPLERTYPE_LinearColor}
  NamedRerouteDeclaration(None=Multiply(A=Subtract(A=@n87.RGB) {constB=0.5}, B=@n85.Swirl Scale)) {name=Swirl UV, nodeColor=<rgba 1 0 0.9609 1>}
  NamedRerouteDeclaration(None=Transform(None=@n31.R)) {name=Voronoi Noise Result 1, nodeColor=<rgba 0.8281 1 0 1>}
  NamedRerouteDeclaration(None=@n20.R) {name=Noise_for_Fire2, nodeColor=<rgba 0 0.9297 1 1>}
  n88 = ScalarParameter() {defaultValue=8, parameterName=Fresnel Alpha Expo}
  n89 = ScalarParameter() {parameterName=BaseReflectFraction}
  n90 = DynamicParameter() {parameterIndex=1, defaultValue=<rgba 5 0.008 0.5 1>, paramNames [Temperature Pow, Noise Scale, Luminance Adjustment, Fresnel Mul]}
  n91 = Power(
   Base=OneMinus(None=Fresnel(ExponentIn=@n88, BaseReflectFractionIn=@n89) {baseReflectFraction=0.01})
   Exp=Multiply(A=ScalarParameter() {defaultValue=20, parameterName=FresnelPower}, B=@n90.Fresnel Mul)
  )
  NamedRerouteDeclaration(None=@n91) {name=Fresnel, nodeColor=<rgba 0.1719 1 0 1>}
  n92 = TextureObjectParameter() {
   group=Volume Textures
   parameterName=Dissolve and Swirl Noise Volume Tex
   samplerType=SAMPLERTYPE_LinearColor
   texture=/Game/Global/MaterialFunctions/VT_ClothSimWPO
  }
  NamedRerouteDeclaration(None=@n92) {name=Noise Texture 1, nodeColor=<rgba 0 0.6719 1 1>}
  n93 = TextureObjectParameter() {
   group=Volume Textures
   parameterName=Displacement Volume Tex
   samplerType=SAMPLERTYPE_LinearColor
   texture=&VT_Voronoi_Seamless
  }
  NamedRerouteDeclaration(None=@n93) {name=Noise Texture 2, nodeColor=<rgba 0 0.7188 1 1>}
  NamedRerouteUsage()
  n94 = MaterialFunctionCall(Distance=ScalarParameter() {defaultValue=200, parameterName=Near Camera Fade Distance}) {materialFunction=&Near_Camera_Fade}
  n95 = Multiply(A=@n94.Result, B=@n94.Result)
  n96 = VertexInterpolator(VS=@n95)
  n97 = DynamicParameter() {parameterIndex=3, defaultValue=<rgba 0 0 1 0>, paramNames [Fire Pan, Dissolve Pan, Temperature, FireNoise Pan]}
  n98 = Multiply(A=Constant3Vector() {constant=<rgba 0 0 1 0>}, B=@n97.Dissolve Pan)
  n99 = WorldPosition()
  n100 = StaticSwitchParameter(True=TransformPosition(None=@n99.XYZ) {desc=Local, transformSourceType=TRANSFORMPOSSOURCE_World}, False=@n99.XYZ) {parameterName=IS_Local Space}
  n101 = Multiply(A=Time(), B=VectorParameter() {group=DIssolve, parameterName=Dissolve Noise Pan Speed XYZ, defaultValue=<rgba 1 1 -40 0>}.RGB)
  n102 = VectorParameter() {group=DIssolve, parameterName=DissolveNoiseScale, defaultValue=<rgba 0.002 0.002 0.005 1>}
  n103 = Divide(
   A=ScalarParameter() {defaultValue=2.56, parameterName=Dissolve Noise Scale}
   B=ScalarParameter() {defaultValue=1, parameterName=Input Global Scale 2}
  ) {constB=1}
  n104 = Add(A=Multiply(A=Add(A=@n100, B=@n101), B=Multiply(A=@n102.RGB, B=@n103)) {constB=0.01}, B=NamedRerouteUsage())
  n105 = OneMinus(None=TextureSample(UVs=Add(A=@n98, B=@n104), Tex=NamedRerouteUsage()) {samplerType=SAMPLERTYPE_LinearColor}.R)
  n106 = Saturate(None=Subtract(
    A=Multiply(A=@n105, B=@n105)
    B=DynamicParameter() {defaultValue=<rgba -100 1 1 -1>, paramNames [Displacement, ShadowStrength, PlayAnimTime, Dissolve]}.Dissolve
   ))
  n107 = ScalarParameter() {defaultValue=9, group=DIssolve, parameterName=Pow for Dissolve}
  n108 = Multiply(A=@n91) {constB=2}
  n109 = Multiply(A=Multiply(A=Power(Base=@n106, Exp=@n107) {constExponent=9}, B=ParticleColor().A), B=Saturate(None=Subtract(A=@n108) {constB=0.1}))
  n110 = StaticSwitchParameter(True=Multiply(A=StaticSwitchParameter(True=@n95, False=@n96.PS) {parameterName=PerPixel Fading}, B=@n109), False=@n109) {defaultValue=true, parameterName=Use Near Camera Fade}
  n111 = DepthFade(Opacity=@n110, FadeDistance=ScalarParameter() {defaultValue=60, parameterName=DepthFade})
  Reroute(None=ShadingPathSwitch(Default=@n111, Deferred=@n111, Mobile=@n110) {desc=Migrated from FeatureLevelSwitch})
  n112 = WorldPosition()
  SkyAtmosphereLightIlluminance(World Position=AppendVector(A=@n112.XY, B=Abs(None=@n112.Z)))
  StaticSwitchParameter(
   True=Constant()
   False=MaterialFunctionCall(Emissive=NamedRerouteUsage()) {materialFunction=&MF_FX_EmissiveAdjustments}.Result
  ) {parameterName=IS_Lit}
  n113 = Saturate(None=DotProduct(A=SkyAtmosphereLightDirection(), B=VertexNormalWS()))
  n114 = LinearInterpolate(B=@n113, Alpha=ScalarParameter() {defaultValue=0.65, parameterName=Dot Blend}) {constA=1}
  n115 = NamedRerouteUsage()
  n116 = Multiply(A=ParticleColor().RGB, B=LinearInterpolate(B=OneMinus(None=@n115)) {constA=1, constAlpha=0.75})
  StaticSwitchParameter(True=Add(A=Multiply(A=@n1, B=@n113), B=Multiply(A=@n114, B=@n116)), False=@n116) {parameterName=IS_Lit}
  Subtract(A=@n21, B=ParticlePositionWS()) {constB=0}
 ]
}
