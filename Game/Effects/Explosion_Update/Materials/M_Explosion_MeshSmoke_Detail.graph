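# Translucent unlit explosion smoke material for Niagara particles, with fire glow, normal-based displacement and detail noise.
# World-space volume textures build swirl, distorted Voronoi fire and panning dissolve noise, driven by per-particle dynamic parameters.
# Fake normal-based lighting mixes particle color, optional sky-atmosphere light and time-of-day tint into emissive.
# Opacity combines dissolve, inverted-Fresnel soft edges, near-camera and depth fades, plus a camera-distance dynamic branch at High/Epic quality.
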Material /Game/Effects/Explosion_Update/Materials/M_Explosion_MeshSmoke_Detail
graph {
 properties {
  bCastRayTracedShadows=false
  bUsedWithNiagaraMeshParticles=true
  bUsedWithNiagaraSprites=true
  blendMode=BLEND_Translucent
  pixelDepthOffsetMode=PDOM_Legacy
  shadingModel=MSM_Unlit
  usedShadingModels=Unlit
 }
 nodes [
  MaterialFunctionCall(Emissive=NamedRerouteUsage()) {materialFunction=&MF_FX_EmissiveAdjustments}
  n1 = Constant3Vector() {constant=<rgba 0 0 1 0>}
  n2 = DynamicParameter() {parameterIndex=3, defaultValue=<rgba 0 0 1 0>, paramNames [Fire Pan, Dissolve Pan, Temperature, FireNoise Pan]}
  n3 = DynamicParameter() {parameterIndex=1, defaultValue=<rgba 5 0.008 0.5 1>, paramNames [Temperature Pow, Noise Scale, Luminance Adjustment, Fresnel Mul]}
  n4 = Multiply(A=@n3.Noise Scale, B=ScalarParameter() {defaultValue=1, group=Fire, parameterName=Fire Noise Scale})
  n5 = Constant3Vector() {constant=<rgba 0 0 1 0>}
  n6 = Multiply(
   A=DynamicParameter() {defaultValue=<rgba -100 1 1 -1>, paramNames [Displacement, ShadowStrength, PlayAnimTime, Dissolve]}.PlayAnimTime
   B=@n5
  )
  n7 = Multiply(A=ParticleRandom()) {constB=100}
  n8 = Add(A=Multiply(A=NamedRerouteUsage()) {constA=2, constB=4}, B=Add(A=Add(A=@n6, B=@n7), B=WorldPosition().XYZ))
  n9 = Add(A=ParticlePositionWS(), B=WorldPosition().XYZ)
  n10 = DynamicParameter() {parameterIndex=3, defaultValue=<rgba 0 0 1 0>, paramNames [Fire Pan, Dissolve Pan, Temperature, FireNoise Pan]}
  n11 = Multiply(A=@n5, B=@n10.FireNoise Pan)
  n12 = ScalarParameter() {defaultValue=0.01, group=Fire, parameterName=Volume Noise Distortion UVW Scale}
  n13 = ScalarParameter() {defaultValue=1, parameterName=Input Global Scale}
  n14 = TextureSample(UVs=Multiply(A=Add(A=@n9, B=@n11), B=Divide(A=@n12, B=@n13) {constB=1}), Tex=NamedRerouteUsage()) {samplerType=SAMPLERTYPE_LinearColor}
  n15 = Subtract(A=@n14.R) {constB=0.5}
  n16 = ScalarParameter() {defaultValue=50, parameterName=Volume Distortion Scale}
  n17 = Subtract(A=Add(A=@n8, B=Multiply(A=@n15, B=@n16)), B=ParticlePositionWS()) {constB=0}
  n18 = TextureSample(UVs=Add(A=Multiply(A=@n1, B=@n2.Fire Pan), B=Multiply(A=@n4, B=@n17)), Tex=NamedRerouteUsage()) {samplerType=SAMPLERTYPE_LinearColor, texture=/Game/Effects/Explosion_Update/Textures/VT_Voronoi_Seamless}
  n19 = Multiply(A=@n18.R, B=@n18.R)
  n20 = Multiply(A=@n19, B=@n19)
  NamedRerouteDeclaration(None=Add(A=Multiply(A=@n20, B=@n20), B=@n18.R)) {name=Voronoi Noise Result Fire, nodeColor=<rgba 1 0 0.0703 1>}
  n21 = DynamicParameter() {parameterIndex=1, defaultValue=<rgba 5 0.008 0.5 1>, paramNames [Temperature Pow, Noise Scale, Luminance Adjustment, Fresnel Mul]}
  n22 = ParticlePositionWS()
  n23 = ParticleRandom()
  n24 = Multiply(
   A=Multiply(A=VectorParameter() {parameterName=Volonoi Noise Scale, defaultValue=<rgba 1 1 1 0>}.RGB, B=@n21.Noise Scale)
   B=Subtract(A=@n8, B=Add(A=@n22, B=@n23))
  )
  n25 = TextureSample(UVs=Add(A=NamedRerouteUsage(), B=Multiply(A=@n24) {constB=0.5}), Tex=NamedRerouteUsage()) {samplerType=SAMPLERTYPE_LinearColor, texture=&VT_Voronoi_Seamless}
  n26 = Multiply(A=@n25.R, B=@n25.R)
  NamedRerouteDeclaration(None=Multiply(A=@n26, B=@n26)) {name=Voronoi Noise Result 2, nodeColor=<rgba 1 0 0.0703 1>}
  n27 = Multiply(
   A=MaterialFunctionCall!(Center Point World Position=ParticlePositionWS()) {materialFunction=/Engine/Functions/Engine_MaterialFunctions02/UVs/WorldSpaceAlignedScreenCoordinates}.Offset and Scaled UVS
   B=ScalarParameter() {defaultValue=5, group=Detail Noise, parameterName=Detail UV Scale, sliderMax=3, sortPriority=2}
  ) {constB=3}
  n28 = Multiply(
   A=NamedRerouteUsage()
   B=ScalarParameter() {defaultValue=0.2, group=Detail Noise, parameterName=Detail UV Distort, sliderMax=1, sortPriority=3}
  )
  n29 = TextureObjectParameter() {group=Detail Noise, parameterName=Detail Noise, sortPriority=1, texture=/Game/Effects/Fort_Effects/Textures/Tile/T_GaussianNoise_01_M}
  n30 = Power(
   Base=TextureSample(UVs=Add(A=@n27, B=@n28), Tex=@n29) {texture=/Game/Effects/Fort_Effects/Textures/Tile/T_PerlinRGB_01_M}.RGB
   Exp=ScalarParameter() {defaultValue=0.3, group=Detail Noise, parameterName=Detail Exponent, sliderMax=3, sortPriority=5}
  ) {constExponent=0.6}
  n31 = ScalarParameter() {defaultValue=1, group=Detail Noise, parameterName=Detail Mult, sliderMax=3, sortPriority=6}
  NamedRerouteDeclaration(None=MaterialFunctionCall(
    Input=Multiply(A=@n30, B=@n31)
    Input Low=ScalarParameter() {group=Detail Noise, parameterName=Detail InputMin, sliderMax=1, sortPriority=12}
    Input High=ScalarParameter() {defaultValue=1, group=Detail Noise, parameterName=Detail InputMax, sliderMax=1, sortPriority=13}
    Target Low=ScalarParameter() {group=Detail Noise, parameterName=Detail OutputtMin, sliderMax=1, sortPriority=14}
    Target High=ScalarParameter() {defaultValue=1, group=Detail Noise, parameterName=Detail OutputtMax, sliderMax=1, sortPriority=15}
   ) {materialFunction=&RemapValueRange}.Result) {name=Detail Noise, nodeColor=<rgba 1 0 0.1172 1>}
  n32 = TextureObjectParameter() {
   group=Volume Textures
   parameterName=Displacement Volume Tex
   samplerType=SAMPLERTYPE_LinearColor
   texture=&VT_Voronoi_Seamless
  }
  NamedRerouteDeclaration(None=@n32) {name=Noise Texture 2, nodeColor=<rgba 0 0.7187 1 1>}
  n33 = DynamicParameter() {parameterIndex=3, defaultValue=<rgba 0 0 1 0>, paramNames [Fire Pan, Dissolve Pan, Temperature, FireNoise Pan]}
  n34 = NamedRerouteUsage()
  n35 = DynamicParameter() {parameterIndex=1, defaultValue=<rgba 5 0.008 0.5 1>, paramNames [Temperature Pow, Noise Scale, Luminance Adjustment, Fresnel Mul]}
  n36 = Power(
   Base=Power(Base=Saturate(None=@n34), Exp=ScalarParameter() {defaultValue=1, group=Fire, parameterName=Fire Posterize Adjustment})
   Exp=@n35.Temperature Pow
  )
  n37 = Multiply(B=@n35.Temperature Pow) {constA=0.02}
  n38 = Power(Base=Saturate(None=Subtract(A=@n36, B=@n37)), Exp=ScalarParameter() {defaultValue=2, group=Fire, parameterName=Fire Contrast})
  n39 = VectorParameter() {group=Fire, parameterName=FireColor, defaultValue=<rgba 4 0.3489 0 0>}
  n40 = StaticSwitch(
   True=Multiply(A=@n33.Temperature, B=Multiply(A=@n38, B=@n39.RGB))
   False=Constant()
   Value=StaticBoolParameter() {defaultValue=true, parameterName=Use High Temperature}
  )
  n41 = Power(Base=NamedRerouteUsage()) {constExponent=0.05}
  n42 = OneMinus(None=@n41)
  n43 = StaticSwitch(
   True=Power(Base=Saturate(None=@n41)) {constExponent=50}
   False=Multiply(A=@n42, B=@n42)
   Value=StaticBoolParameter() {group=Fire, parameterName=Use Inverse Tempereture}
  )
  n44 = DynamicParameter() {parameterIndex=2, defaultValue=<rgba 1 0 1 1>, paramNames [Swirl Scale, Swirl UV Pan, Fire2, Swirl UV Scale]}
  n45 = Multiply(
   A=Multiply(A=Multiply(A=@n43, B=@n44.Fire2), B=VectorParameter() {group=Fire, parameterName=FireColor2, defaultValue=<rgba 4 0.3489 0 0>}.RGB)
   B=Power(Base=NamedRerouteUsage(), Exp=ScalarParameter() {defaultValue=3, group=Fire, parameterName=Fire2NoisePow})
  )
  NamedRerouteDeclaration(None=Add(A=@n40, B=@n45)) {name=Fire, nodeColor=<rgba 1 0 0.7266 1>}
  n46 = Length(None=Subtract(A=CameraPositionWS(), B=ParticlePositionWS())) {is=class MaterialExpressionLength}
  n47 = ScalarParameter() {defaultValue=800, desc=Near distance to camera --250, group=CamFade, parameterName=CamFade Near, sliderMax=500, sortPriority=1}
  n48 = ScalarParameter() {defaultValue=1000, desc=Far distance to camera -- 300, group=CamFade, parameterName=CamFade Far, sliderMax=2000, sortPriority=2}
  NamedRerouteDeclaration(None=Saturate(None=MaterialFunctionCall(Input=@n46, Input Low=@n47, Input High=@n48) {materialFunction=&RemapValue_0-1}.Result)) {name=Camera Fade Alpha, nodeColor=<rgba 1 0 0.6094 1>}
  NamedRerouteDeclaration(None=@n14.R) {name=Noise_for_Fire2, nodeColor=<rgba 0 0.9297 1 1>}
  n49 = Multiply(A=VertexNormalWS(), B=NamedRerouteUsage())
  n50 = DynamicParameter() {defaultValue=<rgba -100 1 1 -1>, paramNames [Displacement, ShadowStrength, PlayAnimTime, Dissolve]}
  NamedRerouteDeclaration(None=StaticSwitch(
    True=Multiply(A=@n49, B=@n50.Displacement)
    False=Constant()
    Value=StaticBoolParameter() {defaultValue=true, parameterName=Use Displacement}
   )) {name=Displacement Result, nodeColor=<rgba 0 0.6484 1 1>}
  n51 = DynamicParameter() {parameterIndex=2, defaultValue=<rgba 1 0 1 1>, paramNames [Swirl Scale, Swirl UV Pan, Fire2, Swirl UV Scale]}
  n52 = Multiply(
   A=WorldPosition().XYZ
   B=Divide(
    A=ScalarParameter() {defaultValue=0.0007, parameterName=Swirl UV Scale}
    B=ScalarParameter() {defaultValue=1, parameterName=Input Global Scale}
   ) {constB=1}
  )
  n53 = DynamicParameter() {parameterIndex=2, defaultValue=<rgba 1 0 1 1>, paramNames [Swirl Scale, Swirl UV Pan, Fire2, Swirl UV Scale]}
  n54 = Constant3Vector() {constant=<rgba 0.1 0.1 1 0>}
  n55 = TextureSample(UVs=Add(A=Multiply(A=@n51.Swirl UV Scale, B=@n52), B=Multiply(A=@n53.Swirl UV Pan, B=@n54)), Tex=NamedRerouteUsage()) {samplerType=SAMPLERTYPE_LinearColor}
  NamedRerouteDeclaration(None=Multiply(A=Subtract(A=@n55.RGB) {constB=0.5}, B=@n53.Swirl Scale)) {name=Swirl UV, nodeColor=<rgba 1 0 0.9609 1>}
  n56 = ScalarParameter() {defaultValue=8, parameterName=Fresnel Alpha Expo}
  n57 = ScalarParameter() {parameterName=BaseReflectFraction}
  n58 = DynamicParameter() {parameterIndex=1, defaultValue=<rgba 5 0.008 0.5 1>, paramNames [Temperature Pow, Noise Scale, Luminance Adjustment, Fresnel Mul]}
  n59 = Power(
   Base=OneMinus(None=Fresnel(ExponentIn=@n56, BaseReflectFractionIn=@n57) {baseReflectFraction=0.01})
   Exp=Multiply(A=ScalarParameter() {defaultValue=20, parameterName=FresnelPower}, B=@n58.Fresnel Mul)
  )
  NamedRerouteDeclaration(None=@n59) {name=Fresnel, nodeColor=<rgba 0.1719 1 0 1>}
  n60 = TextureObjectParameter() {
   group=Volume Textures
   parameterName=Dissolve and Swirl Noise Volume Tex
   samplerType=SAMPLERTYPE_LinearColor
   texture=/Game/Global/MaterialFunctions/VT_ClothSimWPO
  }
  NamedRerouteDeclaration(None=@n60) {name=Noise Texture 1, nodeColor=<rgba 0 0.6719 1 1>}
  n61 = MaterialFunctionCall(
   Color_Morning=VectorParameter() {group=Time of Day Color, parameterName=Color Morning, defaultValue=<rgba 0.4844 0.439 0.439 1>}.RGB
   Color_Day=VectorParameter() {group=Time of Day Color, parameterName=Color Day, defaultValue=<rgba 0.3906 0.4046 0.4167 1>}.RGB
   Color_Evening=VectorParameter() {group=Time of Day Color, parameterName=Color Evening, defaultValue=<rgba 0.6667 0.6253 0.6042 1>}.RGB
   Color_Night=VectorParameter() {group=Time of Day Color, parameterName=Color Night, defaultValue=<rgba 0.5312 0.5633 0.6667 1>}.RGB
  ) {materialFunction=&Color_Time_of_Day}
  n62 = ShadingPathSwitch(
   Default=@n61.Result
   Deferred=@n61.Result
   Mobile=MaterialFunctionCall(
    Color_Morning=VectorParameter() {group=Time of Day Color, parameterName=Color Morning, defaultValue=<rgba 0.4844 0.439 0.439 1>}.RGB
    Color_Day=VectorParameter() {group=Time of Day Color, parameterName=Color Day, defaultValue=<rgba 0.3906 0.4046 0.4167 1>}.RGB
    Color_Evening=VectorParameter() {group=Time of Day Color, parameterName=Color Evening, defaultValue=<rgba 0.6667 0.6253 0.6042 1>}.RGB
    Color_Night=VectorParameter() {group=Time of Day Color, parameterName=Color Night, defaultValue=<rgba 0.5312 0.5633 0.6667 1>}.RGB
   ) {materialFunction=&Color_Time_of_Day}.Result
  ) {desc=Migrated from FeatureLevelSwitch}
  n63 = ScalarParameter() {defaultValue=0.075, desc=How much consider the sky atmosphere lighting, group=Shading, parameterName=SkyAtmosphere LightScale}
  n64 = WorldPosition()
  n65 = Abs(None=@n64.Z)
  n66 = Multiply(A=@n63, B=SkyAtmosphereLightIlluminance!(World Position=AppendVector(A=@n64.XY, B=@n65)))
  n67 = Saturate(None=Add(A=@n66, B=@n63) {constB=0.1})
  n68 = Constant()
  n69 = Constant()
  n70 = ParticleColor()
  n71 = DynamicParameter() {parameterIndex=1, defaultValue=<rgba 5 0.008 0.5 1>, paramNames [Temperature Pow, Noise Scale, Luminance Adjustment, Fresnel Mul]}
  n72 = Multiply(A=Add(A=PixelNormalWS()), B=@n71.Luminance Adjustment) {constB=0.5}
  n73 = NamedRerouteUsage()
  n74 = Add(A=Multiply(A=@n72, B=OneMinus(None=@n73)), B=@n72)
  n75 = SkyAtmosphereLightDirection()
  n76 = ScalarParameter() {defaultValue=0.333, group=Shading, parameterName=SkyAtmosphereLightDir_BlendRate}
  n77 = AppendVector(A=ComponentMask(None=Multiply(A=@n75, B=@n76) {constB=0.333}) {g=true, r=true}, B=Constant() {r=1})
  n78 = Add(A=Multiply(A=DotProduct(A=@n74, B=@n77)) {constB=0.5}) {constB=0.5}
  n79 = DynamicParameter() {parameterIndex=1, defaultValue=<rgba 5 0.008 0.5 1>, paramNames [Temperature Pow, Noise Scale, Luminance Adjustment, Fresnel Mul]}
  n80 = Constant() {r=1.5}
  n81 = Multiply(
   A=NamedRerouteUsage()
   B=DynamicParameter() {defaultValue=<rgba -100 1 1 -1>, paramNames [Displacement, ShadowStrength, PlayAnimTime, Dissolve]}.ShadowStrength
  )
  n82 = Multiply(
   A=ShadingPathSwitch(Default=@n78, Deferred=@n78, Mobile=Multiply(A=@n79.Luminance Adjustment, B=@n80)) {desc=Migrated from FeatureLevelSwitch}
   B=OneMinus(None=Saturate(None=@n81))
  ) {constA=1}
  n83 = Multiply(A=Add(A=StaticSwitchParameter(True=@n66, False=@n69) {group=Shading, parameterName=Use SkyAtmosphereLightColor}, B=@n70.RGB), B=@n82)
  n84 = Multiply(A=@n82) {constB=0.5}
  n85 = StaticSwitchParameter(
   True=Desaturation(
    None=ShadingPathSwitch(Default=@n67, Deferred=@n67, Mobile=@n68) {desc=Migrated from FeatureLevelSwitch}
    Fraction=ScalarParameter() {defaultValue=0.5, parameterName=SkyAtmos Desaturation, sliderMax=1}
   )
   False=Add(A=Multiply(A=@n68, B=@n83) {constA=1}, B=Multiply(A=@n70.RGB, B=@n84))
  ) {group=Shading, parameterName=Use SkyAtmosphereLightColor}
  n86 = Add(A=Add(A=@n85, B=@n40), B=@n45) {constB=0}
  n87 = Add(A=@n85, B=@n45)
  n88 = Multiply(
   A=StaticSwitchParameter(True=@n85, False=ShadingPathSwitch(Default=@n86, Deferred=@n86, Mobile=@n87) {desc=Migrated from FeatureLevelSwitch}) {parameterName=IS_Simple Smoke}
   B=Power(Base=NamedRerouteUsage()) {constExponent=0.6}
  )
  NamedRerouteDeclaration(None=StaticSwitchParameter(True=Multiply(A=@n62, B=@n88), False=@n88) {defaultValue=true, group=Shading, parameterName=UseTimeOfDayColor}) {name=Emissive Result, nodeColor=<rgba 0 0.5781 1 1>}
  NamedRerouteDeclaration(None=Multiply(A=@n70.RGB, B=@n82)) {name=Base Color Result, nodeColor=<rgba 0.0312 1 0 1>}
  NamedRerouteDeclaration(None=Transform(None=@n25.R)) {name=Voronoi Noise Result 1, nodeColor=<rgba 0.8281 1 0 1>}
  NamedRerouteUsage()
  NamedRerouteUsage()
  n89 = MaterialFunctionCall(Distance=ScalarParameter() {defaultValue=200, parameterName=Near Camera Fade Distance}) {materialFunction=&Near_Camera_Fade}
  n90 = Multiply(A=@n89.Result, B=@n89.Result)
  n91 = VertexInterpolator(VS=@n90)
  n92 = DynamicParameter() {parameterIndex=3, defaultValue=<rgba 0 0 1 0>, paramNames [Fire Pan, Dissolve Pan, Temperature, FireNoise Pan]}
  n93 = Multiply(A=Constant3Vector() {constant=<rgba 0 0 1 0>}, B=@n92.Dissolve Pan)
  n94 = WorldPosition()
  n95 = Multiply(A=Time(), B=VectorParameter() {group=DIssolve, parameterName=Dissolve Noise Pan Speed XYZ, defaultValue=<rgba 1 1 -40 0>}.RGB)
  n96 = VectorParameter() {group=DIssolve, parameterName=DissolveNoiseScale, defaultValue=<rgba 0.002 0.002 0.005 1>}
  n97 = Divide(
   A=ScalarParameter() {defaultValue=2.56, parameterName=Dissolve Noise Scale}
   B=ScalarParameter() {defaultValue=1, parameterName=Input Global Scale 2}
  ) {constB=1}
  n98 = Add(A=Multiply(A=Add(A=@n94.XYZ, B=@n95), B=Multiply(A=@n96.RGB, B=@n97)) {constB=0.01}, B=NamedRerouteUsage())
  n99 = OneMinus(None=TextureSample(UVs=Add(A=@n93, B=@n98), Tex=NamedRerouteUsage()) {samplerType=SAMPLERTYPE_LinearColor}.R)
  n100 = Saturate(None=Subtract(
    A=Multiply(A=@n99, B=@n99)
    B=DynamicParameter() {defaultValue=<rgba -100 1 1 -1>, paramNames [Displacement, ShadowStrength, PlayAnimTime, Dissolve]}.Dissolve
   ))
  n101 = ScalarParameter() {defaultValue=9, group=DIssolve, parameterName=Pow for Dissolve}
  n102 = Multiply(A=@n59) {constB=2}
  n103 = Multiply(A=Multiply(A=Power(Base=@n100, Exp=@n101) {constExponent=9}, B=ParticleColor().A), B=Saturate(None=Subtract(A=@n102) {constB=0.1}))
  n104 = NamedRerouteUsage()
  n105 = NamedRerouteUsage()
  n106 = Multiply(
   A=@n103
   B=MaterialFunctionCall!(
    Alpha=OneMinus(None=@n104)
    Color Input 0=Constant() {r=1}
    Color Input 1=LinearInterpolate(A=@n105, Alpha=@n104) {constA=1, desc="lerp near, far"}
   ) {desc=reversed condition, materialFunction=/Engine/Functions/Engine_MaterialFunctions02/Utility/DynamicBranch}.Result
  )
  n107 = QualitySwitch(Default=@n103, Low=@n103, High=@n106, Medium=@n103, Epic=@n106)
  n108 = StaticSwitchParameter(True=Multiply(A=StaticSwitchParameter(True=@n90, False=@n91.PS) {parameterName=PerPixel Fading}, B=@n107), False=@n107) {defaultValue=true, parameterName=Use Near Camera Fade}
  n109 = DepthFade(Opacity=@n108, FadeDistance=ScalarParameter() {defaultValue=60, parameterName=DepthFade})
  Reroute(None=ShadingPathSwitch(Default=@n109, Deferred=@n109, Mobile=@n108) {desc=Migrated from FeatureLevelSwitch})
 ]
}
