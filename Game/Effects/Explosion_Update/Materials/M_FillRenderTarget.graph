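# Translucent unlit Niagara sprite and mesh material for explosion effects that shades a low-resolution FillRT render-target texture.
# A quality switch picks plain sampling or, on higher tiers, a nine-tap blur to hide the small target size.
# That value drives gradient-curve color, with a DDY fake shadow for volume, and smoothstepped opacity with sine edge, depth and camera fades.
# An unused bicubic sampling node remains in the graph.
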Material /Game/Effects/Explosion_Update/Materials/M_FillRenderTarget
graph {
 properties {
  bUsedWithNiagaraMeshParticles=true
  bUsedWithNiagaraSprites=true
  blendMode=BLEND_Translucent
  opacityMaskClipValue=0.1
  pixelDepthOffsetMode=PDOM_Legacy
  shadingModel=MSM_Unlit
  usedShadingModels=Unlit
 }
 nodes [
  MaterialExpressionTextureCoordinate_1 = TextureCoordinate()
  MaterialExpressionTextureObjectParameter_0 = TextureObjectParameter() {parameterName=FillRT, samplerType=SAMPLERTYPE_LinearColor, texture=/Game/Global/Textures/Default/Blanks/T_BlackLinearColor}
  MaterialExpressionDynamicParameter_3 = DynamicParameter() {parameterIndex=1, defaultValue=<rgba 256 0.65 0.5 0>, paramNames [RT Size, Blur Distance, Fake Shadow Strength, SmoothStep Min]}
  MaterialExpressionScalarParameter_2 = ScalarParameter() {parameterName=s}
  Custom!(
   UV=@MaterialExpressionTextureCoordinate_1
   Tex=@MaterialExpressionTextureObjectParameter_0
   RenderTargetSize=@MaterialExpressionDynamicParameter_3.RT Size
   Mip=@MaterialExpressionScalarParameter_2
   Samp=@MaterialExpressionTextureObjectParameter_0
  ) {
   code="// Inputs: Texture2D Tex, SamplerState Samp, float2 UV, float2 TexSize, float Mip\r\n// 出力: float4\r\n\r\n// Catmull-Rom cubic weight (a = -0.5)\r\nfloat cubic_w(float x)\r\n{\r\n    x = abs(x);\r\n    float x2 = x * x;\r\n    float x3 = x2 * x;\r\n\r\n    // 0 <= x < 1\r\n    float w0 = (1.5*x3 - 2.5*x2 + 1.0);\r\n    // 1 <= x < 2\r\n    float w1 = (-0.5*x3 + 2.5*x2 - 4.0*x + 2.0);\r\n\r\n    // piecewise blend\r\n    return (x < 1.0) ? w0 : ((x < 2.0) ? w1 : 0.0);\r\n}\r\n\r\n// 4x4 bicubic sample (point-filter fetches)\r\nfloat4 Bicubic(Texture2D T, SamplerState S, float2 uv, float2 texSize, float mip)\r\n{\r\n    // テクセル中心基準の座標へ（uv を texel 空間に）\r\n    float2 p  = uv * texSize - 0.5;\r\n    float2 pf = floor(p);\r\n    float2 f  = p - pf;          // 小数部\r\n\r\n    // 1D 重みを先に計算\r\n    float wx[4];\r\n    float wy[4];\r\n    [unroll] for (int i = 0; i < 4; ++i)\r\n    {\r\n        wx[i] = cubic_w((i - 1) - f.x);\r\n        wy[i] = cubic_w((i - 1) - f.y);\r\n    }\r\n\r\n    // 正規化（Catmull-Rom は理論上ほぼ1だが端での精度を安定化）\r\n    float sx = wx[0]+wx[1]+wx[2]+wx[3];\r\n    float sy = wy[0]+wy[1]+wy[2]+wy[3];\r\n    [unroll] for (int i = 0; i < 4; ++i) { wx[i] /= sx; wy[i] /= sy; }\r\n\r\n    float4 c = 0;\r\n    [unroll] for (int j = -1; j <= 2; ++j)\r\n    {\r\n        float wyj = wy[j+1];\r\n        [unroll] for (int i = -1; i <= 2; ++i)\r\n        {\r\n            float wxi = wx[i+1];\r\n            float2 tc = (pf + float2(i, j) + 0.5) / texSize; // 再び 0..1 に戻す\r\n            c += Texture2DSampleLevel(T, S, tc, mip) * (wxi * wyj);\r\n        }\r\n    }\r\n    return c;\r\n}\r\n\r\nreturn Bicubic(Tex, Samp, UV, TexSize, Mip);\r\n"
   desc="The reason why we need Blur, Render Target size is tiny due to redude memory usage."
   description=Blur
   outputType=CMOT_Float4
   showCode=true
   inputs [
    {inputName=UV, input {expression=&self:MaterialExpressionTextureCoordinate_1}}
    {inputName=Tex, input {expression=&self:MaterialExpressionTextureObjectParameter_0}}
    {
     inputName=RenderTargetSize
     input {expression=&self:MaterialExpressionDynamicParameter_3, mask=1, maskR=1}
    }
    {inputName=Mip, input {expression=&self:MaterialExpressionScalarParameter_2}}
    {inputName=Samp, input {expression=&self:MaterialExpressionTextureObjectParameter_0}}
   ]
  }
  n1 = DynamicParameter() {parameterIndex=1, defaultValue=<rgba 256 0.65 0.5 0>, paramNames [RT Size, Blur Distance, Fake Shadow Strength, SmoothStep Min]}
  n2 = TextureSample(Tex=@MaterialExpressionTextureObjectParameter_0)
  n3 = Custom!(
   UV=@MaterialExpressionTextureCoordinate_1
   Tex=@MaterialExpressionTextureObjectParameter_0
   RenderTargetSize=@MaterialExpressionDynamicParameter_3.RT Size
   BlurDistance=@MaterialExpressionDynamicParameter_3.RT Size
  ) {
   code="float2 texel = BlurDistance / RenderTargetSize;\r\nfloat4 bl = 0.0;\r\nbl += Texture2DSample(Tex, TexSampler, UV)                          * 0.2;\r\nbl += Texture2DSample(Tex, TexSampler, UV + texel * float2( 1,  0)) * 0.1;\r\nbl += Texture2DSample(Tex, TexSampler, UV + texel * float2(-1,  0)) * 0.1;\r\nbl += Texture2DSample(Tex, TexSampler, UV + texel * float2( 0,  1)) * 0.1;\r\nbl += Texture2DSample(Tex, TexSampler, UV + texel * float2( 0, -1)) * 0.1;\r\nbl += Texture2DSample(Tex, TexSampler, UV + texel * float2( 1,  1)) * 0.1;\r\nbl += Texture2DSample(Tex, TexSampler, UV + texel * float2(-1,  1)) * 0.1;\r\nbl += Texture2DSample(Tex, TexSampler, UV + texel * float2( 1, -1)) * 0.1;\r\nbl += Texture2DSample(Tex, TexSampler, UV + texel * float2(-1, -1)) * 0.1;\r\nreturn bl;"
   desc="The reason why we need Blur, Render Target size is tiny due to redude memory usage."
   description=Blur
   outputType=CMOT_Float4
   showCode=true
   inputs [
    {inputName=UV, input {expression=&self:MaterialExpressionTextureCoordinate_1}}
    {inputName=Tex, input {expression=&self:MaterialExpressionTextureObjectParameter_0}}
    {
     inputName=RenderTargetSize
     input {expression=&self:MaterialExpressionDynamicParameter_3, mask=1, maskR=1}
    }
    {
     inputName=BlurDistance
     input {expression=&self:MaterialExpressionDynamicParameter_3, mask=1, maskG=1, outputIndex=1}
    }
   ]
  }
  n4 = ComponentMask(None=@n3) {r=true}
  n5 = MaterialFunctionCall!(
   Low _ PC Low and Med=@n2.R
   "High _ XB1, PS4"=@n4
   "Epic _ PC High, XSX, PS5"=@n4
   --------=@n2.R
   "Mobile Low _ PC Perf, Android"=@n2.R
   Mobile Med _ Switch=@n2.R
  ) {materialFunction=/Game/Global/MaterialFunctions/TechArt/MF_Fortnite_QualitySwitch_Scalar}
  n6 = ParticleColor()
  n7 = Multiply(A=SmoothStep(Min=@n1.SmoothStep Min, Max=ScalarParameter() {defaultValue=0.8, parameterName=SmoothStep Max}, Value=@n5.Result), B=@n6.A)
  n8 = Multiply(A=TextureCoordinate()) {constB=0.5}
  n9 = MaterialFunctionCall(Float2=Abs(None=Sine(None=@n8))) {materialFunction=&BreakOutFloat2Components}
  n10 = Power(Base=Multiply(A=@n9.R, B=@n9.R)) {constExponent=0.8}
  DepthFade(
   Opacity=Multiply(A=NamedRerouteUsage(), B=Multiply(A=@n7, B=@n10))
   FadeDistance=DynamicParameter() {defaultValue=<rgba 80 2 1 0>, paramNames [Depth Fade, Curve Color Pow, Intensity Scale, ""]}.Depth Fade
  )
  n11 = AtmosphericLightColor()
  n12 = ScalarParameter() {defaultValue=0.002, parameterName=Env Color Add Rate}
  n13 = DynamicParameter() {parameterIndex=2, defaultValue=<rgba 1 1 1 0>, paramNames [R, G, B, 2nd Color Hue Shift]}
  n14 = DynamicParameter() {defaultValue=<rgba 80 2 1 0>, paramNames [Depth Fade, Curve Color Pow, Intensity Scale, ""]}
  n15 = CurveAtlasRowParameter(CurveTime=Power(Base=@n5.Result, Exp=@n14.Curve Color Pow)) {
   atlas=/Game/Curie/Effects/temp/Curves/CA_Curie_Fire_Visuals
   curve=/Game/Curie/Effects/temp/Curves/Curve_Gradient_Particle
   defaultValue=11
   parameterName=Curve
  }
  n16 = MaterialFunctionCall(Hue Shift Percentage=@n13.2nd Color Hue Shift, Texture=@n13.2nd Color Hue Shift) {materialFunction=&HueShift}
  n17 = OneMinus(None=@n15)
  n18 = ScalarParameter() {defaultValue=500, parameterName=Color Balance}
  n19 = Multiply(
   A=LinearInterpolate(A=Multiply(A=@n13.RGB, B=@n15), B=Multiply(A=@n16.Result, B=@n15), Alpha=Power(Base=@n17, Exp=@n18))
   B=@n14.Intensity Scale
  )
  n20 = Add(A=Multiply(A=DDY(Value=@n5.Result) {desc=for Volume look}, B=@n1.Fake Shadow Strength), B=@n6.RGB)
  MaterialFunctionCall(Emissive=Add(A=Multiply(A=@n11, B=@n12), B=Add(A=@n19, B=@n20))) {materialFunction=&MF_FX_EmissiveAdjustments}
  n21 = MaterialFunctionCall(Distance=ScalarParameter() {defaultValue=330, parameterName=Near Camera Fade Distance}) {materialFunction=&Near_Camera_Fade}
  NamedRerouteDeclaration(None=VertexInterpolator(VS=Multiply(A=@n21.Result, B=@n21.Result)).PS) {name=Near Camera Fade, nodeColor=<rgba 1 0.1406 0 1>}
 ]
}
