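# Translucent unlit Niagara mesh-particle material that animates explosion sparks in the vertex shader.
# World position offset thickens spline meshes, pushes and stretches them radially from the particle center, then adds gravity and panning-noise turbulence, blending previous and current time along UV to bend streaks.
# Dynamic parameters drive motion and shape, while vertex color varies sparks.
# Particle color gets hue shift and HDR remapping; opacity combines streak mask, camera/depth fades and random disappearance.
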Material /Game/Effects/Explosion_Update/Materials/M_MeshSparks_WPO
graph {
 properties {
  bAutomaticallySetUsageInEditor=false
  bUsedWithNiagaraMeshParticles=true
  blendMode=BLEND_Translucent
  shadingModel=MSM_Unlit
  usedShadingModels=Unlit
 }
 nodes [
  n1 = DynamicParameter() {parameterIndex=1, defaultValue=<rgba 1 0 1 0>, paramNames [Sparks Thickness, Turbulence, Global Scaling, Gravity Adjustment]}
  n2 = Multiply(A=NamedRerouteUsage(), B=Multiply(A=@n1.Sparks Thickness, B=@n1.Sparks Thickness))
  n3 = StaticBool() {value=true}
  n4 = ComponentMask(None=TextureCoordinate()) {g=true}
  n5 = Add(A=@n4) {constB=-0.5}
  n6 = Sign(None=@n5)
  n7 = VertexColor()
  n8 = Min(A=Multiply(A=@n5) {constB=3})
  n9 = NamedRerouteUsage()
  n10 = DynamicParameter() {parameterIndex=2, defaultValue=<rgba 0.5 10 0 10>, paramNames [Particle Time, Shape Adjustment, Random Hue Shift, Param4]}
  n11 = DynamicParameter() {defaultValue=<rgba 0 20 0 1>, paramNames [Spark Travel Offset, Depth Fade, Random Disappearance, Spark Length]}
  n12 = DynamicParameter() {parameterIndex=1, defaultValue=<rgba 1 0 1 0>, paramNames [Sparks Thickness, Turbulence, Global Scaling, Gravity Adjustment]}
  n13 = Multiply(A=LinearInterpolate(A=@n4, B=Multiply(A=@n8, B=@n9), Alpha=@n10.Particle Time), B=Multiply(A=@n11.Spark Length, B=@n12.Global Scaling))
  n14 = Normalize(VectorInput=Subtract(A=WorldPosition().XYZ, B=ParticlePositionWS()))
  n15 = Multiply(A=@n9, B=@n14) {constB=-1}
  n16 = LinearInterpolate(
   A=Multiply(A=Multiply(A=@n6, B=@n7.B), B=Multiply(A=@n13, B=@n15))
   Alpha=Saturate(None=Multiply(A=@n10.Particle Time, B=@n10.Particle Time))
  ) {constB=0.1}
  n17 = DynamicParameter() {defaultValue=<rgba 0 20 0 1>, paramNames [Spark Travel Offset, Depth Fade, Random Disappearance, Spark Length]}
  n18 = DynamicParameter() {parameterIndex=1, defaultValue=<rgba 1 0 1 0>, paramNames [Sparks Thickness, Turbulence, Global Scaling, Gravity Adjustment]}
  n19 = Multiply(A=@n17.Spark Travel Offset, B=@n18.Global Scaling)
  n20 = Multiply(A=Multiply(A=@n14, B=VertexColor().B), B=@n19)
  Add(
   A=NamedRerouteUsage()
   B=Add(
    A=MaterialFunctionCall(WidthBase=@n2, WidthTip=@n2, Expand U or V UV Channel=@n3) {materialFunction=&SplineThicken}.WorldPositionOffset
    B=Add(A=@n16, B=@n20)
   ) {constB=0}
  )
  DynamicParameter() {defaultValue=<rgba 0 20 0 1>, paramNames [Spark Travel Offset, Depth Fade, Random Disappearance, Spark Length]}
  n21 = NamedRerouteUsage()
  n22 = NamedRerouteUsage()
  n23 = VertexColor()
  n24 = DynamicParameter() {parameterIndex=2, defaultValue=<rgba 0.5 10 0 10>, paramNames [Particle Time, Shape Adjustment, Random Hue Shift, Param4]}
  n25 = Multiply(A=@n23.G, B=@n24.Random Hue Shift)
  n26 = ParticleColor()
  n27 = TextureCoordinate()
  n28 = ComponentMask(None=@n27) {g=true}
  n29 = DynamicParameter() {parameterIndex=2, defaultValue=<rgba 0.5 10 0 10>, paramNames [Particle Time, Shape Adjustment, Random Hue Shift, Param4]}
  n30 = VertexColor()
  n31 = DynamicParameter() {parameterIndex=2, defaultValue=<rgba 0.5 10 0 10>, paramNames [Particle Time, Shape Adjustment, Random Hue Shift, Param4]}
  n32 = Max(A=Divide(A=Multiply(A=@n29.Shape Adjustment, B=@n30.B), B=Add(A=@n31.Particle Time)))
  n33 = Multiply(A=@n28, B=@n32) {constB=2}
  n34 = OneMinus(None=Sign(None=Subtract(A=@n33) {constB=0.5}))
  n35 = Multiply(A=@n33, B=@n34)
  n36 = Multiply(A=OneMinus(None=@n28), B=@n32) {constB=2}
  n37 = Subtract(A=@n36) {constB=0.5}
  n38 = Saturate(None=OneMinus(None=Sign(None=@n37)))
  n39 = OneMinus(None=Add(A=@n34, B=@n38))
  n40 = Add(A=Multiply(A=Saturate(None=@n35)) {constB=0.5}, B=Saturate(None=Multiply(A=@n39) {constB=0.5}))
  n41 = Multiply(A=@n38, B=OneMinus(None=@n36))
  n42 = Constant() {r=0.5}
  n43 = SphereMask(A=AppendVector(A=ComponentMask(None=@n27) {r=true}, B=Add(A=@n40, B=@n41)), B=@n42, Radius=@n42) {hardnessPercent=0}
  n44 = Multiply(
   A=Multiply(
    A=LinearInterpolate(A=@n21, B=@n22, Alpha=@n23.B)
    B=MaterialFunctionCall(Hue Shift Percentage=@n25, Texture=@n26.RGB) {materialFunction=&HueShift}.Result
   )
   B=@n43
  )
  n45 = Clamp(None=@n44)
  n46 = ScalarParameter() {defaultValue=1.2, parameterName=HDRMult}
  MaterialFunctionCall(Emissive=ShadingPathSwitch(
    Default=MaterialFunctionCall(LDR Input=@n45, HDR Tint=@n44, HDR Intensity Multiplier=@n46) {materialFunction=&DeriveHDRfromLDR}.RemappedRange
    Mobile=@n44
   )) {materialFunction=&MF_FX_EmissiveAdjustments}
  n47 = OneMinus(None=Abs(None=TransformPosition(None=@n14) {transformSourceType=TRANSFORMPOSSOURCE_World}))
  Multiply(A=@n19, B=Add(A=ComponentMask(None=@n47) {g=true}, B=ComponentMask(None=@n47) {b=true}))
  n48 = Multiply(A=Constant3Vector() {constant=<rgba 0 0 1 0>}, B=NamedRerouteUsage())
  n49 = DynamicParameter() {parameterIndex=1, defaultValue=<rgba 1 0 1 0>, paramNames [Sparks Thickness, Turbulence, Global Scaling, Gravity Adjustment]}
  n50 = DynamicParameter() {parameterIndex=2, defaultValue=<rgba 0.5 10 0 10>, paramNames [Particle Time, Shape Adjustment, Random Hue Shift, Param4]}
  n51 = OneMinus(None=@n50.Particle Time)
  n52 = DeltaTime()
  n53 = Subtract(A=@n50.Particle Time, B=Multiply(A=@n51, B=Multiply(A=@n52) {constB=2}))
  n54 = Power(Base=@n49.Gravity Adjustment, Exp=Add(A=Max(A=@n53) {constB=0})) {desc=previous}
  n55 = Multiply(A=@n49.Global Scaling, B=Multiply(A=Subtract(A=@n54)) {constB=-1})
  n56 = DynamicParameter() {parameterIndex=2, defaultValue=<rgba 0.5 10 0 10>, paramNames [Particle Time, Shape Adjustment, Random Hue Shift, Param4]}
  n57 = Subtract(A=Power(Base=@n49.Gravity Adjustment, Exp=Add(A=@n56.Particle Time)) {desc=previous})
  n58 = Multiply(A=Multiply(A=@n49.Global Scaling, B=Multiply(A=@n57) {constB=-1}), B=@n48)
  n59 = DynamicParameter() {parameterIndex=2, defaultValue=<rgba 0.5 10 0 10>, paramNames [Particle Time, Shape Adjustment, Random Hue Shift, Param4]}
  NamedRerouteDeclaration(None=LinearInterpolate(A=Multiply(A=@n48, B=@n55), B=NamedRerouteDeclaration(None=@n58) {name=Grav, nodeColor=<rgba 0.4062 0 1 1>}, Alpha=@n59.Particle Time)) {name=Prev Grav, nodeColor=<rgba 1 0 0.5716 1>}
  n60 = DynamicParameter() {parameterIndex=1, defaultValue=<rgba 1 0 1 0>, paramNames [Sparks Thickness, Turbulence, Global Scaling, Gravity Adjustment]}
  NamedRerouteDeclaration(None=Multiply(A=@n60.Turbulence, B=@n60.Turbulence)) {name=Turb, nodeColor=<rgba 0 0.1562 1 1>}
  n61 = VertexColor()
  MaterialExpressionAdd_10 = Add(A=Multiply(A=@n61) {constB=0.8}) {constB=0.2}
  n62 = Convert(None=@MaterialExpressionAdd_10) {
   nodeName=Break Float3
   convertInputs [{
     type=Vector3
     expressionInput {
      expression=&self:MaterialExpressionAdd_10
      inputName=None
      mask=0
      maskA=0
      maskB=0
      maskG=0
      maskR=0
      outputIndex=0
     }
    }]
   convertOutputs [{type=Scalar}, {type=Scalar}, {type=Scalar}]
  }
  NamedRerouteDeclaration(None=@n62) {name=Rand Vert G, nodeColor=<rgba 0.1699 1 0.1865 1>}
  n63 = Multiply(A=@n61) {constB=0.5}
  NamedRerouteDeclaration(None=ComponentMask(None=Add(A=@n63) {constB=0.5}) {b=true}) {name=Rand Vert B, nodeColor=<rgba 0 0.1982 1 1>}
  n64 = VectorParameter() {parameterName=Turbulence Strength XYZ, defaultValue=<rgba 1 1 0.5 1>}
  n65 = NamedRerouteUsage()
  n66 = WorldPosition()
  n67 = ActorPositionWS()
  n68 = Add(
   A=@n30.B
   B=Multiply(A=Subtract(A=@n66.XYZ, B=@n67), B=ScalarParameter() {defaultValue=0.003, parameterName=Turbulence Noise Scale}) {constB=0.01}
  )
  n69 = Multiply(A=ParticleRandom()) {constB=5}
  n70 = VectorParameter() {parameterName=Noise Pan Velocity, defaultValue=<rgba 0.05 0.05 -0.01 0>}
  n71 = Multiply(A=@n52) {constB=5}
  n72 = Multiply(A=@n70.RGB, B=Subtract(A=@n50.Particle Time, B=Multiply(A=@n51, B=@n71)))
  n73 = TextureSample(UVs=Add(A=@n68, B=Add(A=@n69, B=@n72))) {samplerType=SAMPLERTYPE_LinearColor, texture=/Game/Gadgets/Assets/VinderTech_GliderChute/Glider_Mitosis_Male/Textures/Noise/VT_GaussianNoise}
  n74 = Multiply(A=@n64.RGB, B=Multiply(A=@n65, B=Subtract(A=@n73.RGB) {constB=0.55}))
  n75 = NamedRerouteUsage()
  n76 = Multiply(A=@n70.RGB, B=@n50.Particle Time)
  n77 = TextureSample(UVs=Add(A=@n68, B=Add(A=@n76, B=@n69))) {samplerType=SAMPLERTYPE_LinearColor, texture=&VT_GaussianNoise}
  n78 = Multiply(A=Multiply(A=Subtract(A=@n77.RGB) {constB=0.55}, B=@n65), B=@n64.RGB)
  n79 = NamedRerouteUsage()
  n80 = TextureCoordinate()
  NamedRerouteDeclaration(None=LinearInterpolate(A=Add(A=@n74, B=@n75), B=Add(A=@n78, B=@n79), Alpha=ComponentMask(None=@n80) {g=true})) {name=WPO, nodeColor=<rgba 0.1158 0.2917 0.1012 1>}
  NamedRerouteDeclaration(None=@n62) {name=Rand Vert R, nodeColor=<rgba 1 0.0722 0.1272 1>}
  ScalarParameter() {defaultValue=3}
  n81 = MaterialFunctionCall!(Distance=ScalarParameter() {defaultValue=200, parameterName=Camera Distance Fade}) {materialFunction=/Game/Effects/Fort_Effects/Materials/_Functions/MF_CameraDistanceFade}
  n82 = Multiply(A=Multiply(A=@n26.A, B=@n43), B=Multiply(A=@n81.Result, B=@n81.Result))
  n83 = DynamicParameter() {defaultValue=<rgba 0 20 0 1>, paramNames [Spark Travel Offset, Depth Fade, Random Disappearance, Spark Length]}
  n84 = NamedRerouteUsage()
  n85 = ParticleRandom()
  n86 = Multiply(A=@n83.Random Disappearance, B=Frac(None=Add(A=@n84, B=@n85)))
  Subtract(A=@n82, B=@n86)
  Subtract(A=Multiply(A=DepthFade(FadeDistance=@n83.Depth Fade), B=@n82), B=@n86)
 ]
}
