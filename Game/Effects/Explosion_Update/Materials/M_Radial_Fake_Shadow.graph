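# Translucent, unlit, two-sided material for Niagara mesh particles that fakes radial shadows in explosion effects.
# Opacity combines Alpha, a UV V gradient and squared camera-distance fade, softened by depth fade.
# Emissive blends two shadow colors by distance from the particle, optionally lerps to particle color, then passes through emissive adjustments.
# An offset vector stretches the low-V end away from a dynamic-parameter XYZ point, scaled by Shadow Length and per-particle randomness.
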Material /Game/Effects/Explosion_Update/Materials/M_Radial_Fake_Shadow
graph {
 properties {
  bAutomaticallySetUsageInEditor=false
  bUsedWithNiagaraMeshParticles=true
  blendMode=BLEND_Translucent
  shadingModel=MSM_Unlit
  twoSided=true
  usedShadingModels=Unlit
 }
 nodes [
  n1 = ScalarParameter() {defaultValue=0.1, parameterName=Alpha}
  n2 = ComponentMask(None=TextureCoordinate()) {g=true}
  n3 = MaterialFunctionCall!(Distance=ScalarParameter() {defaultValue=100, parameterName=Camera Distance Fade}) {materialFunction=/Game/Effects/Fort_Effects/Materials/_Functions/MF_CameraDistanceFade}
  DepthFade(
   Opacity=Multiply(A=Multiply(A=@n1, B=@n2), B=Multiply(A=@n3.Result, B=@n3.Result))
   FadeDistance=ScalarParameter() {defaultValue=20, parameterName=Depth Fade}
  )
  n4 = VectorParameter() {group=Color, parameterName=Shadow Color, defaultValue=<rgba 0.0051 0 0.0156 1>}
  n5 = VectorParameter() {group=Color, parameterName=Shadow Additional Color, defaultValue=<rgba 2 0 0.1964 1>}
  n6 = ScalarParameter() {parameterName=Shadow Color Lerp}
  n7 = Distance(A=ParticlePositionWS(), B=WorldPosition().XYZ)
  n8 = ScalarParameter() {defaultValue=0.0001, parameterName=Shadow Color Lerp Adjustment}
  n9 = OneMinus(None=Saturate(None=Multiply(A=@n7, B=@n8) {constB=0.1}))
  n10 = Saturate(None=Power(Base=Multiply(A=@n6, B=@n9), Exp=ScalarParameter() {defaultValue=1, parameterName=Lerp Balance}))
  MaterialFunctionCall(Emissive=LinearInterpolate(A=LinearInterpolate(A=@n4.RGB, B=@n5.RGB, Alpha=@n10), B=ParticleColor().RGB, Alpha=ScalarParameter() {parameterName=Use Particle Color})) {materialFunction=&MF_FX_EmissiveAdjustments}
  n11 = Add(A=ParticleRandom())
  n12 = DynamicParameter() {parameterIndex=3, defaultValue=<rgba 0 0 0 1>, paramNames [X, Y, Z, Shadow Length]}
  n13 = Subtract(A=WorldPosition().XYZ, B=@n12.RGB)
  Multiply(A=OneMinus(None=@n2), B=Multiply(A=Multiply(A=@n11, B=@n12.Shadow Length), B=Normalize(VectorInput=@n13)))
 ]
}
