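# Niagara system for a grenade launcher explosion (Asteria variant), combining flash, fire, smoke, chunks and sparks.
# Short-lived sprite and mesh flashes accompany curl-noise-driven smoke meshes whose fire and dissolve are animated through dynamic material parameters.
# GPU Grid2D emitters add camera-facing fluid smoke/fire and smoke trails, with instance limiting to contain cost.
# User parameters drive scale, color and heat, while time-of-day scalars adjust brightness and some emitters cull by distance.
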NiagaraSystem /Game/Effects/Explosion_Update/NS_GrenadeLauncher_Explosion_Asteria
dependencies {
 usedDataInterfaces [
  &NiagaraDataInterfaceColorCurve
  /Script/Niagara.NiagaraDataInterfaceGrid2DCollection
  /Script/Niagara.NiagaraDataInterfaceRenderTarget2D
  &NiagaraDataInterfaceParticleRead
  &NiagaraDataInterfaceCamera
  &NiagaraDataInterfacePlatformSet
  &NiagaraDataInterfaceCurve
  /Script/Niagara.NiagaraDataInterfaceEmitterProperties
  &NiagaraDataInterfaceArrayFloat
  &NiagaraDataInterfaceVectorCurve
  &NiagaraDataInterfaceArrayColor
  &NiagaraDataInterfaceMeshRendererInfo
  &NiagaraDataInterfaceVector2DCurve
 ]
 usedDynamicInputs [
  &MakeIntFromFloat
  &FN_LOD_Distance_Remap
  /Niagara/DynamicInputs/Math/Min_Int
  &MakeVector4_FromFloat
 ]
 usedModules [
  &self:SystemSpawnScript.SystemScriptSource.SystemScriptGraph.NiagaraNodeAssignment_1.SetVariables_#1
  &self:SystemSpawnScript.SystemScriptSource.SystemScriptGraph.NiagaraNodeAssignment_2.SetVariables_#2
  &self:SystemSpawnScript.SystemScriptSource.SystemScriptGraph.NiagaraNodeAssignment_3.SetVariables_#3
  &self:SystemSpawnScript.SystemScriptSource.SystemScriptGraph.NiagaraNodeAssignment_4.SetVariables_#4
  &SystemState
  &self:SystemSpawnScript.SystemScriptSource.SystemScriptGraph.NiagaraNodeAssignment_5.SetVariables_#5
 ]
}
emitters {
 Flash_Merged {
  emitterSpawnScript [/SetVariables(
    Emitter.CameraOffset:NiagaraFloat=&Multiply_Float(A=50, B=$User.Global Scale)
    Emitter.Color:LinearColor=<rgba 8 0 3.5772 0.8>
    Emitter.SpriteSize:Vector2f=&Multiply_Vector2D_ByFloat(Float=$User.Global Scale, Vector2D=<1500 20>)
   ) {isSetParameters=true}]
  emitterUpdateScript [
   &EmitterState(
    Enable Distance Culling=true
    Life Cycle Mode=::NewEnumerator1
    Loop Behavior=::NewEnumerator1
    Loop Duration=0.3247
    Max Distance Response=::KillAfterParticlesFinish
    MaxDistance=8500
    Module.bUseMaxDistance=true
    Module.bUseMinDistance=false
    Scalability Mode=::NewEnumerator1
    UseLoopDelay=true
   )
   &SpawnBurst_Instantaneous(Spawn Count=1, Use Spawn Probability=false)
   /SetVariables(
    Emitter.Color2:LinearColor=&MakeLinearColorFromVectorAndFloat(
     "Vector (RGB)"=&Lerp_Vector(
      A=&Multiply_VectorByFloat(
       Vector=<2 1 0>
       Float=&FNTimeOfDayScalar(Night Scale=2, Morning Scale=0.5, Day Scale=0.5, Evening Scale=0.5)
      )
      B=&Multiply_VectorByFloat(
       Vector=<16 5 1>
       Float=&FNTimeOfDayScalar(Night Scale=2, Morning Scale=0.5, Day Scale=0.5, Evening Scale=0.5)
      )
      Alpha=&FloatFromCurve(
       FloatCurve=DataInterface() {propertyValues {Curve {keys [
           {arriveTangent=-1.9247, interpMode=RCIM_Cubic, time=0, value=1}
           {arriveTangent=-1.8467, interpMode=RCIM_Cubic, leaveTangent=-1.8467, time=0.2339, value=0.25}
           {interpMode=RCIM_Cubic, tangentMode=RCTM_User, time=0.5415, value=0}
          ]}}}
       CurveIndex=$Emitter.NormalizedLoopAge
      )
     )
     "Float (Alpha)"=&FloatFromCurve(
      FloatCurve=DataInterface() {propertyValues {Curve {keys [{arriveTangent=-5.1417, interpMode=RCIM_Cubic, time=0.4976, value=1}, <at 0.7576 0 RCIM_Cubic>]}}}
      CurveIndex=$Emitter.NormalizedLoopAge
      Scale Curve=0.05
     )
    )
    Emitter.Scale:Vector3f=&VectorFromCurve(
     VectorCurve=DataInterface() {propertyValues {
       XCurve {keys [
         {arriveTangent=4.1785, interpMode=RCIM_Cubic, leaveTangent=4.1785, tangentMode=RCTM_User, time=0, value=0}
         {arriveTangent=0.9995, interpMode=RCIM_Cubic, leaveTangent=0.9995, tangentMode=RCTM_User, time=0.1905, value=0.6825}
         {arriveTangent=-0.0405, interpMode=RCIM_Cubic, leaveTangent=-0.0405, tangentMode=RCTM_User, time=0.6079, value=1.575}
        ]}
       YCurve {keys [
         {arriveTangent=4.1785, interpMode=RCIM_Cubic, leaveTangent=4.1785, tangentMode=RCTM_User, time=0, value=0}
         {arriveTangent=0.9995, interpMode=RCIM_Cubic, leaveTangent=0.9995, tangentMode=RCTM_User, time=0.1905, value=0.6825}
         {arriveTangent=-0.0405, interpMode=RCIM_Cubic, leaveTangent=-0.0405, tangentMode=RCTM_User, time=0.6079, value=1.575}
        ]}
       ZCurve {keys [
         {arriveTangent=4.1785, interpMode=RCIM_Cubic, leaveTangent=4.1785, tangentMode=RCTM_User, time=0, value=0}
         {arriveTangent=0.9995, interpMode=RCIM_Cubic, leaveTangent=0.9995, tangentMode=RCTM_User, time=0.1905, value=0.6825}
         {arriveTangent=-0.0405, interpMode=RCIM_Cubic, leaveTangent=-0.0405, tangentMode=RCTM_User, time=0.6079, value=1.575}
        ]}
      }}
     CurveIndex=$Emitter.NormalizedLoopAge
     Scale Curve=&Multiply_VectorByFloat(Vector=<2 2 2>, Float=$User.Global Scale)
    )
    Emitter.SpriteSize:Vector2f=&Multiply_Vector2D_ByFloat(
     Float=&FloatFromCurve(
      FloatCurve=DataInterface() {propertyValues {Curve {keys [
          {arriveTangent=9.901, interpMode=RCIM_Cubic, leaveTangent=9.901, tangentMode=RCTM_User, time=0.45, value=0}
          {arriveTangent=12.2238, interpMode=RCIM_Cubic, leaveTangent=12.2238, tangentMode=RCTM_User, time=0.7, value=3}
          <at 0.9 0>
         ]}}}
      CurveIndex=$Emitter.NormalizedLoopAge
     )
     Vector2D=$Emitter.Initial.SpriteSize
    )
   ) {isSetParameters=true}
  ]
  particleSpawnScript [
   &InitializeParticle(
    Color=<rgba 0.2479 1 0.2583 0.2>
    Lifetime=0.3
    Mesh Scale=&Multiply_VectorByFloat(
     Vector=&VectorFromFloat(Value=&Lerp_Float(A=1, B=0.9, Alpha=&MakeFloatFromInt(Int=&SpawnGroup())))
     Float=$User.Global Scale
    )
    Mesh Scale Mode=::NewEnumerator4
    Sprite Size Mode=::NewEnumerator3
    Uniform Sprite Size=&Multiply_Float(A=500, B=&Multiply_Float(A=$User.Global Scale, B=$System.DistanceScale))
    UsePositionOffset=false
    Write Material Random=true
   ) {module=InitializeParticle001}
   &CameraOffset(Camera Offset Amount=&Multiply_Float(A=15, B=$User.Global Scale))
   /SetVariables(Particles.CamOfs2=80) {isSetParameters=true}
  ]
  particleUpdateScript [
   &ParticleState()
   &ScaleColor(
    Scale Alpha=&FloatFromCurve(FloatCurve=DataInterface() {propertyValues {Curve {keys [{arriveTangent=-3.5509, interpMode=RCIM_Cubic, leaveTangent=-3.5509, tangentMode=RCTM_User, time=0, value=1}, <at 1 0>]}}})
    Scale RGB=&VectorFromCurve(
     VectorCurve=DataInterface() {propertyValues {XCurve {keys [<at 0 1>,<at 1 0.1>]}, YCurve {keys [<at 0 1>,<at 1 0.1>]}, ZCurve {keys [<at 0 1>,<at 1 0.1>]}}}
     Scale Curve=&Multiply_VectorByFloat(Vector=<5 3 2>, Float=&FNTimeOfDayScalar(Night Scale=2, Morning Scale=1, Evening Scale=1))
    )
    ScaleA=true
    ScaleRGB=true
   )
   &SpriteRotationRate(
    Delta Time:NiagaraFloat=&FloatFromCurve(FloatCurve=DataInterface() {propertyValues {Curve {keys [
         {tangentMode=RCTM_User, time=0, value=1}
         {arriveTangent=-0.3174, interpMode=RCIM_Cubic, leaveTangent=-0.3174, tangentMode=RCTM_User, time=0.1832, value=1}
         {arriveTangent=0.0123, interpMode=RCIM_Cubic, leaveTangent=0.0123, tangentMode=RCTM_User, time=0.7355, value=0}
        ]}}})
    Rotation Rate=-5
   )
   &ScaleSpriteSize(
    Uniform Curve Scale=3
    Uniform Curve Sprite Scale=DataInterface() {propertyValues {Curve {keys [
        {arriveTangent=9.0725, interpMode=RCIM_Cubic, leaveTangent=9.0725, tangentMode=RCTM_User, time=0, value=0}
        {arriveTangent=-2.5249, interpMode=RCIM_Cubic, leaveTangent=-2.5249, tangentMode=RCTM_User, time=0.1003, value=0.9847}
        {arriveTangent=-0.0405, interpMode=RCIM_Cubic, leaveTangent=-0.0405, tangentMode=RCTM_User, time=0.6973, value=0}
       ]}}}
   )
   &ScaleMeshSize(Scale Factor=&VectorFromFloat(Value=&FloatFromCurve(
      FloatCurve=DataInterface() {propertyValues {Curve {keys [
          {arriveTangent=4.1785, interpMode=RCIM_Cubic, leaveTangent=4.1785, tangentMode=RCTM_User, time=0.0247, value=0}
          {arriveTangent=1.7203, interpMode=RCIM_Cubic, leaveTangent=1.7203, tangentMode=RCTM_User, time=0.2, value=0.5}
          <at 0.35 0.95>
          {arriveTangent=-0.0405, interpMode=RCIM_Cubic, leaveTangent=-0.0405, tangentMode=RCTM_User, time=0.55, value=0}
         ]}}}
      Scale Curve=&Multiply_Float(A=4, B=$User.Global Scale)
     )))
   /SetVariables(
    Particles.Color2=&ColorFromCurve(ColorCurve=DataInterface() {propertyValues {
       AlphaCurve {keys [<at 0 0.16 RCIM_Cubic>,<at 1 0 RCIM_Cubic>]}
       BlueCurve {keys [<at 0 0.5023 RCIM_Cubic>,<at 1 1 RCIM_Cubic>]}
       GreenCurve {keys [<at 0 5.4862 RCIM_Cubic>,<at 1 1 RCIM_Cubic>]}
       RedCurve {keys [<at 0 12 RCIM_Cubic>,<at 1 1 RCIM_Cubic>]}
      }})
    Particles.Color3:LinearColor=&MakeLinearColorFromVectorAndFloat(
     "Vector (RGB)"=&MakeVectorFromLinearColorRGB(Color=$User.Global Color)
     "Float (Alpha)"=&FloatFromCurve(FloatCurve=DataInterface() {propertyValues {Curve {keys [
          {interpMode=RCIM_Cubic, tangentMode=RCTM_User, time=-0.0046, value=0.1}
          {arriveTangent=-1.0036, interpMode=RCIM_Cubic, leaveTangent=-1.0036, tangentMode=RCTM_User, time=0.6, value=0.9}
          {arriveTangent=-2.4196, interpMode=RCIM_Cubic, leaveTangent=-2.4196, tangentMode=RCTM_User, time=1.0011, value=0.002}
         ]}}})
    )
   ) {isSetParameters=true}
  ]
  rendererClasses [&NiagaraSpriteRendererProperties, &NiagaraMeshRendererProperties]
  renderers [
   {rendererClass=&NiagaraSpriteRendererProperties}
   {rendererClass=&NiagaraMeshRendererProperties, rendererIndex=1}
   {rendererClass=&NiagaraMeshRendererProperties, rendererIndex=2}
   {rendererClass=&NiagaraSpriteRendererProperties, rendererIndex=3}
  ]
 }
 Grid2D_Gas_SmokeFire_Emitter {
  simTarget=GPUComputeSim
  emitterSpawnScript [
   /NiagaraFluids/Functions/General/Fluids_Gas_Manager(
    Density Bouyancy=-0.1
    Density Color Curve=DataInterface() {propertyValues {
      ExposedName=Curve Density
      bExposeCurve=true
      AlphaCurve {keys [<at 0 0 RCIM_Cubic>,<at 2.7618 1>]}
      BlueCurve {keys [<at 0.0245 0.045>,<at 4.4467 0.055>]}
      GreenCurve {keys [<at 0.0245 0.0307>,<at 4.4467 0.0375>]}
      RedCurve {keys [<at 0.0245 0.0266>,<at 4.4467 0.0325>]}
     }}
    Depth Fade=25
    Dissipation Temperature=5
    Dissipation Velocity=0.5
    EnvironmentColor:LinearColor=<rgba 0.1427 0.2031 0.3542 1>
    Pressure Relaxation=1
    Radius Scale=0.8
    Shadowing=0.25
    SunColor:LinearColor=<rgba 1.2417 1.53 2 1>
    Temperature Bouyancy=0.0025
    Temperature Color Curve=DataInterface() {propertyValues {
      ExposedName=Curve Temperature
      bExposeCurve=true
      AlphaCurve {keys [<at 0 0>,<at 0.4853 1>]}
      BlueCurve {keys [<at 0.0696 0>,<at 0.8304 0.0517>,<at 0.9917 0.0783>]}
      GreenCurve {keys [<at 0.0696 0>,<at 0.8304 0.729>,<at 0.9917 0.2349>]}
      RedCurve {keys [<at 0.0696 0>,<at 0.8304 8>,<at 0.9917 0.9401>]}
     }}
    Temperature Maximum:NiagaraFloat=&Multiply_Float(A=1200, B=$User.Global Scale)
    Temperature Scale=0.2
    Vorticity Confinement=0.1
   )
   /SetVariables(Emitter.SimGrid=DataInterface() {propertyValues {
      ClearBeforeNonIterationStage=true
      NumAttributes=0
      NumCellsMaxAxis=3
      NumCellsX=1
      NumCellsY=1
      OverrideBufferFormat=Float
      PreviewAttribute=None
      SetGridFromMaxAxis=false
      bOverrideFormat=false
      bPreviewGrid=false
      RenderTargetUserParameter {parameter {name=None, typeDefHandle {registeredTypeIndex=1}, varData []}}
      WorldBBoxSize=<1200 1200>
     }}) {isSetParameters=true}
   /SetVariables(Emitter.LightingGrid=DataInterface() {propertyValues {
      ClearBeforeNonIterationStage=true
      NumAttributes=3
      NumCellsMaxAxis=3
      NumCellsX=512
      NumCellsY=512
      OverrideBufferFormat=Float
      PreviewAttribute=None
      SetGridFromMaxAxis=false
      bOverrideFormat=false
      bPreviewGrid=false
      RenderTargetUserParameter {parameter {name=None, typeDefHandle {registeredTypeIndex=4}, varData []}}
      WorldBBoxSize=<100 100>
     }}) {isSetParameters=true}
   /SetVariables(Emitter.LightingGridDownsample=DataInterface() {propertyValues {
      ClearBeforeNonIterationStage=true
      NumAttributes=1
      NumCellsMaxAxis=3
      NumCellsX=64
      NumCellsY=64
      OverrideBufferFormat=Float
      PreviewAttribute=None
      SetGridFromMaxAxis=false
      bOverrideFormat=false
      bPreviewGrid=false
      RenderTargetUserParameter {parameter {name=None, typeDefHandle {registeredTypeIndex=4}, varData []}}
      WorldBBoxSize=<100 100>
     }}) {isSetParameters=true}
   /SetVariables() {isSetParameters=true}
   /SetVariables() {isSetParameters=true}
   /SetVariables() {isSetParameters=true}
   /NiagaraFluids/Modules/Grid2D/Grid2D_SetResolution(
    Grid:NiagaraDataInterfaceGrid2DCollection=$Emitter.SimGrid
    NumCellsMaxAxis=320
    Resolution Method:ENiagaraGrid2DResolution=::NewEnumerator1
    WorldGridExtents:Vector2f=&Multiply_Vector2D_ByFloat(Float=$User.Global Scale, Vector2D=<1280 1280>)
   )
   &Grid2D_SetResolution(
    Grid:NiagaraDataInterfaceGrid2DCollection=$Emitter.LightingGridDownsample
    OtherGrid:NiagaraDataInterfaceGrid2DCollection=$Emitter.SimGrid
    Resolution Method:ENiagaraGrid2DResolution=::NewEnumerator3
    Resolution Mult=0.2
    WorldGridExtents:Vector2f=$Emitter.Grid2D_SetResolution.WorldGridExtents
   ) {module=Grid2D_SetResolution003}
   &Grid2D_SetResolution(
    Grid:NiagaraDataInterfaceGrid2DCollection=$Emitter.LightingGrid
    OtherGrid:NiagaraDataInterfaceGrid2DCollection=$Emitter.SimGrid
    Resolution Method:ENiagaraGrid2DResolution=::NewEnumerator3
    WorldGridExtents:Vector2f=$Emitter.Grid2D_SetResolution.WorldGridExtents
   ) {module=Grid2D_SetResolution004}
   &Grid2D_SetResolution(
    Grid:NiagaraDataInterfaceGrid2DCollection=$Emitter.SourceGrid
    OtherGrid:NiagaraDataInterfaceGrid2DCollection=$Emitter.SimGrid
    Resolution Method:ENiagaraGrid2DResolution=::NewEnumerator3
    WorldGridExtents:Vector2f=$Emitter.Grid2D_SetResolution.WorldGridExtents
   ) {module=Grid2D_SetResolution005}
   /SetVariables(Emitter.g:Vector3f=<0 0 -980>) {isSetParameters=true}
   &Grid2D_SetResolution(
    Grid:NiagaraDataInterfaceGrid2DCollection=$Emitter.PressureGrid
    OtherGrid:NiagaraDataInterfaceGrid2DCollection=$Emitter.SimGrid
    Resolution Method:ENiagaraGrid2DResolution=::NewEnumerator3
    WorldGridExtents:Vector2f=$Emitter.Grid2D_SetResolution.WorldGridExtents
   ) {module=Grid2D_SetResolution001}
   &Grid2D_SetResolution(
    Grid:NiagaraDataInterfaceGrid2DCollection=$Emitter.TemporaryGrid
    OtherGrid:NiagaraDataInterfaceGrid2DCollection=$Emitter.SimGrid
    Resolution Method:ENiagaraGrid2DResolution=::NewEnumerator3
    WorldGridExtents:Vector2f=$Emitter.Grid2D_SetResolution.WorldGridExtents
   ) {module=Grid2D_SetResolution002}
   /SetVariables() {isSetParameters=true}
   /SetVariables(Emitter.dt:NiagaraFloat=$Engine.DeltaTime) {enabled=false, isSetParameters=true}
   /SetVariables(Emitter.ParticleSource=DataInterface() {propertyValues {EmitterBinding {bindingMode=Other, emitterName=SourceParticles}}}) {isSetParameters=true}
   /SetVariables() {isSetParameters=true}
   &self:Grid2D_Gas_SmokeFire_Emitter.NiagaraScratchPadContainer_2.SetLightingRT_0(LightingGrid:NiagaraDataInterfaceGrid2DCollection=$Emitter.LightingGrid, LightingRT:NiagaraDataInterfaceRenderTarget2D=$Emitter.LightingRT) {module=SetLightingRT_29}
   /SetVariables() {enabled=false, isSetParameters=true}
   /SetVariables(Emitter.density=1) {isSetParameters=true}
   /NiagaraFluids/Modules/Grid2D/Grid2D_UpdateSimAttributes(Grid:NiagaraDataInterfaceGrid2DCollection=$Emitter.SimGrid, WorldGridSize:Vector2f=$Emitter.Grid2D_SetResolution.WorldGridExtents) {module=UpdateSimAttributes}
   /SetVariables() {isSetParameters=true}
   &self:Grid2D_Gas_SmokeFire_Emitter.NiagaraScratchPadContainer_2.ScratchModule_02_1(New Grid2D Collection:NiagaraDataInterfaceGrid2DCollection=$Emitter.SimGrid) {module=ScratchModule_02001}
   /SetVariables(Emitter.PressureIterations=5) {isSetParameters=true}
   /NiagaraFluids/Modules/Grid2D/Grid2D_InitRT(Grid:NiagaraDataInterfaceGrid2DCollection=$Emitter.SimGrid, SimRT:NiagaraDataInterfaceRenderTarget2D=$Emitter.SimRT) {module=Grid2D_Gas_InitRT}
   &self:Grid2D_Gas_SmokeFire_Emitter.NiagaraScratchPadContainer_2.ScratchModule_0(
    Grid:NiagaraDataInterfaceGrid2DCollection=$Emitter.SimGrid
    PressureGrid:NiagaraDataInterfaceGrid2DCollection=$Emitter.PressureGrid
    TemporaryGrid:NiagaraDataInterfaceGrid2DCollection=$Emitter.TemporaryGrid
   ) {module=ScratchModule}
   /SetVariables(Emitter.GridsAllocated=true) {isSetParameters=true}
  ]
  emitterUpdateScript [
   &self:Emitter_Limitation_and_Culling__01(Max Num=3, Platform Set=DataInterface() {propertyValues {Platforms {qualityLevelMask=-17, cVarConditions [], deviceProfileStates []}}}) {module=Emitter_Limitation_and_Culling__01}
   &EmitterState(
    Inactive Response=::NewEnumerator1
    Life Cycle Mode=::NewEnumerator1
    Loop Behavior=::NewEnumerator1
    Loop Duration=/Niagara/DynamicInputs/Math/Min_Float(
     A:NiagaraFloat=&Multiply_Float(A=2.5, B=&Max_Float(A:NiagaraFloat=$User.Global Scale, B:NiagaraFloat=0.4))
     B:NiagaraFloat=2.5
    )
    Scalability Mode=::NewEnumerator1
    UseLoopDelay=false
   )
   /SetVariables(Emitter.dt:NiagaraFloat=$Engine.DeltaTime) {isSetParameters=true}
   /NiagaraFluids/Modules/Grid2D/Grid2D_CreateUnitToWorldTransform(
    Camera:NiagaraDataInterfaceCamera=$Emitter.CameraQuery
    Camera Facing=true
    Offset:NiagaraPosition=$Engine.Owner.Position
    Rotation:Quat4f=$Engine.Owner.Rotation
    WorldGridExtents:Vector2f=$Emitter.Grid2D_SetResolution.WorldGridExtents
   )
   /SetVariables(
    Emitter.Position:NiagaraPosition=&ConvertVectorToPosition(
     Input Position=$Emitter.Grid2D_CreateUnitToWorldTransform.LocalToWorldTranslation
     Conversion Operation:ENiagara_LWCConvertVecToPos=::NewEnumerator0
    )
    Emitter.float:NiagaraFloat=&FloatFromCurve(
     FloatCurve=DataInterface() {propertyValues {Curve {
        defaultValue=1
        keys [{arriveTangent=2.2215, interpMode=RCIM_Cubic, leaveTangent=2.2215, tangentMode=RCTM_User, time=0.5, value=0}, <at 1 2 RCIM_Cubic>]
       }}}
     CurveIndex=$Emitter.NormalizedLoopAge
    )
   ) {isSetParameters=true}
   /NiagaraFluids/Modules/Grid2D/Grid2D_SetBounds(
    CameraFacing:NiagaraBool=$Emitter.Grid2D_CreateUnitToWorldTransform.CameraFacing
    EmitterProperties=DataInterface() {propertyValues {EmitterBinding {bindingMode=Self, emitterName=None}}}
    LocalPivot:Vector2f=$Emitter.Grid2D_CreateUnitToWorldTransform.LocalPivot
    Rotate for Camera Facing=true
    WorldGridExtents:Vector2f=$Emitter.Grid2D_SetResolution.WorldGridExtents
   )
   /SetVariables(Emitter.ExecutionState:ENiagaraExecutionState=$Emitter.ENiagaraExecutionState) {isSetParameters=true}
   &self:ScratchModule_02(PerformDeallocation:NiagaraBool=/Niagara/DynamicInputs/Scalability/IsEmitterComplete()) {module=ScratchModule_02002}
  ]
  rendererClasses [&NiagaraSpriteRendererProperties]
  renderers [
   {rendererClass=&NiagaraSpriteRendererProperties}
   {rendererClass=&NiagaraSpriteRendererProperties, rendererIndex=1}
  ]
 }
 LW_Dark_Chunks {rendererClasses [], renderers []}
 LW_Explosion_Main {rendererClasses [], renderers []}
 LW_Smoke_Ground {rendererClasses [], renderers []}
 Mesh_Sparks001 {rendererClasses [], renderers []}
 NE_Explosion_Main {
  emitterSpawnScript [
   /SetVariables(Emitter.Global Color:LinearColor=$User.Global Color, Emitter.Global Scale:NiagaraFloat=$User.Global Scale) {isSetParameters=true}
   /SetVariables(
    Emitter.Color2:LinearColor=&MultiplyLinearColor(
     Linear Color=$Emitter.Global Color
     Scale Factor=&MakeVector4_FromLinearColor(LinearColor=$User.Additional Fire Color)
    )
    Emitter.Depth Fade:NiagaraFloat=&Multiply_Float(A=125, B=$Emitter.Global Scale)
    Emitter.Dissolve Noise Power=1.1
    Emitter.Dissolve Noise Scale:LinearColor=&MultiplyLinearColor(
     Linear Color=/Niagara/DynamicInputs/TypeConversions/MakeLinearColorFromFloat(Value:NiagaraFloat=&FloatFromCurve(FloatCurve=DataInterface() {propertyValues {Curve {keys [<at 0 2>,<at 1 1>,<at 2 0.5>]}}}, CurveIndex=$Emitter.Global Scale))
     Scale Factor=<0.0008 0.0008 0.0009 1>
    )
    Emitter.Near Camera Fade:NiagaraFloat=&Multiply_Float(A=400, B=$Emitter.Global Scale)
   ) {isSetParameters=true}
   /SetVariables(Emitter.Dissolve Scale Adjsutment:NiagaraFloat=&FloatFromCurve(
     FloatCurve=DataInterface() {propertyValues {Curve {keys [<at 0 3>,<at 1 1>,<at 5 0.5>,<at 20 0.4>,<at 100 0.4>]}}}
     CurveIndex=$Emitter.Global Scale
     Scale Curve=$Emitter.Global Scale
    )) {isSetParameters=true}
  ]
  emitterUpdateScript [
   &EmitterState(Life Cycle Mode=::NewEnumerator1, Loop Behavior=::NewEnumerator1, Loop Delay=0.04, Loop Duration=0.627, UseLoopDelay=true)
   &SpawnBurst_Instantaneous(Spawn Group=1, Use Spawn Probability=false, UseLoopCountLimit=false) {module=SpawnBurst_Instantaneous002}
   &SpawnBurst_Instantaneous(Spawn Count=6, Spawn Time=0.01, Use Spawn Probability=false, UseLoopCountLimit=false) {module=SpawnBurst_Instantaneous001}
  ]
  particleSpawnScript [
   &InitializeParticle(
    Color=$User.Explosion Smoke Color
    Lifetime=&MaskFloatBySpawnGroup(
     Masked Float=2.1
     Passthrough Float=&SelectFloatFromArray(
      Array Sampling Mode=::NewEnumerator0
      Direct Array Index=?
      Float Selection Array=DataInterface() {propertyValues {FloatData [1.5,1.95]}}
      Random Seed=0
     )
     Spawn Group Mask=1
    )
    Mesh Scale=&Multiply_VectorByFloat(
     Vector=&MaskVectorBySpawnGroup(
      Masked Vector=<3 3.5 2.6>
      Passthrough Vector=&RandomRangeVector(Minimum=<3 2.8 2.8>, Maximum=<3 3.5 3.5>)
      Spawn Group Mask=1
     )
     Float=&MaskFloatBySpawnGroup(
      Masked Float=1
      Passthrough Float=&FloatFromCurve(
       FloatCurve=DataInterface() {propertyValues {Curve {keys [<at 0 0.9 RCIM_Constant>,<at 0.5 1 RCIM_Constant>,<at 1 1>]}}}
       CurveIndex=&RandomRangeFloat(Minimum=0, Maximum=1)
      )
      Spawn Group Mask=1
     )
    )
    Mesh Scale Mode=::NewEnumerator4
    UsePositionOffset=false
    Write Material Random=true
   )
   &ShapeLocation(
    MaskSpawnGroup=false
    Override Seed=false
    Sphere Radius=&Multiply_Float(A=100, B=$Emitter.Global Scale)
    UseOwnerScale=true
   )
   &AddVelocity(
    Module.Use Falloff Distance=false
    Origin Offset=&Multiply_VectorByFloat(Vector=&RandomRangeVector(Minimum=<0 0 -10>, Maximum=<0 0 -10>), Float=$Emitter.Global Scale)
    Override Seed=false
    Velocity Mode=::NewEnumerator1
    Velocity Speed=&MaskFloatBySpawnGroup(
     Masked Float=0
     Passthrough Float=&Multiply_Float(A=&RandomRangeFloat(Minimum=900, Maximum=1100), B=$Emitter.Global Scale)
     Spawn Group Mask=1
    )
   )
   &CameraOffset(Camera Offset Amount=&Multiply_Float(A=&MaskFloatBySpawnGroup(Masked Float=70, Passthrough Float=40, Spawn Group Mask=1), B=$Emitter.Global Scale))
   /SetVariables(Particles.Scale=&Multiply_VectorByFloat(Vector=$Particles.Scale, Float=$Emitter.Global Scale)) {isSetParameters=true}
  ]
  particleUpdateScript [
   &ParticleState()
   &GravityForce(Gravity=&Multiply_VectorByFloat(
     Vector=&MaskVectorBySpawnGroup(
      Masked Vector=<0 0 1600>
      Passthrough Vector=&RandomRangeVector(Minimum=<0 0 500>, Maximum=<1 1 1500>)
      Spawn Group Mask=1
     )
     Float=$Emitter.Global Scale
    ))
   &Drag(Drag=&RandomRangeFloat(Minimum=9, Maximum=10), Use Linear Drag=true, Use Rotational Drag=false)
   &DynamicMaterialParameters(
    Index 0 Param 1=&FloatFromCurve(
     FloatCurve=DataInterface() {propertyValues {Curve {keys [{arriveTangent=0.9809, interpMode=RCIM_Cubic, leaveTangent=0.9809, tangentMode=RCTM_User, time=0.2, value=0.8}, <at 1 0.5>]}}}
     Scale Curve=&Multiply_Float(A=-150, B=$Emitter.Global Scale)
    )
    Index 0 Param 2=&FloatFromCurve(
     FloatCurve=DataInterface() {propertyValues {Curve {keys [
         {tangentMode=RCTM_User, time=0, value=1}
         {arriveTangent=-0.0405, interpMode=RCIM_Cubic, leaveTangent=-0.0405, tangentMode=RCTM_User, time=0.2, value=1}
         {arriveTangent=-0.0002, interpMode=RCIM_Cubic, leaveTangent=-0.0002, tangentMode=RCTM_User, time=0.9, value=0.6}
        ]}}}
     Scale Curve=0.8
    )
    Index 0 Param 3=&FloatFromCurve(
     FloatCurve=DataInterface() {propertyValues {Curve {keys [
         {arriveTangent=1.4725, interpMode=RCIM_Cubic, leaveTangent=1.4725, tangentMode=RCTM_User, time=0, value=0}
         {tangentMode=RCTM_User, time=1, value=1}
        ]}}}
     Scale Curve=&Multiply_Float(
      A=&MaskFloatBySpawnGroup(Masked Float=250, Passthrough Float=&RandomRangeFloat(Minimum=250, Maximum=300), Spawn Group Mask=1)
      B=$Emitter.Global Scale
     )
    )
    Index 0 Param 4=&MaskFloatBySpawnGroup(
     Masked Float=&FloatFromCurve(
      FloatCurve=DataInterface() {propertyValues {Curve {keys [
          {arriveTangent=0.1175, interpMode=RCIM_Cubic, leaveTangent=0.6205, tangentMode=RCTM_Break, time=0.22, value=-2}
          {arriveTangent=6.3488, interpMode=RCIM_Cubic, leaveTangent=6.3488, tangentMode=RCTM_User, time=0.5987, value=-0.8048}
          {arriveTangent=42.4007, interpMode=RCIM_Cubic, leaveTangent=14.4841, tangentMode=RCTM_Break, time=1, value=7.3}
         ]}}}
      Scale Curve=0.2
     )
     Passthrough Float=&FloatFromCurve(
      FloatCurve=DataInterface() {propertyValues {Curve {keys [
          {arriveTangent=0.1175, interpMode=RCIM_Cubic, leaveTangent=1.7315, tangentMode=RCTM_Break, time=0.1, value=-3}
          {arriveTangent=35.5342, interpMode=RCIM_Cubic, leaveTangent=14.4841, tangentMode=RCTM_Break, time=1, value=9}
         ]}}}
      Scale Curve=0.2
     )
     Spawn Group Mask=1
    )
    Index 1 Param 1=&FloatFromCurve(
     FloatCurve=DataInterface() {propertyValues {Curve {keys [
         {arriveTangent=0.5348, interpMode=RCIM_Cubic, leaveTangent=0.5348, tangentMode=RCTM_User, time=0, value=0.01}
         {arriveTangent=1.457, interpMode=RCIM_Cubic, leaveTangent=1.457, tangentMode=RCTM_User, time=0.04, value=0.05}
         {arriveTangent=49.6264, interpMode=RCIM_Cubic, leaveTangent=49.6264, tangentMode=RCTM_User, time=0.08, value=16}
        ]}}}
     Scale Curve=&MaskFloatBySpawnGroup(Masked Float=30, Passthrough Float=&RandomRangeFloat(Minimum=85, Maximum=150), Spawn Group Mask=1)
    )
    Index 1 Param 2=&Divide_Float(
     A=&MaskFloatBySpawnGroup(
      Masked Float=0.014
      Passthrough Float=&RandomRangeFloat(Minimum=0.012, Maximum=0.014)
      Spawn Group Mask=1
     )
     B=$Emitter.Global Scale
    )
    Index 1 Param 3=&FloatFromCurve(
     FloatCurve=DataInterface() {propertyValues {Curve {keys [{interpMode=RCIM_Cubic, tangentMode=RCTM_User, time=0.4, value=0.5}, <at 0.8 0.1>]}}}
     Scale Curve=5
    )
    Index 1 Param 4=&FloatFromCurve(FloatCurve=DataInterface() {propertyValues {Curve {keys [
         {interpMode=RCIM_Cubic, tangentMode=RCTM_User, time=0.05, value=1}
         {arriveTangent=6.745, interpMode=RCIM_Cubic, leaveTangent=6.745, tangentMode=RCTM_User, time=1, value=6}
        ]}}})
    Index 2 Param 1=&FloatFromCurve(
     FloatCurve=DataInterface() {propertyValues {Curve {keys [
         {arriveTangent=0.2023, interpMode=RCIM_Cubic, leaveTangent=0.2023, tangentMode=RCTM_User, time=0.3, value=0.1}
         {arriveTangent=0.8132, interpMode=RCIM_Cubic, leaveTangent=0.8132, tangentMode=RCTM_User, time=1, value=0.6}
        ]}}}
     Scale Curve=&Divide_Float(A=1.5, B=$Emitter.Global Scale)
    )
    Index 2 Param 2=&FloatFromCurve(
     FloatCurve=DataInterface() {propertyValues {Curve {keys [
         {interpMode=RCIM_Cubic, tangentMode=RCTM_User, time=0.3, value=0}
         {arriveTangent=2.4333, interpMode=RCIM_Cubic, leaveTangent=2.4333, tangentMode=RCTM_User, time=1, value=1}
        ]}}}
     Scale Curve=&Multiply_Float(A=0.1, B=$Emitter.Global Scale)
    )
    Index 2 Param 3=&MaskFloatBySpawnGroup(
     Masked Float=&FloatFromCurve(
      FloatCurve=DataInterface() {propertyValues {Curve {keys [
          {arriveTangent=-1.9023, interpMode=RCIM_Cubic, leaveTangent=-1.9023, tangentMode=RCTM_User, time=0, value=1}
          {arriveTangent=-0.7527, interpMode=RCIM_Cubic, leaveTangent=-0.7527, tangentMode=RCTM_User, time=0.1, value=0}
         ]}}}
      Scale Curve=&Multiply_Float(A=150, B=$User.Fire Intensity)
     )
     Passthrough Float=&FloatFromCurve(
      FloatCurve=DataInterface() {propertyValues {Curve {keys [
          {arriveTangent=-4.2868, interpMode=RCIM_Cubic, leaveTangent=-4.2868, tangentMode=RCTM_User, time=0, value=1}
          {arriveTangent=-0.7527, interpMode=RCIM_Cubic, leaveTangent=-0.7527, tangentMode=RCTM_User, time=0.12, value=0}
         ]}}}
      CurveIndex=&Multiply_Float(A=$Particles.NormalizedAge, B=$User.Heat Dissipation)
      Scale Curve=&Multiply_Float(A=100, B=$User.Fire Intensity)
     )
     Spawn Group Mask=1
    )
    Index 2 Param 4=$Emitter.Global Scale
    Index 3 Param 1=&FloatFromCurve(
     FloatCurve=DataInterface() {propertyValues {Curve {keys [
         {arriveTangent=2.2353, interpMode=RCIM_Cubic, leaveTangent=2.2353, tangentMode=RCTM_User, time=0, value=0.2}
         {arriveTangent=1.4416, interpMode=RCIM_Cubic, leaveTangent=1.4416, tangentMode=RCTM_User, time=0.2, value=0.5}
         {interpMode=RCIM_Cubic, tangentMode=RCTM_User, time=0.8, value=1.25}
        ]}}}
     Scale Curve=&Multiply_Float(A=-2.5, B=$Emitter.Global Scale)
    )
    Index 3 Param 2=&FloatFromCurve(
     FloatCurve=DataInterface() {propertyValues {Curve {keys [
         {arriveTangent=5.1563, interpMode=RCIM_Cubic, leaveTangent=5.1563, tangentMode=RCTM_User, time=0, value=0}
         {arriveTangent=0.1309, interpMode=RCIM_Cubic, leaveTangent=0.1309, tangentMode=RCTM_User, time=0.1699, value=0.4}
         {arriveTangent=0.8545, interpMode=RCIM_Cubic, leaveTangent=0.8545, tangentMode=RCTM_User, time=1, value=1}
        ]}}}
     Scale Curve=&Multiply_Float(A=-0.4, B=$Emitter.Global Scale)
    )
    Index 3 Param 3=&FloatFromCurve(
     FloatCurve=DataInterface() {propertyValues {Curve {keys [
         {interpMode=RCIM_Cubic, tangentMode=RCTM_User, time=0.0014, value=200}
         {arriveTangent=-36.1009, interpMode=RCIM_Cubic, leaveTangent=-36.1009, tangentMode=RCTM_User, time=0.1, value=75}
         {arriveTangent=-5.9757, interpMode=RCIM_Cubic, leaveTangent=-5.9757, tangentMode=RCTM_User, time=0.15, value=1.25}
        ]}}}
     Scale Curve=&Multiply_Float(
      A=&FNTimeOfDayScalar(Night Scale=0.6, Morning Scale=0.6, Evening Scale=0.6)
      B=$User.Temperature Scale
     )
    )
    Index 3 Param 4=&FloatFromCurve(
     FloatCurve=DataInterface() {propertyValues {Curve {keys [{tangentMode=RCTM_User, time=0, value=0}, {tangentMode=RCTM_User, time=1, value=1}]}}}
     Scale Curve=&Multiply_Float(A=-100, B=$Emitter.Global Scale)
    )
    Module.Index0Param0Write=true
    Module.Index0Param1Write=true
    Module.Index0Param2Write=true
    Module.Index0Param3Write=true
    Module.Index1Param0Write=true
    Module.Index1Param1Write=true
    Module.Index1Param2Write=true
    Module.Index1Param3Write=true
    Module.Index2Param0Write=true
    Module.Index2Param1Write=true
    Module.Index2Param2Write=true
    Module.Index2Param3Write=true
    Module.Index3Param0Write=true
    Module.Index3Param1Write=true
    Module.Index3Param2Write=true
    Module.Index3Param3Write=true
    Write Parameter Index 1=true
    Write Parameter Index 2=true
    Write Parameter Index 3=true
   ) {module=DynamicMaterialParameters001}
   &ScaleColor(
    Linear Color Curve=DataInterface() {propertyValues {
      AlphaCurve {keys [<at 0 1>,<at 0.6968 1>,<at 0.9894 0>]}
      BlueCurve {keys [<at 0.1888 0.6042>,<at 0.4061 0.4497>,<at 0.9329 0.8021>]}
      GreenCurve {keys [<at 0.1888 0.6042>,<at 0.4061 0.4497>,<at 0.9329 0.8021>]}
      RedCurve {keys [<at 0.1888 0.6042>,<at 0.4061 0.4497>,<at 0.9329 0.8021>]}
     }}
    Scale Mode=::NewEnumerator2
   )
   &CurlNoiseForce(
    Noise Frequency=&Divide_Float(A=1, B=$Emitter.Global Scale)
    Noise Quality / Cost=::NewEnumerator0
    Noise Strength=&Multiply_Float(
     A=&FloatFromCurve(
      FloatCurve=DataInterface() {propertyValues {Curve {keys [
          {arriveTangent=3.7261, interpMode=RCIM_Cubic, leaveTangent=3.7261, tangentMode=RCTM_User, time=0.05, value=0}
          {interpMode=RCIM_Cubic, tangentMode=RCTM_User, time=0.5, value=1}
          <at 0.8 0>
         ]}}}
      Scale Curve=1000
     )
     B=$Emitter.Global Scale
    )
    Pan Noise=true
    Pan Noise Field=&RandomRangeVector(Minimum=<0 0 0>, Maximum=<1 1 1>)
   )
   &SolveForcesAndVelocity(Clamp Velocity=false, Limit Acceleration=false)
   &ScaleMeshSize(Scale Factor=&VectorFromCurve(
     VectorCurve=DataInterface() {propertyValues {
       XCurve {keys [
         {arriveTangent=1.1946, interpMode=RCIM_Cubic, leaveTangent=1.1946, tangentMode=RCTM_User, time=0, value=0.7}
         {arriveTangent=0.581, interpMode=RCIM_Cubic, leaveTangent=0.581, tangentMode=RCTM_User, time=0.2, value=0.8875}
         {arriveTangent=0.0837, interpMode=RCIM_Cubic, leaveTangent=0.0837, tangentMode=RCTM_User, time=1, value=1.1}
        ]}
       YCurve {keys [
         {arriveTangent=1.1946, interpMode=RCIM_Cubic, leaveTangent=1.1946, tangentMode=RCTM_User, time=0, value=0.7}
         {arriveTangent=0.581, interpMode=RCIM_Cubic, leaveTangent=0.581, tangentMode=RCTM_User, time=0.2, value=0.8875}
         {arriveTangent=0.0837, interpMode=RCIM_Cubic, leaveTangent=0.0837, tangentMode=RCTM_User, time=1, value=1.1}
        ]}
       ZCurve {keys [
         {arriveTangent=0.9745, interpMode=RCIM_Cubic, leaveTangent=0.9745, tangentMode=RCTM_User, time=0, value=0.7}
         {arriveTangent=0.8594, interpMode=RCIM_Cubic, leaveTangent=0.8594, tangentMode=RCTM_User, time=0.2, value=0.8875}
         {arriveTangent=0.0837, interpMode=RCIM_Cubic, leaveTangent=0.0837, tangentMode=RCTM_User, time=1, value=1.1}
        ]}
      }}
     Scale Curve=&Multiply_VectorByFloat(
      Vector=<1.1 1.1 1.1>
      Float=&Multiply_Float(
       A=$System.DistanceScale
       B=&FloatFromCurve(
        FloatCurve=DataInterface() {propertyValues {Curve {keys [<at 0.5 0.8 RCIM_Constant>,<at 0.7 0.9 RCIM_Constant>]}}}
        CurveIndex=&RandomRangeFloat(Minimum=0.4, Maximum=1)
        Scale Curve=&MaskFloatBySpawnGroup(
         Masked Float=&FloatFromCurve(
          FloatCurve=DataInterface() {propertyValues {Curve {keys [
              {arriveTangent=0.8977, interpMode=RCIM_Cubic, leaveTangent=0.8977, tangentMode=RCTM_User, time=0, value=0.6}
              {arriveTangent=0.4659, interpMode=RCIM_Cubic, leaveTangent=0.4659, tangentMode=RCTM_User, time=0.7, value=1}
             ]}}}
          Scale Curve=0.88
         )
         Passthrough Float=1
         Spawn Group Mask=1
        )
       )
      )
     )
    ))
  ]
  rendererClasses [&NiagaraMeshRendererProperties]
  renderers [{rendererClass=&NiagaraMeshRendererProperties, rendererIndex=0}]
 }
 NE_Grid2D_Smoke_Trail_HIGH {
  simTarget=GPUComputeSim
  emitterSpawnScript [
   /SetVariables(
    Emitter.Global Fire Color:LinearColor=$User.Global Color
    Emitter.Global Scale:NiagaraFloat=$User.Global Scale
    Emitter.Global Smoke Color:LinearColor=$User.Explosion Smoke Color
   ) {isSetParameters=true}
   /SetVariables(Emitter.Render Target 2DA=DataInterface() {propertyValues {
      MipMapGeneration=Disabled
      MipMapGenerationType=Linear
      OverrideRenderTargetFilter=TF_Bilinear
      OverrideRenderTargetFormat=RTF_R8
      bInheritUserParameterSettings=false
      bOverrideFormat=true
      bPreviewRenderTarget=false
      RenderTargetUserParameter {parameter {name=None, typeDefHandle {registeredTypeIndex=4}, varData []}}
      Size=<160 160>
     }}) {isSetParameters=true}
   /SetVariables(Emitter.Render Target 2DB=DataInterface() {propertyValues {
      MipMapGeneration=Disabled
      MipMapGenerationType=Linear
      OverrideRenderTargetFilter=TF_Bilinear
      OverrideRenderTargetFormat=RTF_R8
      bInheritUserParameterSettings=false
      bOverrideFormat=true
      bPreviewRenderTarget=false
      RenderTargetUserParameter {parameter {name=None, typeDefHandle {registeredTypeIndex=4}, varData []}}
      Size=<160 160>
     }}) {isSetParameters=true}
   /SetVariables(
    Emitter.CameraOffset=25
    Emitter.Position:NiagaraPosition=&Multiply_VectorByFloat(Vector=<0 0 150>, Float=$User.Global Scale)
    Emitter.Smoke Color:LinearColor=$Emitter.Global Smoke Color
   ) {isSetParameters=true}
  ]
  emitterUpdateScript [
   &self:NE_Grid2D_Smoke_Trail_HIGH.NiagaraScratchPadContainer_2.Emitter_Limitation_and_Culling_0(Max Num=3) {module=Emitter_Limitation_and_Culling}
   &EmitterState(
    Enable Distance Culling=true
    Life Cycle Mode=::NewEnumerator1
    Loop Behavior=::NewEnumerator1
    Loop Duration=2.6
    Max Distance Response=::KillAfterParticlesFinish
    MaxDistance=12000
    Module.bUseMaxDistance=true
    Module.bUseMinDistance=false
    Scalability Mode=::NewEnumerator1
    UseLoopDelay=false
   )
   /SetVariables(
    Emitter.Fire Color:LinearColor=$Emitter.Global Fire Color
    Emitter.Outer Fire Color Hue Shift=0.97
    Emitter.Smoke Color:LinearColor=&MultiplyLinearColor(
     Linear Color=&ColorFromCurve(
      ColorCurve=DataInterface() {propertyValues {
        AlphaCurve {keys [<at 0.4268 1.25 RCIM_Cubic>,<at 1 0 RCIM_Cubic>]}
        BlueCurve {keys [<at 0.0716 0.033 RCIM_Cubic>,<at 0.4288 0.117>,<at 1 0.297>]}
        GreenCurve {keys [<at 0.0716 0.033 RCIM_Cubic>,<at 0.4288 0.117>,<at 1 0.297>]}
        RedCurve {keys [<at 0.0716 0.033 RCIM_Cubic>,<at 0.4288 0.117>,<at 1 0.297>]}
       }}
      CurveIndex=$Emitter.NormalizedLoopAge
     )
     Scale Factor=&MakeVector4_FromLinearColor(LinearColor=$Emitter.Global Smoke Color)
    )
    Emitter.SpriteSize:Vector2f=&MakeCustomVector2DFromBool(
     Bool=$Emitter.CullBool
     True Vector2D=<0.01 0.01>
     False Vector2D=&Multiply_Vector2D_ByFloat(Float=$Emitter.Global Scale, Vector2D=<1200 1200>)
    )
   ) {isSetParameters=true}
   /SetVariables(
    Emitter.DMP0:Vector4f=&MakeVector4(
     X=100
     Y=&FloatFromCurve(
      FloatCurve=DataInterface() {propertyValues {Curve {keys [
          {arriveTangent=1.1271, interpMode=RCIM_Cubic, leaveTangent=1.1271, tangentMode=RCTM_User, time=0, value=0.5}
          {arriveTangent=11.9482, interpMode=RCIM_Cubic, leaveTangent=11.9482, time=0.2288, value=3.0293}
          {arriveTangent=21.6445, interpMode=RCIM_Cubic, leaveTangent=21.6445, tangentMode=RCTM_User, time=0.7951, value=10}
         ]}}}
      CurveIndex=$Emitter.NormalizedLoopAge
      Scale Curve=1.5
     )
     Z=&FloatFromCurve(
      FloatCurve=DataInterface() {propertyValues {Curve {keys [
          <at 0.04 5>
          <at 0.1 0.5>
          {arriveTangent=0.2332, interpMode=RCIM_Cubic, leaveTangent=0.2332, tangentMode=RCTM_User, time=0.35, value=0.41}
          <at 0.5 0>
         ]}}}
      CurveIndex=$Emitter.NormalizedLoopAge
      Scale Curve=32
     )
     W=0
    )
    Emitter.DynamicMaterialParameter1:Vector4f=&MakeVector4(
     X=368
     Y=1.5
     Z=&FloatFromCurve(FloatCurve=DataInterface() {propertyValues {Curve {keys [<at 0.6 0.5 RCIM_Cubic>,<at 1 0>]}}}, CurveIndex=$Emitter.LoopedAge)
     W=&FloatFromCurve(
      FloatCurve=DataInterface() {propertyValues {Curve {keys [<at 0 0.3>, {arriveTangent=-1.327, interpMode=RCIM_Cubic, leaveTangent=-1.327, tangentMode=RCTM_User, time=0.4, value=0.3}, <at 1 0>]}}}
      CurveIndex=$Emitter.NormalizedLoopAge
      Scale Curve=0.5
     )
    )
    Emitter.DynamicMaterialParameter2:Vector4f=&MakeVector4FromVectorAndFloat(
     "Vector (RGB)"=&MakeVectorFromLinearColorRGB(Color=$Emitter.Fire Color)
     "Float (Alpha)"=$Emitter.Outer Fire Color Hue Shift
    )
    Emitter.ExecutionState:ENiagaraExecutionState=$Emitter.ENiagaraExecutionState
    Emitter.SImAge:NiagaraFloat=&Add_Float(A=$Emitter.SImAge, B=$Engine.DeltaTime)
   ) {isSetParameters=true}
   &self:NE_Grid2D_Smoke_Trail_HIGH.NiagaraScratchPadContainer_2.Release_RTs_0(RT1:NiagaraDataInterfaceRenderTarget2D=$Emitter.Render Target 2D, RT2:NiagaraDataInterfaceRenderTarget2D=$Emitter.Render Target 2DB) {module=Release_RTs}
  ]
  particleSpawnScript [&InitializeParticle(Color Mode=::NewEnumerator0, Lifetime=5, UsePositionOffset=false, Write Material Random=true)]
  particleUpdateScript [&ParticleState()]
  rendererClasses [&NiagaraSpriteRendererProperties]
  renderers [{rendererClass=&NiagaraSpriteRendererProperties}]
 }
 SourceParticles {
  simTarget=GPUComputeSim
  emitterUpdateScript [
   /NiagaraFN/Modules/Scalability/SignificantInstancesCull(
    Cull Response:ENiagaraExecutionStateManagement=::KillImmediately
    First Frame Only=false
    Num Instances=3
    Platform Set=DataInterface() {propertyValues {Platforms {qualityLevelMask=-17, cVarConditions [], deviceProfileStates []}}}
   )
   &EmitterState(
    Life Cycle Mode=::NewEnumerator1
    Loop Behavior=::NewEnumerator1
    Loop Delay=0.015
    Loop Duration=0.2
    Scalability Mode=::NewEnumerator1
    UseLoopDelay=true
   )
   &SpawnBurst_Instantaneous(
    Spawn Count=&RandomRangeInt(Minimum=6, Maximum=10, Evaluation Type=::NewEnumerator1)
    Use Spawn Probability=false
   )
   &SpawnRate(SpawnRate=55, Use Spawn Probability=false) {enabled=false}
   &Grid2D_CreateUnitToWorldTransform(
    Camera Facing=true
    Local Pivot:Vector2f=<0.5 0.1>
    Offset:NiagaraPosition=$Engine.Owner.Position
    Rotation:Quat4f=$Engine.Owner.Rotation
    WorldGridExtents:Vector2f=$Emitter.Grid2D_SetResolution.WorldGridExtents
   ) {enabled=false}
  ]
  particleSpawnScript [
   /Niagara/Modules/Spawn/Initialization/InitializeParticle(
    Color=<rgba 1111 13.6009 0 1>
    Lifetime=&RandomRangeFloat(Minimum=0.85, Maximum=1.5)
    Mass=&RandomRangeFloat(Minimum=0.75, Maximum=1)
    Module.Write Color=true
    Module.Write Lifetime=true
    Module.Write Mass=true
    Module.Write Material Random=true
    Module.Write Position=true
    Module.Write Scale=false
    Module.Write SpriteRotation=false
    Module.Write SpriteSize=true
    Sprite Size=&Vector2DFromFloat(Value=1)
   )
   &AddVelocityInCone(
    Cone Angle=80
    Cone Axis=<0 0 1>
    Cone Axis Coordinate Space=::World
    Module.Use Velocity Falloff On Cone Axis=true
    Velocity Distribution Along Cone Axis=0
    Velocity Falloff Away From Cone Axis=0
    Velocity Strength=&Multiply_Float(A=&RandomRangeFloat(Minimum=1075, Maximum=1756), B=$User.Global Scale)
   )
   &SphereLocation(
    MaskSpawnGroup=false
    Override Seed=false
    Sphere Radius=&Multiply_Float(
     A=&Lerp_Float(A=25, B=50, Alpha=&Sine(Normalized Angle=$Emitter.Age, Period=4, Scale=0.5, Bias=1))
     B=$User.Global Scale
    )
    Surface Only=false
    UseNonUniformScale=true
    UseOffset=true
    UseSphereOrientation=false
   ) {enabled=false}
   /Niagara/Modules/Update/Forces/CurlNoiseForce(
    Noise Frequency:NiagaraFloat=&Divide_Float(A=0.04, B=$User.Global Scale)
    Noise Strength=&Multiply_Float(A=250, B=$User.Global Scale)
    Pan Noise=true
    Pan Noise Field=<0.3 0.5 0.3>
   ) {enabled=false}
   &InheritVelocity(
    Module.Limit Speed=true
    Velocity Limit:NiagaraFloat=&Multiply_Float(A=1500, B=$User.Global Scale)
    Velocity Scale=<0 0 0>
   )
   &ApplyInitialForces()
   &TorusLocation(
    Align Torus With Axis=true
    Large Radius:NiagaraFloat=&Multiply_Float(A=60, B=$User.Global Scale)
    MaskSpawnGroup=false
    Torus Axis=<1 0 0>
    Torus Distribution Mode:ENiagaraTorusDistributionMode=::NewEnumerator1
    Torus Mode=::NewEnumerator2
    U Position:NiagaraFloat=&Subtract_Float(A=&ReturnNormalizedExecIndex(), B=2)
   )
   &AddVelocityFromPoint(
    Module.Use Falloff Distance=false
    Origin Offset=&Multiply_VectorByFloat(
     Vector=$Engine.Owner.SystemXAxis
     Float=&RandomRangeFloat(Minimum=0, Maximum=&Multiply_Float(A=-60, B=$User.Global Scale))
    )
    Origin Offset Coordinate Space=::Simulation
    Velocity Origin:Vector3f=&SimulationPosition()
    Velocity Strength=&Multiply_Float(A=&UniformRangedFloat(Minimum=5, Maximum=10), B=$User.Global Scale)
   ) {module=AddVelocityFromPoint001}
  ]
  particleUpdateScript [
   &ParticleState()
   &CurlNoiseForce(Noise Strength=&Multiply_Float(A=300, B=$User.Global Scale), Pan Noise=true, Pan Noise Field=<0.1 0.5 0.2>) {module=CurlNoiseForce001}
   &GravityForce(Gravity=&Multiply_VectorByFloat(Vector=<0 0 -1500>, Float=$User.Global Scale))
   &Drag(Drag=&Multiply_Float(A=1.4, B=1), Use Linear Drag=true, Use Rotational Drag=false)
   &ScaleColor(
    Scale Alpha=&FloatFromCurve(FloatCurve=DataInterface() {propertyValues {Curve {keys [<at -0.0046 1.0113 RCIM_Cubic>,<at 1.0011 0.002 RCIM_Cubic>]}}})
    ScaleA=true
    ScaleRGB=true
   )
   &Collision(
    Age Colliding Particles=false
    Collision Enabled=&SetBoolByFloatComparison(A=$Particles.Age, B=0.4)
    Enable Rest State=false
    Friction=0.2
    Friction During a Bounce=0.3
    Kill Offscreen Particles=true
    Maximum time offscreen=99
    Module.EnableMaxCollisionCount=false
    Module.Randomize Collision Normal Vector=false
    Restitution=0.2
   ) {module=Collision001}
   &SolveForcesAndVelocity(Clamp Velocity=false, Limit Acceleration=false)
   /NiagaraFluids/Modules/Particles/Fluids_Gas_Source(
    NoiseIntensity=3
    NoiseScale=3
    Radius:NiagaraFloat=&FloatFromCurve(
     FloatCurve=DataInterface() {propertyValues {Curve {keys [<at 0 0>, {interpMode=RCIM_Cubic, leaveTangent=-2.282, tangentMode=RCTM_Break, time=0.0361, value=0.984}, <at 1.0022 0.3>]}}}
     Scale Curve=&Multiply_Float(A=&RandomRangeFloat(Minimum=30, Maximum=40), B=$User.Global Scale)
    )
    Temperature:NiagaraFloat=&FloatFromCurve(
     FloatCurve=DataInterface() {propertyValues {Curve {keys [<at 0 1>,<at 0.228 0.772>,<at 1 0>]}}}
     Scale Curve=&RandomRangeFloat(Minimum=500, Maximum=3000)
    )
    VelocityScale:NiagaraFloat=&RandomRangeFloat(Minimum=0.5, Maximum=1)
   )
   /SetVariables(Particles.SpriteSize=&Vector2DFromFloat(Value=&Multiply_Float(A=$Particles.fluids_source_radius, B=&Multiply_Float(A=1.4, B=1.1)))) {enabled=false, isSetParameters=true}
  ]
  rendererClasses [&NiagaraSpriteRendererProperties]
  renderers [{rendererClass=&NiagaraSpriteRendererProperties}]
 }
 Sparkles_and_Black_Chunks {
  simTarget=GPUComputeSim
  emitterUpdateScript [
   &EmitterState(
    Enable Distance Culling=true
    Life Cycle Mode=::NewEnumerator1
    Loop Behavior=::NewEnumerator1
    Loop Delay=0.0829
    Loop Duration=1.5729
    Max Distance Response=::KillAfterParticlesFinish
    MaxDistance=4000
    Module.bUseMaxDistance=true
    Module.bUseMinDistance=false
    Scalability Mode=::NewEnumerator1
    UseLoopDelay=true
   )
   &SpawnBurst_Instantaneous(Spawn Count=300, Use Spawn Probability=false, UseLoopCountLimit=false)
  ]
  particleSpawnScript [
   /SetVariables(Particles.VisibilityTag=$Emitter.SpawnGroup) {isSetParameters=true}
   /Niagara/Modules/Spawn/Initialization/V2/InitializeParticle(
    Color=&MaskLinearColorBySpawnGroup(Masked Linear Color=$User.Global Color, Passthrough Linear Color=<rgba 0 0 0 5>)
    Lifetime=&MaskFloatBySpawnGroup(
     Masked Float=&FloatFromCurve(
      FloatCurve=DataInterface() {propertyValues {Curve {keys [<at 0 0.25 RCIM_Constant>,<at 0.3 0.45>,<at 0.7 1 RCIM_Constant>,<at 0.85 1.5>,<at 1 2>]}}}
      CurveIndex=&RandomRangeFloat(Minimum=0, Maximum=1)
      Scale Curve=0.4
     )
     Passthrough Float=0.05
    )
    Sprite Rotation Mode=::NewEnumerator2
    Sprite Size=&MaskVector2dBySpawnGroup(
     Masked Vector2D=&RandomRangeVector2D(Minimum=<5 5>, Maximum=<6 9>)
     Passthrough Vector2D=&Multiply_Vector2D_ByFloat(Float=&RandomRangeFloat(Minimum=40, Maximum=50), Vector2D=<1 1>)
     Spawn Group Mask=0
    )
    Sprite Size Mode=::NewEnumerator4
    UsePositionOffset=false
    Write Material Random=true
   )
   /Niagara/Modules/Spawn/Location/SphereLocation(
    MaskSpawnGroup=false
    Module.Override Randomness=false
    Sphere Radius=&Multiply_Float(A=0, B=$User.Global Scale)
    Surface Only=false
   )
   &AddVelocity(
    Module.Use Falloff Distance=false
    Override Seed=false
    Velocity Mode=::NewEnumerator1
    Velocity Speed=&Multiply_Float(A=&RandomRangeFloat(Minimum=5000, Maximum=8000), B=$User.Global Scale)
   )
   &ApplyInitialForces(Apply Rotational Force to Rotational Velocity=false)
   /SetVariables(Particles.SpriteSize=&Multiply_Vector2D_ByFloat(Float=$User.Global Scale, Vector2D=$StackContext.SpriteSize)) {isSetParameters=true}
   &CameraOffset(Camera Offset Amount=&Multiply_Float(A=&MaskFloatBySpawnGroup(Masked Float=10, Passthrough Float=100), B=$User.Global Scale))
  ]
  particleUpdateScript [
   &ParticleState()
   &GravityForce(Gravity=&RandomRangeVector(Minimum=<0 0 -900>, Maximum=<0 0 -500>))
   &Drag(
    Drag=&MaskFloatBySpawnGroup(Masked Float=&RandomRangeFloat(Minimum=7, Maximum=15), Passthrough Float=20)
    Use Linear Drag=true
    Use Rotational Drag=false
   )
   &Collision(
    Age Colliding Particles=false
    Kill Offscreen Particles=true
    Module.Control Rest State Via Surface Normal=false
    Module.EnableMaxCollisionCount=false
    Module.Randomize Collision Normal Vector=false
   )
   /Niagara/Modules/Update/Forces/V2/CurlNoiseForce(Noise Frequency=15, Noise Quality / Cost=::NewEnumerator0, Noise Strength=500, Pan Noise=false)
   &SolveForcesAndVelocity(Clamp Velocity=false, Limit Acceleration=false)
   &ScaleColor(
    Scale Alpha=&FloatFromCurve(FloatCurve=DataInterface() {propertyValues {Curve {keys [
         {arriveTangent=26.7273, interpMode=RCIM_Cubic, leaveTangent=26.7273, tangentMode=RCTM_User, time=0.1, value=0.0144}
         {interpMode=RCIM_Cubic, tangentMode=RCTM_User, time=0.15, value=1}
         {interpMode=RCIM_Cubic, tangentMode=RCTM_User, time=0.85, value=1}
         {arriveTangent=-1.3701, interpMode=RCIM_Cubic, leaveTangent=-1.3701, tangentMode=RCTM_User, time=1.0011, value=0}
        ]}}})
    Scale RGB=&Lerp_Vector(
     A=<0 0 0>
     B=<12.5 4 0.8>
     Alpha=&FloatFromCurve(FloatCurve=DataInterface() {propertyValues {Curve {keys [
          {arriveTangent=15.3906, interpMode=RCIM_Cubic, leaveTangent=15.3906, tangentMode=RCTM_User, time=-0.0092, value=1}
          {interpMode=RCIM_Cubic, tangentMode=RCTM_User, time=0.0967, value=0.8}
          {interpMode=RCIM_Cubic, tangentMode=RCTM_User, time=0.45, value=0.8}
          <at 0.7 0 RCIM_Cubic>
         ]}}})
    )
    ScaleA=true
    ScaleRGB=true
   )
   &ScaleSpriteSize(Scale Factor=&Vector2DFromFloat(Value=&FloatFromCurve(FloatCurve=DataInterface() {propertyValues {Curve {keys [<at -0.0046 0>,<at 0.025 1>,<at 0.66 1>,<at 1.0011 0.5>]}}}, Scale Curve=2)))
   &ScaleSpriteSizeBySpeed(
    Max Scale Factor=&MaskVector2dBySpawnGroup(Masked Vector2D=<0.5 2.2>, Passthrough Vector2D=<1 1>, Spawn Group Mask=0)
    Min Scale Factor=&MaskVector2dBySpawnGroup(Masked Vector2D=<1 0.6>, Passthrough Vector2D=<1 1>, Spawn Group Mask=0)
    Velocity Threshold=1500
   )
  ]
  rendererClasses [&NiagaraSpriteRendererProperties]
  renderers [{rendererClass=&NiagaraSpriteRendererProperties}]
 }
 Sparks_GPU {
  simTarget=GPUComputeSim
  emitterSpawnScript [/SetVariables(
    Emitter.Global Color:LinearColor=$User.Global Color
    Emitter.Global Scale:NiagaraFloat=$User.Global Scale
    Emitter.RandomFloat:NiagaraFloat=&RandomRangeFloat(Minimum=-180, Maximum=180)
    Emitter.RandomInt:NiagaraInt32=&RandomRangeInt(Minimum=0, Maximum=100)
   ) {isSetParameters=true}]
  emitterUpdateScript [
   &EmitterState(
    Enable Distance Culling=true
    Life Cycle Mode=::NewEnumerator1
    Loop Behavior=::NewEnumerator1
    Loop Delay=0.01
    Loop Duration=0.5
    Max Distance Response=::KillAfterParticlesFinish
    MaxDistance:NiagaraFloat=&FN_PickFloatDependOnPlatform(TargetPlatform=DataInterface() {propertyValues {Platforms {qualityLevelMask=-24}}}, TargetPlatformActive=9000, TargetPlatformNoActive=6500)
    Module.bApplySpawnCountScaleByDistance=true
    Module.bUseMaxDistance=true
    Module.bUseMinDistance=false
    Scalability Mode=::NewEnumerator1
    Scale Spawn Count=true
    Spawn Count Scale By Distance=DataInterface() {propertyValues {Curve {keys [<at 0.3676 0.9657>,<at 0.9962 0.4982>]}}}
    UseLoopDelay=true
   )
   &SpawnBurst_Instantaneous(Spawn Count=200, Use Spawn Probability=false, UseLoopCountLimit=true)
   &SpawnBurst_Instantaneous(Spawn Count=50, Spawn Group=1, Use Spawn Probability=false, UseLoopCountLimit=true) {module=SpawnBurst_Instantaneous001}
  ]
  particleSpawnScript [
   &PartitionParticles(Partition Mode:ENiagara_PartitionMode_v2=::NewEnumerator7)
   &InitializeParticle(
    Color=$Emitter.Global Color
    Lifetime=&FloatFromCurve(
     FloatCurve=DataInterface() {propertyValues {Curve {keys [
         {arriveTangent=-9.9637, interpMode=RCIM_Cubic, leaveTangent=-9.9637, tangentMode=RCTM_User, time=0.0772, value=6.0113}
         {arriveTangent=-5.5088, interpMode=RCIM_Cubic, leaveTangent=-5.5088, time=0.2581, value=2.8753}
         <at 0.9935 0.9632 RCIM_Cubic>
        ]}}}
     CurveIndex=&RandomRangeFloat(Minimum=0, Maximum=1)
     Scale Curve=&MaskFloat(
      Mask Source=0
      Mask Parameter=?
      Mask=0
      Masked Float=&Multiply_Float(A=&RandomRangeFloat(Minimum=0.3, Maximum=0.5), B=$User.Global Scale)
      Passthrough Float=0.22
     )
    )
    Position Offset=&Multiply_VectorByFloat(Vector=<0 0 10>, Float=$User.Global Scale)
    Sprite Size=&Multiply_Vector2D_ByFloat(
     Float=&Multiply_Float(A=750, B=$User.Global Scale)
     Vector2D=&RandomRangeVector2D(Minimum=<0.019 0.032>, Maximum=<0.029 0.072>)
    )
    Sprite Size Mode=::NewEnumerator4
    UsePositionOffset=true
    Write Material Random=true
   )
   &AddVelocity(
    Velocity=&NormalizeVector(
     Vector To Normalize=&RandomRangeVector(
      Minimum=<-1 -1 0.1>
      Maximum=<1 1 1>
      Randomness Mode=::NewEnumerator1
      Random Seed=&Add_Integer(
       A=&RandomRangeInt(Minimum=0, Maximum=6)
       B=&Multiply_Int(
        A=$Emitter.RandomInt
        B=/Niagara/DynamicInputs/Helpers/MaskInt(
         Mask Source:NiagaraInt32=0
         Mask Parameter:NiagaraInt32=?
         Mask:NiagaraInt32=0
         Masked Int:NiagaraInt32=99
         Passthrough Int:NiagaraInt32=&RandomRangeInt(Minimum=0, Maximum=200)
        )
       )
      )
      Fixed Random Seed=true
     )
     Safe Normalize=true
     Safe Fallback Threshold=0
     Safe Fallback Vector=<0 0 0.001>
    )
    Velocity Speed Scale=&FloatFromCurve(
     FloatCurve=DataInterface() {propertyValues {Curve {keys [
         {arriveTangent=-2.0184, interpMode=RCIM_Cubic, leaveTangent=-2.0184, tangentMode=RCTM_User, time=0, value=1}
         {arriveTangent=-1.2785, interpMode=RCIM_Cubic, leaveTangent=-1.2785, tangentMode=RCTM_User, time=1, value=0.3534}
        ]}}}
     CurveIndex=&RandomRangeFloat(Minimum=0, Maximum=1)
     Scale Curve=&MaskFloat(Mask Source=0, Mask Parameter=?, Mask=0, Masked Float=4000, Passthrough Float=5000)
    )
   )
   &AddVelocity(
    Velocity=&RandomVector(Vector Scale=1)
    Velocity Speed Scale=&FloatFromCurve(
     FloatCurve=DataInterface() {propertyValues {Curve {keys [
         {arriveTangent=-2.7658, interpMode=RCIM_Cubic, leaveTangent=-2.7658, tangentMode=RCTM_User, time=0, value=1}
         {arriveTangent=-0.635, interpMode=RCIM_Cubic, leaveTangent=-0.635, tangentMode=RCTM_User, time=0.1761, value=0.3439}
         <at 1 0 RCIM_Cubic>
        ]}}}
     CurveIndex=&RandomRangeFloat(Minimum=0, Maximum=1)
     Scale Curve=400
    )
   ) {enabled=false, module=AddVelocity001}
   /SetVariables(Particles.Velocity=&Multiply_VectorByFloat(Vector=$Particles.Velocity, Float=$User.Global Scale)) {isSetParameters=true}
   &CameraOffset(Camera Offset Amount=50)
  ]
  particleUpdateScript [
   &ParticleState()
   &CurlNoiseForce(
    Bias Noise=false
    Noise Frequency=1
    Noise Strength=&Multiply_Float(
     A=1000
     B=&FloatFromCurve(FloatCurve=DataInterface() {propertyValues {Curve {
         defaultValue=0
         keys [
          <at 0.0046 1 RCIM_Cubic>
          {arriveTangent=-0.1399, interpMode=RCIM_Cubic, leaveTangent=-0.1399, tangentMode=RCTM_User, time=0.0908, value=0.171}
          {arriveTangent=0.062, interpMode=RCIM_Cubic, leaveTangent=0.062, tangentMode=RCTM_User, time=0.9, value=0}
         ]
        }}})
    )
    Pan Noise=false
   )
   &Drag(
    Drag=&MaskFloat(
     Mask Source=0
     Mask Parameter=?
     Mask=0
     Masked Float=&FloatFromCurve(
      FloatCurve=DataInterface() {propertyValues {Curve {keys [<at 0.1 4>,<at 0.5007 1>]}}}
      CurveIndex=&Add_Float(A=$Particles.Age, B=&RandomRangeFloat(Minimum=-0.12, Maximum=0))
      Scale Curve=&UniformRangedFloat(Minimum=0.3, Maximum=1)
     )
     Passthrough Float=&FloatFromCurve(FloatCurve=DataInterface() {propertyValues {Curve {keys [<at 0.4 1>,<at 1 0.2>]}}}, Scale Curve=5)
    )
    Use Linear Drag=true
    Use Rotational Drag=true
   )
   &GravityForce(Gravity=&MaskVector(
     Mask Source=0
     Mask Parameter=?
     Mask=0
     Masked Vector=&Multiply_VectorByFloat(Vector=&RandomRangeVector(Minimum=<0 0 -800>, Maximum=<0 0 -650>), Float=$Emitter.Global Scale)
     Passthrough Vector=<0 0 -100>
    ))
   &PointForce(
    Force Falloff Distance=3000
    Force Strength=&MaskFloat(Mask Source=0, Mask Parameter=?, Mask=0, Masked Float=100, Passthrough Float=0)
    Linear Force Vector:Vector3f=<1 0 -25>
    Module.Use Falloff=true
    Module.UseLinearForce=true
   )
   &ScaleColor(
    Scale Alpha=&FloatFromCurve(FloatCurve=DataInterface() {propertyValues {Curve {keys [
         {arriveTangent=26.7273, interpMode=RCIM_Cubic, leaveTangent=26.7273, tangentMode=RCTM_User, time=-0.0092, value=0.0144}
         {interpMode=RCIM_Cubic, tangentMode=RCTM_User, time=0.1, value=1}
         {interpMode=RCIM_Cubic, tangentMode=RCTM_User, time=0.85, value=1}
         {arriveTangent=-1.3701, interpMode=RCIM_Cubic, leaveTangent=-1.3701, tangentMode=RCTM_User, time=1.0011, value=0}
        ]}}})
    Scale RGB=&Lerp_Vector(
     A=<0 0 0>
     B=<12.5 4 0.8>
     Alpha=&FloatFromCurve(FloatCurve=DataInterface() {propertyValues {Curve {keys [
          {arriveTangent=15.3906, interpMode=RCIM_Cubic, leaveTangent=15.3906, tangentMode=RCTM_User, time=-0.0092, value=1}
          {interpMode=RCIM_Cubic, tangentMode=RCTM_User, time=0.0967, value=0.8}
          {interpMode=RCIM_Cubic, tangentMode=RCTM_User, time=0.45, value=0.8}
          <at 0.7 0 RCIM_Cubic>
         ]}}})
    )
    ScaleA=true
    ScaleRGB=true
   )
   &CurlNoiseForce(
    Noise Frequency=&Divide_Float(A=15, B=$Emitter.Global Scale)
    Noise Quality / Cost=::NewEnumerator0
    Noise Strength=&Multiply_Float(A=500, B=$Emitter.Global Scale)
    Pan Noise=false
   ) {enabled=false, module=CurlNoiseForce001}
   &SolveForcesAndVelocity(Clamp Velocity=false, Limit Acceleration=false)
   &ScaleSpriteSize(Scale Factor=&Vector2DFromFloat(Value=&FloatFromCurve(
      FloatCurve=DataInterface() {propertyValues {Curve {keys [<at 0 1>,<at 1 0>]}}}
      Scale Curve=&RandomRangeFloat(Minimum=0.2, Maximum=0.7, Evaluation Type=::NewEnumerator1)
     )))
   &ScaleSpriteSizeBySpeed(
    Max Scale Factor=&Lerp_Vector2(
     A=<1 0.8>
     B=<0.75 4>
     Alpha=&FloatFromCurve(FloatCurve=DataInterface() {propertyValues {Curve {keys [
          {arriveTangent=-3.7663, interpMode=RCIM_Cubic, leaveTangent=-3.7663, tangentMode=RCTM_User, time=0, value=1}
          {interpMode=RCIM_Cubic, tangentMode=RCTM_User, time=0.4, value=0}
         ]}}})
    )
    Velocity Threshold:NiagaraFloat=&Multiply_Float(A=1000, B=1)
   )
  ]
  rendererClasses [&NiagaraSpriteRendererProperties]
  renderers [{rendererClass=&NiagaraSpriteRendererProperties}]
 }
 Streaks_Fat {
  emitterUpdateScript [
   &EmitterState(
    Enable Distance Culling=true
    Life Cycle Mode=::NewEnumerator1
    Loop Behavior=::NewEnumerator1
    Loop Delay=0.0353
    Loop Duration=0.9701
    Max Distance Response=::KillAfterParticlesFinish
    MaxDistance=8000
    Module.bUseMaxDistance=true
    Module.bUseMinDistance=false
    Scalability Mode=::NewEnumerator1
    UseLoopDelay=true
   )
   &SpawnBurst_Instantaneous(Spawn Count=3, Use Spawn Probability=false, UseLoopCountLimit=false)
   &SpawnBurst_Instantaneous(Spawn Count=2, Spawn Group=1, Use Spawn Probability=false, UseLoopCountLimit=false) {module=SpawnBurst_Instantaneous001}
   &SpawnBurst_Instantaneous(
    Spawn Count=4
    Spawn Group=2
    Spawn Probability=&MakeCustomFloatFromBool(Bool=$User.Use Iconic Blast, True Float=1, False Float=0)
    Use Spawn Probability=true
    UseLoopCountLimit=false
   ) {module=SpawnBurst_Instantaneous002}
  ]
  particleSpawnScript [
   /SetVariables(Particles.VisibilityTag=&MakeCustomIntFromBool(
     Bool=&SetBoolByFloatComparison(A=1.5, B=&MakeFloatFromInt(Int=$Emitter.SpawnGroup))
     True INT=0
     False INT=1
    )) {isSetParameters=true}
   &InitializeParticle(
    Color=&MultiplyLinearColor(
     Linear Color=&SelectLinearColorFromArray(
      Array Sampling Mode=::NewEnumerator1
      Direct Array Index=$Emitter.SpawnGroup
      Color Selection Array=DataInterface() {propertyValues {ColorData [<rgba 1.5 1.4526 1.4382 1>, <rgba 2 1.7761 1.5667 1>, <rgba 7 0.403 0 1>]}}
      Random Seed=?
     )
     Scale Factor=&MakeVector4_FromLinearColor(LinearColor=$User.Explosion Smoke Color)
    )
    Lifetime=&MakeCustomFloatFromBool(
     Bool=&SetBoolByFloatComparison(A=1.4, B=&MakeFloatFromInt(Int=$Emitter.SpawnGroup))
     True Float=&RandomRangeFloat(
      Minimum=1.5
      Maximum=&FloatFromCurve(FloatCurve=DataInterface() {propertyValues {Curve {keys [<at 0 0.5>,<at 1 2>]}}}, CurveIndex=$User.Global Scale)
     )
     False Float=&RandomRangeFloat(Minimum=0.25, Maximum=0.25)
    )
    Sprite Size=&MakeCustomVector2DFromBool(
     Bool=&SetBoolByFloatComparison(A=1.5, B=&MakeFloatFromInt(Int=$Emitter.SpawnGroup))
     True Vector2D=&Multiply_Vector2D_ByFloat(Float=$User.Global Scale, Vector2D=&UniformRangedVector2D(Minimum=<85 200>, Maximum=<125 300.01>))
     False Vector2D=&Multiply_Vector2D_ByFloat(Float=$User.Global Scale, Vector2D=&UniformRangedVector2D(Minimum=<100 100>, Maximum=<160 200>))
    )
    Sprite Size Mode=::NewEnumerator4
    UsePositionOffset=false
    Write Material Random=true
   )
   &SphereLocation(
    MaskSpawnGroup=false
    Module.Override Randomness=false
    Offset=&Multiply_VectorByFloat(Vector=<0 0 0.01>, Float=$User.Global Scale)
    Sphere Radius=&Multiply_Float(A=6, B=$User.Global Scale)
    Surface Only=true
   )
   &AddVelocityFromPoint(
    Module.Use Falloff Distance=false
    Origin Offset=&Multiply_VectorByFloat(
     Vector=&MakeCustomVectorFromBool(
      Bool=&SetBoolByFloatComparison(A=1.5, B=&MakeFloatFromInt(Int=$Emitter.SpawnGroup))
      True Vector=<0 0 -6.01>
      False Vector=<0 0 0>
     )
     Float=$User.Global Scale
    )
    Velocity Strength=&UniformRangedFloat(Minimum=400, Maximum=500)
   )
   &SubUVAnimation(End Frame=3, SubUV Animation Mode:ENiagara_SubUVLookupMode=::NewEnumerator2)
   &CameraOffset(Camera Offset Amount=&MakeCustomFloatFromBool(
     Bool=&SetBoolByFloatComparison(A=1.5, B=&MakeFloatFromInt(Int=$Emitter.SpawnGroup))
     True Float=&Multiply_Float(A=&MaskFloatBySpawnGroup(Masked Float=10.01, Passthrough Float=0, Spawn Group Mask=1), B=$User.Global Scale)
     False Float=&Multiply_Float(A=50, B=$User.Global Scale)
    ))
  ]
  particleUpdateScript [
   &ParticleState()
   &Drag(
    Drag=&FloatFromCurve(FloatCurve=DataInterface() {propertyValues {Curve {keys [<at 0.2 1>,<at 0.3 0>]}}}, Scale Curve=15)
    Use Linear Drag=true
    Use Rotational Drag=false
   )
   &GravityForce(Gravity=&Multiply_VectorByFloat(
     Vector=<0 0 -2.2>
     Float=&FloatFromCurve(
      FloatCurve=DataInterface() {propertyValues {Curve {keys [{arriveTangent=3.1357, interpMode=RCIM_Cubic, leaveTangent=3.1357, tangentMode=RCTM_User, time=0.1, value=0}, <at 0.6 1>]}}}
      CurveIndex=$Emitter.Age
     )
    )) {enabled=false}
   &SolveForcesAndVelocity(Clamp Velocity=false, Limit Acceleration=false)
   &ScaleSpriteSize(Scale Factor=&Vector2DFromCurve(
     Vector2Curve=DataInterface() {propertyValues {
       XCurve {keys [
         <at 0 0.625 RCIM_Cubic>
         {arriveTangent=1.125, interpMode=RCIM_Cubic, leaveTangent=1.125, time=0.4, value=0.75}
         {arriveTangent=2.6138, interpMode=RCIM_Cubic, leaveTangent=2.6138, tangentMode=RCTM_User, time=1, value=1.75}
        ]}
       YCurve {keys [
         {arriveTangent=9.4573, interpMode=RCIM_Cubic, leaveTangent=9.4573, tangentMode=RCTM_User, time=0, value=0}
         <at 0.18 1.625>
         <at 1 1.875 RCIM_Cubic>
        ]}
      }}
     ScaleCurve=<1.5 2>
    ))
   &ScaleColor(
    Linear Color Curve=DataInterface() {propertyValues {
      AlphaCurve {keys [
        {arriveTangent=-0.3577, interpMode=RCIM_Cubic, leaveTangent=-0.3577, tangentMode=RCTM_User, time=0.08, value=1}
        {arriveTangent=-1.087, interpMode=RCIM_Cubic, leaveTangent=-1.0164, tangentMode=RCTM_Break, time=0.3702, value=1}
        <at 1 0>
       ]}
      BlueCurve {keys [<at 0.0194 0.0859>,<at 0.1603 0.2>,<at 0.4462 0.3>,<at 1 0.45>]}
      GreenCurve {keys [<at 0.0194 0.4664>,<at 0.1603 0.2>,<at 0.4462 0.3>,<at 1 0.45>]}
      RedCurve {keys [<at 0.0194 1.5>,<at 0.1603 0.2>,<at 0.4462 0.3>,<at 1 0.45>]}
     }}
    Scale Mode=::NewEnumerator2
   )
   &DynamicMaterialParameters(
    Index 0 Param 1=&Multiply_Float(A=100.01, B=$User.Global Scale)
    Index 0 Param 4=&FloatFromCurve(FloatCurve=DataInterface() {propertyValues {Curve {keys [{interpMode=RCIM_Cubic, tangentMode=RCTM_User, time=0, value=0}, {interpMode=RCIM_Cubic, tangentMode=RCTM_User, time=0.4, value=1}]}}})
    Index 1 Param 1=&FloatFromCurve(
     FloatCurve=DataInterface() {propertyValues {Curve {keys [
         {arriveTangent=1.4482, interpMode=RCIM_Cubic, leaveTangent=1.4482, tangentMode=RCTM_User, time=0, value=0.4}
         {arriveTangent=0.6295, interpMode=RCIM_Cubic, leaveTangent=0.6295, tangentMode=RCTM_User, time=1, value=0.7808}
        ]}}}
     Scale Curve=0.3
    )
    Index 1 Param 2=&FloatFromCurve(
     FloatCurve=DataInterface() {propertyValues {Curve {keys [
         {arriveTangent=-236.6193, interpMode=RCIM_Cubic, leaveTangent=-236.6193, tangentMode=RCTM_User, time=0.1, value=50}
         {arriveTangent=-0.3869, interpMode=RCIM_Cubic, leaveTangent=-0.3869, tangentMode=RCTM_User, time=0.27, value=0}
        ]}}}
     CurveIndex=$Emitter.Age
     Scale Curve=&MaskFloatBySpawnGroup(Masked Float=0, Passthrough Float=0.9, Spawn Group Mask=1)
    )
    Index 1 Param 3=&FloatFromCurve(
     FloatCurve=DataInterface() {propertyValues {Curve {keys [{interpMode=RCIM_Cubic, tangentMode=RCTM_User, time=0.2, value=0.5}, {interpMode=RCIM_Cubic, tangentMode=RCTM_User, time=1, value=1}]}}}
     Scale Curve=2
    )
    Index 1 Param 4=&FloatFromCurve(
     FloatCurve=DataInterface() {propertyValues {Curve {keys [{arriveTangent=-2.2728, interpMode=RCIM_Cubic, leaveTangent=-2.2728, tangentMode=RCTM_User, time=0, value=1}, <at 1 0>]}}}
     Scale Curve=-0.15
    )
    Module.Param0WriteEnabled=true
    Module.Param0WriteEnabled1=true
    Module.Param1WriteEnabled=false
    Module.Param1WriteEnabled1=true
    Module.Param2WriteEnabled=false
    Module.Param2WriteEnabled1=true
    Module.Param3WriteEnabled=true
    Module.Param3WriteEnabled1=true
    Write Parameter Index 1=true
   )
  ]
  rendererClasses [&NiagaraSpriteRendererProperties]
  renderers [
   {rendererClass=&NiagaraSpriteRendererProperties}
   {rendererClass=&NiagaraSpriteRendererProperties, rendererIndex=1}
  ]
 }
}
summary {
 systemName=NS_GrenadeLauncher_Explosion_Asteria
 emitters [
  {emitterName=Grid2D_Gas_SmokeFire_Emitter, simTarget=GPUComputeSim, rendererClasses [&NiagaraSpriteRendererProperties]}
  {emitterName=NE_Explosion_Main, rendererClasses [&NiagaraMeshRendererProperties]}
  {emitterName=Streaks_Fat, rendererClasses [&NiagaraSpriteRendererProperties]}
  {emitterName=Flash_Merged, rendererClasses [&NiagaraSpriteRendererProperties, &NiagaraMeshRendererProperties]}
  {emitterName=Sparkles_and_Black_Chunks, simTarget=GPUComputeSim, rendererClasses [&NiagaraSpriteRendererProperties]}
  {emitterName=Sparks_GPU, simTarget=GPUComputeSim, rendererClasses [&NiagaraSpriteRendererProperties]}
  {emitterName=SourceParticles, simTarget=GPUComputeSim, rendererClasses [&NiagaraSpriteRendererProperties]}
  {emitterName=NE_Grid2D_Smoke_Trail_HIGH, simTarget=GPUComputeSim, rendererClasses [&NiagaraSpriteRendererProperties]}
  {emitterName=Mesh_Sparks001, rendererClasses []}
  {emitterName=LW_Dark_Chunks, rendererClasses []}
  {emitterName=LW_Smoke_Ground, rendererClasses []}
  {emitterName=LW_Explosion_Main, rendererClasses []}
 ]
 userVariables [
  User.Additional Fire Color {type=LinearColor, description="", defaultValue {struct=LinearColor, value=<rgba 2 1.25 1 1>}}
  User.Explosion Smoke Color {type=LinearColor, description="", defaultValue {struct=LinearColor, value=<rgba 0.3021 0.3021 0.3021 1.25>}}
  User.Fire Intensity {type=NiagaraFloat, defaultValue=1, description=""}
  User.Global Color {type=LinearColor, description="", defaultValue {struct=LinearColor, value=<rgba 4 0.5231 0 1>}}
  User.Global Scale {type=NiagaraFloat, defaultValue=1, description=""}
  User.Heat Dissipation {type=NiagaraFloat, defaultValue=0.954, description=""}
  User.Temperature Scale {type=NiagaraFloat, defaultValue=1, description=""}
  User.Use Iconic Blast {type=NiagaraBool, defaultValue=true}
 ]
}
