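# Niagara system for a grenade explosion, combining GPU fluid smoke and fire with sparks and named flash, glow, blast, chunk, ground-smoke and light emitters.
# Short-lived SourceParticles inject density, temperature and velocity into a camera-facing Grid2D gas simulation.
# A second Grid2D smoke-trail emitter drives material parameters with curves, while GPU sparks burst in five randomized partitions.
# User parameters supply color and scale; fluid source and trail emitters use instance-limit and distance culling.
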NiagaraSystem /Game/Effects/Explosion_Update/NS_Grenade_Explosion_Asteria_LW
dependencies {
 usedDataInterfaces [
  &NiagaraDataInterfacePlatformSet
  &NiagaraDataInterfaceCamera
  &NiagaraDataInterfaceCurve
  &NiagaraDataInterfaceArrayFloat3
  &NiagaraDataInterfaceColorCurve
  /Script/Niagara.NiagaraDataInterfaceGrid2DCollection
  &NiagaraDataInterfaceParticleRead
  /Script/Niagara.NiagaraDataInterfaceDataChannelRead
  /Script/Niagara.NiagaraDataInterfaceRenderTarget2D
  /Script/Niagara.NiagaraDataInterfaceEmitterProperties
  &NiagaraDataInterfaceVector2DCurve
  &NiagaraDataInterfaceVectorCurve
 ]
 usedDynamicInputs [
  &FN_LOD_Distance_Remap
  &VectorFromFloat
  /Niagara/DynamicInputs/Math/Min_Int
  &MakeIntFromFloat
  &MakeVector4_FromFloat
  &MakeCustomFloatFromBool
  &FN_PickFloatDependOnPlatform
  &Vector2DFromCurve
  &VectorFromCurve
 ]
 usedModules [
  &self:SystemSpawnScript.SystemScriptSource.SystemScriptGraph.NiagaraNodeAssignment_4.SetVariables_#1
  &self:SystemSpawnScript.SystemScriptSource.SystemScriptGraph.NiagaraNodeAssignment_1.SetVariables_#2
  &self:SystemSpawnScript.SystemScriptSource.SystemScriptGraph.NiagaraNodeAssignment_3.SetVariables_#3
  &SystemState
  &self:SystemSpawnScript.SystemScriptSource.SystemScriptGraph.NiagaraNodeAssignment_2.SetVariables_#4
 ]
}
emitters {
 Grid2D_Gas_Control_Emitter001 {
  simTarget=GPUComputeSim
  emitterSpawnScript [
   /NiagaraFluids/Functions/General/Fluids_Gas_Manager(
    Albedo=0.4
    Core=1
    Density Bouyancy=0.01
    Density Gain=1
    Density Range:Vector2f=<0 1>
    DensityMult=2
    Depth Fade=50
    Dissipation Density=2.1
    Dissipation Temperature=5
    Dissipation Velocity=1.1
    EnvironmentColor:LinearColor=$User.Explosion Smoke Color
    Fire Gain=1
    Fire Opacity Gain=1
    MaxStreakSamples=7
    Multiple Scattering Mask=0.2
    Pressure Relaxation=0.9
    Smoke Color:LinearColor=$User.Explosion Smoke Color
    SunColor:LinearColor=<rgba 2 1.773 1.4886 1>
    Temperature Bouyancy=0.025
    Temperature Color Curve=DataInterface() {propertyValues {
      bExposeCurve=true
      AlphaCurve {keys [<at 0 0>,<at 0.5951 1>]}
      BlueCurve {keys [<at 0.4895 0.0227>,<at 0.9094 0>,<at 1 0.019>]}
      GreenCurve {keys [<at 0.4895 0.0309>,<at 0.9094 0.0783>,<at 1 0.7055>]}
      RedCurve {keys [<at 0.4895 0.0417>,<at 0.9094 0.4896>,<at 1 1>]}
     }}
    Temperature Range:Vector2f=<0 1>
    TemperatureMult=0.05
    UseStreaking=true
    VelocityMult=1.1
    Vorticity Confinement=20
   )
   /SetVariables(Emitter.SimGrid=DataInterface() {propertyValues {
      ClearBeforeNonIterationStage=true
      NumAttributes=0
      NumCellsMaxAxis=3
      NumCellsX=1
      NumCellsY=1
      OverrideBufferFormat=Float
      PreviewAttribute=None
      SetGridFromMaxAxis=false
      bOverrideFormat=false
      bPreviewGrid=false
      RenderTargetUserParameter {parameter {name=None, typeDefHandle {registeredTypeIndex=1}, varData []}}
      WorldBBoxSize=<1200 1200>
     }}) {isSetParameters=true}
   /SetVariables(Emitter.LightingGridDownsample=DataInterface() {propertyValues {
      ClearBeforeNonIterationStage=true
      NumAttributes=1
      NumCellsMaxAxis=3
      NumCellsX=64
      NumCellsY=64
      OverrideBufferFormat=Float
      PreviewAttribute=None
      SetGridFromMaxAxis=false
      bOverrideFormat=false
      bPreviewGrid=false
      RenderTargetUserParameter {parameter {name=None, typeDefHandle {registeredTypeIndex=4}, varData []}}
      WorldBBoxSize=<100 100>
     }}) {isSetParameters=true}
   /SetVariables(Emitter.PressureGrid=DataInterface() {propertyValues {
      ClearBeforeNonIterationStage=false
      NumAttributes=0
      NumCellsMaxAxis=3
      NumCellsX=3
      NumCellsY=3
      OverrideBufferFormat=Float
      PreviewAttribute=None
      SetGridFromMaxAxis=false
      bOverrideFormat=false
      bPreviewGrid=false
      RenderTargetUserParameter {parameter {name=None, typeDefHandle {registeredTypeIndex=4}, varData []}}
      WorldBBoxSize=<100 100>
     }}) {isSetParameters=true}
   /SetVariables() {isSetParameters=true}
   /SetVariables() {isSetParameters=true}
   /NiagaraFluids/Modules/Grid2D/Grid2D_Gas_SetResolution(
    "Max Axis World Extents "=1280
    NumCellsX:ENiagaraPowerOfTwoResolution=::NewEnumerator8
    NumCellsY:ENiagaraPowerOfTwoResolution=::NewEnumerator8
   )
   /NiagaraFluids/Modules/Grid2D/Grid2D_SetResolution(
    Grid:NiagaraDataInterfaceGrid2DCollection=$Emitter.SimGrid
    NumCellsX:NiagaraInt32=$Emitter.NumGridCellsX
    NumCellsY:NiagaraInt32=$Emitter.NumGridCellsY
    WorldGridExtents:Vector2f=$Emitter.WorldGridExtents
   )
   &Grid2D_SetResolution(
    Grid:NiagaraDataInterfaceGrid2DCollection=$Emitter.LightingGridDownsample
    OtherGrid:NiagaraDataInterfaceGrid2DCollection=$Emitter.SimGrid
    Resolution Method:ENiagaraGrid2DResolution=::NewEnumerator3
    Resolution Mult=0.2
    WorldGridExtents:Vector2f=$Emitter.WorldGridExtents
   ) {module=Grid2D_SetResolution003}
   &Grid2D_SetResolution(
    Grid:NiagaraDataInterfaceGrid2DCollection=$Emitter.SourceGrid
    OtherGrid:NiagaraDataInterfaceGrid2DCollection=$Emitter.SimGrid
    Resolution Method:ENiagaraGrid2DResolution=::NewEnumerator3
    WorldGridExtents:Vector2f=$Emitter.WorldGridExtents
   ) {module=Grid2D_SetResolution005}
   &Grid2D_SetResolution(
    Grid:NiagaraDataInterfaceGrid2DCollection=$Emitter.PressureGrid
    OtherGrid:NiagaraDataInterfaceGrid2DCollection=$Emitter.SimGrid
    Resolution Method:ENiagaraGrid2DResolution=::NewEnumerator3
    WorldGridExtents:Vector2f=$Emitter.WorldGridExtents
   ) {module=Grid2D_SetResolution001}
   &Grid2D_SetResolution(
    Grid:NiagaraDataInterfaceGrid2DCollection=$Emitter.TemporaryGrid
    OtherGrid:NiagaraDataInterfaceGrid2DCollection=$Emitter.SimGrid
    Resolution Method:ENiagaraGrid2DResolution=::NewEnumerator3
    WorldGridExtents:Vector2f=$Emitter.WorldGridExtents
   ) {module=Grid2D_SetResolution002}
   /SetVariables(Emitter.g:Vector3f=<0 0 -980>) {isSetParameters=true}
   /SetVariables(
    Emitter.NDCParticleSource=DataInterface() {propertyValues {
      Channel=/NiagaraFluids/NDC_Grid3D_Gas
      bAutoLinkToSpawningNDC=true
      bOnlySpawnOnceOnSubticks=true
      bOverrideSpawnGroupToDataChannelIndex=false
      bReadCurrentFrame=false
      bUpdateSourceDataEveryTick=true
      AccessContext {accessContext {}}
     }}
    Emitter.ParticleSource=DataInterface() {propertyValues {EmitterBinding {bindingMode=Other, emitterName=SourceParticles}}}
   ) {isSetParameters=true}
   /SetVariables() {isSetParameters=true}
   /NiagaraFluids/Modules/Grid2D/Grid2D_UpdateSimAttributes(Grid:NiagaraDataInterfaceGrid2DCollection=$Emitter.SimGrid, WorldGridSize:Vector2f=$Emitter.Grid2D_SetResolution.WorldGridExtents) {module=UpdateSimAttributes}
   /SetVariables(Emitter.LightingRT=DataInterface() {propertyValues {
      MipMapGeneration=Disabled
      MipMapGenerationType=Linear
      OverrideRenderTargetFilter=TF_Default
      OverrideRenderTargetFormat=RTF_RGBA16f
      bInheritUserParameterSettings=false
      bOverrideFormat=true
      bPreviewRenderTarget=false
      RenderTargetUserParameter {parameter {name=None, typeDefHandle {registeredTypeIndex=4}, varData []}}
      Size=<0 0>
     }}) {isSetParameters=true}
   /SetVariables(Emitter.SimRT=DataInterface() {propertyValues {
      MipMapGeneration=Disabled
      MipMapGenerationType=Linear
      OverrideRenderTargetFilter=TF_Default
      OverrideRenderTargetFormat=RTF_RG16f
      bInheritUserParameterSettings=false
      bOverrideFormat=true
      bPreviewRenderTarget=false
      RenderTargetUserParameter {parameter {name=None, typeDefHandle {registeredTypeIndex=4}, varData []}}
      Size=<0 0>
     }}) {isSetParameters=true}
   /NiagaraFluids/Modules/Grid2D/Grid2D_InitRT(
    NumCellsX:NiagaraInt32=$Emitter.PaddedNumGridCellsX
    NumCellsY:NiagaraInt32=$Emitter.PaddedNumGridCellsY
    SetNumCells=true
    SimRT:NiagaraDataInterfaceRenderTarget2D=$Emitter.SimRT
   )
   &Grid2D_InitRT(
    NumCellsX:NiagaraInt32=$Emitter.PaddedNumGridCellsX
    NumCellsY:NiagaraInt32=$Emitter.PaddedNumGridCellsY
    SetNumCells=true
    SimRT:NiagaraDataInterfaceRenderTarget2D=$Emitter.LightingRT
   ) {module=Grid2D_InitRT001}
   &self:Grid2D_Gas_Control_Emitter001.NiagaraScratchPadContainer_0.ScratchModule_2(
    Grid:NiagaraDataInterfaceGrid2DCollection=$Emitter.SimGrid
    PressureGrid:NiagaraDataInterfaceGrid2DCollection=$Emitter.PressureGrid
    TemporaryGrid:NiagaraDataInterfaceGrid2DCollection=$Emitter.TemporaryGrid
   ) {module=ScratchModule002}
   &self:Grid2D_Gas_Control_Emitter001.NiagaraScratchPadContainer_0.ScratchModule_03_0(Reader:NiagaraDataInterfaceParticleRead=$Emitter.ParticleSource) {module=ScratchModule_03}
   /SetVariables(Emitter.GridsAllocated=true) {isSetParameters=true}
  ]
  emitterUpdateScript [
   &EmitterState()
   /SetVariables(Emitter.dt:NiagaraFloat=$Engine.DeltaTime) {isSetParameters=true}
   /NiagaraFluids/Modules/Grid2D/Grid2D_CreateUnitToWorldTransform(
    Camera:NiagaraDataInterfaceCamera=$Emitter.CameraQuery
    Camera Facing=true
    Local Pivot:Vector2f=<0.5 0.45>
    Offset:NiagaraPosition=$Engine.Owner.Position
    Rotation:Quat4f=$Engine.Owner.Rotation
    WorldGridExtents:Vector2f=$Emitter.Grid2D_SetResolution.WorldGridExtents
   )
   /SetVariables(Emitter.Position:NiagaraPosition=&ConvertVectorToPosition(
     Input Position=$Emitter.Grid2D_CreateUnitToWorldTransform.LocalToWorldTranslation
     Conversion Operation:ENiagara_LWCConvertVecToPos=::NewEnumerator0
    )) {isSetParameters=true}
   /NiagaraFluids/Modules/Grid2D/Grid2D_SetBounds(
    CameraFacing:NiagaraBool=$Emitter.Grid2D_CreateUnitToWorldTransform.CameraFacing
    EmitterProperties=DataInterface() {propertyValues {EmitterBinding {bindingMode=Self, emitterName=None}}}
    LocalPivot:Vector2f=$Emitter.Grid2D_CreateUnitToWorldTransform.LocalPivot
    Rotate for Camera Facing=true
    WorldGridExtents:Vector2f=$Emitter.Grid2D_SetResolution.WorldGridExtents
   )
   &self:Grid2D_Gas_Control_Emitter001.NiagaraScratchPadContainer_0.ScratchModule_02_2() {module=ScratchModule_02}
  ]
  rendererClasses [&NiagaraSpriteRendererProperties]
  renderers [
   {rendererClass=&NiagaraSpriteRendererProperties}
   {rendererClass=&NiagaraSpriteRendererProperties, rendererIndex=1}
  ]
 }
 LW_Blast_Stylized {rendererClasses [], renderers []}
 LW_Dark_Chunks {rendererClasses [], renderers []}
 LW_Explosion_Main {rendererClasses [], renderers []}
 LW_Flash_wide {rendererClasses [], renderers []}
 LW_GlowSphereMesh1 {rendererClasses [], renderers []}
 LW_GlowSphereMesh2 {rendererClasses [], renderers []}
 LW_Intro_Flash {rendererClasses [], renderers []}
 LW_Smoke_Ground {rendererClasses [], renderers []}
 LW_Streaks_Fat {rendererClasses [], renderers []}
 Light001 {rendererClasses [], renderers []}
 Mesh_Sparks001 {rendererClasses [], renderers []}
 NE_Grid2D_Smoke_Trail_HIGH {
  simTarget=GPUComputeSim
  emitterSpawnScript [
   /SetVariables(
    Emitter.Global Fire Color:LinearColor=$User.Global Color
    Emitter.Global Scale:NiagaraFloat=$User.Global Scale
    Emitter.Global Smoke Color:LinearColor=$User.Explosion Smoke Color
   ) {isSetParameters=true}
   /SetVariables(Emitter.Render Target 2DA=DataInterface() {propertyValues {
      MipMapGeneration=Disabled
      MipMapGenerationType=Linear
      OverrideRenderTargetFilter=TF_Bilinear
      OverrideRenderTargetFormat=RTF_R8
      bInheritUserParameterSettings=false
      bOverrideFormat=true
      bPreviewRenderTarget=false
      RenderTargetUserParameter {parameter {name=None, typeDefHandle {registeredTypeIndex=4}, varData []}}
      Size=<160 160>
     }}) {isSetParameters=true}
   /SetVariables(Emitter.Render Target 2DB=DataInterface() {propertyValues {
      MipMapGeneration=Disabled
      MipMapGenerationType=Linear
      OverrideRenderTargetFilter=TF_Bilinear
      OverrideRenderTargetFormat=RTF_R8
      bInheritUserParameterSettings=false
      bOverrideFormat=true
      bPreviewRenderTarget=false
      RenderTargetUserParameter {parameter {name=None, typeDefHandle {registeredTypeIndex=4}, varData []}}
      Size=<160 160>
     }}) {isSetParameters=true}
   /SetVariables(
    Emitter.CameraOffset=25
    Emitter.Position:NiagaraPosition=&Multiply_VectorByFloat(Vector=<0 0 150>, Float=$User.Global Scale)
    Emitter.Smoke Color:LinearColor=$Emitter.Global Smoke Color
   ) {isSetParameters=true}
  ]
  emitterUpdateScript [
   &self:NE_Grid2D_Smoke_Trail_HIGH.NiagaraScratchPadContainer_2.Emitter_Limitation_and_Culling_0(Max Num=3) {module=Emitter_Limitation_and_Culling}
   &EmitterState(
    Enable Distance Culling=true
    Life Cycle Mode=::NewEnumerator1
    Loop Behavior=::NewEnumerator1
    Loop Duration=2.6
    Max Distance Response=::KillAfterParticlesFinish
    MaxDistance=12000
    Module.bUseMaxDistance=true
    Module.bUseMinDistance=false
    Scalability Mode=::NewEnumerator1
    UseLoopDelay=false
   )
   /SetVariables(
    Emitter.Fire Color:LinearColor=$Emitter.Global Fire Color
    Emitter.Outer Fire Color Hue Shift=0.97
    Emitter.Smoke Color:LinearColor=&MultiplyLinearColor(
     Linear Color=&ColorFromCurve(
      ColorCurve=DataInterface() {propertyValues {
        AlphaCurve {keys [<at 0.4268 1.25 RCIM_Cubic>,<at 1 0 RCIM_Cubic>]}
        BlueCurve {keys [<at 0.0716 0.033 RCIM_Cubic>,<at 0.4288 0.117>,<at 1 0.297>]}
        GreenCurve {keys [<at 0.0716 0.033 RCIM_Cubic>,<at 0.4288 0.117>,<at 1 0.297>]}
        RedCurve {keys [<at 0.0716 0.033 RCIM_Cubic>,<at 0.4288 0.117>,<at 1 0.297>]}
       }}
      CurveIndex=$Emitter.NormalizedLoopAge
     )
     Scale Factor=&MakeVector4_FromLinearColor(LinearColor=$Emitter.Global Smoke Color)
    )
    Emitter.SpriteSize:Vector2f=&MakeCustomVector2DFromBool(
     Bool=$Emitter.CullBool
     True Vector2D=<0.01 0.01>
     False Vector2D=&Multiply_Vector2D_ByFloat(Float=$Emitter.Global Scale, Vector2D=<1200 1200>)
    )
   ) {isSetParameters=true}
   /SetVariables(
    Emitter.DMP0:Vector4f=&MakeVector4(
     X=100
     Y=&FloatFromCurve(
      FloatCurve=DataInterface() {propertyValues {Curve {keys [
          {arriveTangent=1.1271, interpMode=RCIM_Cubic, leaveTangent=1.1271, tangentMode=RCTM_User, time=0, value=0.5}
          {arriveTangent=11.9482, interpMode=RCIM_Cubic, leaveTangent=11.9482, time=0.2288, value=3.0293}
          {arriveTangent=21.6445, interpMode=RCIM_Cubic, leaveTangent=21.6445, tangentMode=RCTM_User, time=0.7951, value=10}
         ]}}}
      CurveIndex=$Emitter.NormalizedLoopAge
      Scale Curve=1.5
     )
     Z=&FloatFromCurve(
      FloatCurve=DataInterface() {propertyValues {Curve {keys [
          <at 0.04 5>
          <at 0.1 0.5>
          {arriveTangent=0.2332, interpMode=RCIM_Cubic, leaveTangent=0.2332, tangentMode=RCTM_User, time=0.35, value=0.41}
          <at 0.5 0>
         ]}}}
      CurveIndex=$Emitter.NormalizedLoopAge
      Scale Curve=32
     )
     W=0
    )
    Emitter.DynamicMaterialParameter1:Vector4f=&MakeVector4(
     X=368
     Y=1.5
     Z=&FloatFromCurve(FloatCurve=DataInterface() {propertyValues {Curve {keys [<at 0.6 0.5 RCIM_Cubic>,<at 1 0>]}}}, CurveIndex=$Emitter.LoopedAge)
     W=&FloatFromCurve(
      FloatCurve=DataInterface() {propertyValues {Curve {keys [<at 0 0.3>, {arriveTangent=-1.327, interpMode=RCIM_Cubic, leaveTangent=-1.327, tangentMode=RCTM_User, time=0.4, value=0.3}, <at 1 0>]}}}
      CurveIndex=$Emitter.NormalizedLoopAge
      Scale Curve=0.5
     )
    )
    Emitter.DynamicMaterialParameter2:Vector4f=&MakeVector4FromVectorAndFloat(
     "Vector (RGB)"=&MakeVectorFromLinearColorRGB(Color=$Emitter.Fire Color)
     "Float (Alpha)"=$Emitter.Outer Fire Color Hue Shift
    )
    Emitter.ExecutionState:ENiagaraExecutionState=$Emitter.ENiagaraExecutionState
    Emitter.SImAge:NiagaraFloat=&Add_Float(A=$Emitter.SImAge, B=$Engine.DeltaTime)
   ) {isSetParameters=true}
   &self:NE_Grid2D_Smoke_Trail_HIGH.NiagaraScratchPadContainer_2.Release_RTs_0(RT1:NiagaraDataInterfaceRenderTarget2D=$Emitter.Render Target 2D, RT2:NiagaraDataInterfaceRenderTarget2D=$Emitter.Render Target 2DB) {module=Release_RTs}
  ]
  particleSpawnScript [&InitializeParticle(Color Mode=::NewEnumerator0, Lifetime=5, UsePositionOffset=false, Write Material Random=true)]
  particleUpdateScript [&ParticleState()]
  rendererClasses [&NiagaraSpriteRendererProperties]
  renderers [{rendererClass=&NiagaraSpriteRendererProperties}]
 }
 NE_Sparks_GPU {
  simTarget=GPUComputeSim
  emitterSpawnScript [
   /SetVariables(
    Emitter.Global Color:LinearColor=$User.Global Color
    Emitter.Global Scale:NiagaraFloat=$User.Global Scale
    Emitter.RandomFloat:NiagaraFloat=&RandomRangeFloat(Minimum=-180, Maximum=180)
    Emitter.RandomInt:NiagaraInt32=&RandomRangeInt(Minimum=0, Maximum=4)
   ) {isSetParameters=true}
   /SetVariables(Emitter.RandVector:Vector3f=&RandomRangeVector(
     Minimum=&SelectVectorFromArray(
      Direct Array Index=$Emitter.RandomInt
      Vector Selection Array=DataInterface() {propertyValues {FloatData [<0.5 1 0.2>, <1 0.5 0.5>, <-1 1 0.2>, <1 -1 0.5>, <-1 -1 0.2>]}}
     )
     Maximum=&SelectVectorFromArray(
      Direct Array Index=$Emitter.RandomInt
      Vector Selection Array=DataInterface() {propertyValues {FloatData [<0.5 1 2>, <1 0.5 2>, <-1 1 2>, <1 -1 2>, <-1 -1 2>]}}
     )
    )) {isSetParameters=true}
  ]
  emitterUpdateScript [
   &EmitterState(Life Cycle Mode=::NewEnumerator1, Loop Behavior=::NewEnumerator1, Loop Delay=0.01, Loop Duration=0.1, UseLoopDelay=true)
   &SpawnRate(SpawnRate=10000, Use Spawn Probability=false)
  ]
  particleSpawnScript [
   &PartitionParticles(Number Of Partitions=5)
   &InitializeParticle(
    Color=$Emitter.Global Color
    Lifetime=&FloatFromCurve(
     FloatCurve=DataInterface() {propertyValues {Curve {keys [<at 0 0.25 RCIM_Constant>,<at 0.3 0.45>,<at 0.7 1 RCIM_Constant>,<at 0.85 1.5>,<at 1 2>]}}}
     CurveIndex=&RandomRangeFloat(Minimum=0, Maximum=1)
     Scale Curve=0.35
    )
    Mesh Scale Mode=::NewEnumerator1
    Mesh Uniform Scale Max=1.7
    Mesh Uniform Scale Min=0.6
    Sprite Rotation Mode=::NewEnumerator2
    Sprite Size Mode=::NewEnumerator3
    Uniform Sprite Size=&Multiply_Float(A=&RandomRangeFloat(Minimum=7, Maximum=11), B=$Emitter.Global Scale)
    UsePositionOffset=false
    Write Material Random=true
   )
   &SystemLocation(
    MaskSpawnGroup=false
    Offset=&Multiply_Vector(
     A=$Emitter.RandVector
     B=&SelectVectorFromArray(
      Direct Array Index=$Particles.PartitionParticles.Partition
      Direct Array Mode=::NewEnumerator1
      Vector Selection Array=DataInterface() {propertyValues {FloatData [<0.1 0.1 0.1>, <-0.1 -0.1 0.2>, <0 0.1 0.4>, <0.1 0 0.3>, <0.1 -0.1 0.5>]}}
     )
    )
   )
   &AddVelocity(
    Cone Angle=4
    Cone Axis=&Multiply_Vector(
     A=&SelectVectorFromArray(
      Direct Array Index=$Particles.PartitionParticles.Partition
      Vector Selection Array=DataInterface() {propertyValues {FloatData [<0 0.1 1>, <-0.5 0 1>, <1 1 0.1>, <-1 1 1>, <1 -1 0.4>]}}
     )
     B=$Emitter.RandVector
    )
    Module.UseConeFalloff=false
    Module.UseVelDistribution=false
    Override Seed=false
    Rotation Mode=::NewEnumerator2
    Velocity Mode=::NewEnumerator2
    Velocity Speed=&Multiply_Float(
     A=&Multiply_Float(
      A=&RandomRangeFloat(Minimum=1800, Maximum=2200)
      B=&ReturnNormalizedExecIndex()
     )
     B=$Emitter.Global Scale
    )
    Yaw / Pitch / Roll:Vector3f=&MakeVector(X=$Emitter.RandomFloat, Y=0, Z=0)
   ) {module=AddVelocity001}
   &JitterPosition(Jitter Amount:NiagaraFloat=&Multiply_Float(A=3, B=$Emitter.Global Scale))
  ]
  particleUpdateScript [
   &ParticleState()
   &GravityForce(Gravity=&Multiply_VectorByFloat(Vector=&RandomRangeVector(Minimum=<0 0 -900>, Maximum=<0 0 -650>), Float=$Emitter.Global Scale))
   &Drag(Drag=&RandomRangeFloat(Minimum=2, Maximum=2.5), Use Linear Drag=true, Use Rotational Drag=false)
   &CurlNoiseForce(
    Noise Frequency=&Divide_Float(A=15, B=$Emitter.Global Scale)
    Noise Quality / Cost=::NewEnumerator0
    Noise Strength=&Multiply_Float(A=500, B=$Emitter.Global Scale)
    Pan Noise=false
   )
   &ScaleColor(
    Scale Alpha=&FloatFromCurve(FloatCurve=DataInterface() {propertyValues {Curve {keys [
         {arriveTangent=26.7273, interpMode=RCIM_Cubic, leaveTangent=26.7273, tangentMode=RCTM_User, time=-0.0092, value=0.0144}
         {interpMode=RCIM_Cubic, tangentMode=RCTM_User, time=0.1, value=1}
         {interpMode=RCIM_Cubic, tangentMode=RCTM_User, time=0.85, value=1}
         {arriveTangent=-1.3701, interpMode=RCIM_Cubic, leaveTangent=-1.3701, tangentMode=RCTM_User, time=1.0011, value=0}
        ]}}})
    Scale RGB=&Lerp_Vector(
     A=<0 0 0>
     B=<12.5 4 0.8>
     Alpha=&FloatFromCurve(FloatCurve=DataInterface() {propertyValues {Curve {keys [
          {arriveTangent=15.3906, interpMode=RCIM_Cubic, leaveTangent=15.3906, tangentMode=RCTM_User, time=-0.0092, value=1}
          {interpMode=RCIM_Cubic, tangentMode=RCTM_User, time=0.0967, value=0.8}
          {interpMode=RCIM_Cubic, tangentMode=RCTM_User, time=0.45, value=0.8}
          <at 0.7 0 RCIM_Cubic>
         ]}}})
    )
    ScaleA=true
    ScaleRGB=true
   )
   &SolveForcesAndVelocity(Clamp Velocity=false, Limit Acceleration=false)
   &ScaleSpriteSize(Scale Factor=&Vector2DFromFloat(Value=&FloatFromCurve(FloatCurve=DataInterface() {propertyValues {Curve {keys [<at -0.0046 0>,<at 0.025 1>,<at 0.66 1>,<at 1.0011 0.5>]}}}, Scale Curve=1.5)))
   &ScaleSpriteSizeBySpeed(Max Scale Factor=<0.5 4>, Min Scale Factor=<1 0.6>, Velocity Threshold=1250)
  ]
  rendererClasses [&NiagaraSpriteRendererProperties]
  renderers [{rendererClass=&NiagaraSpriteRendererProperties}]
 }
 SourceParticles {
  simTarget=GPUComputeSim
  emitterUpdateScript [
   /NiagaraFN/Modules/Scalability/SignificantInstancesCull(
    Cull Response:ENiagaraExecutionStateManagement=::KillImmediately
    First Frame Only=false
    Num Instances=3
    Platform Set=DataInterface() {propertyValues {Platforms {qualityLevelMask=-17, cVarConditions [], deviceProfileStates []}}}
   )
   &EmitterState(
    Enable Distance Culling=true
    Life Cycle Mode=::NewEnumerator1
    Loop Behavior=::NewEnumerator1
    Loop Delay=0.015
    Loop Duration=0.2
    Max Distance Response=::KillAfterParticlesFinish
    MaxDistance=7500
    Module.bUseMaxDistance=true
    Module.bUseMinDistance=false
    Scalability Mode=::NewEnumerator1
    UseLoopDelay=true
   )
   &SpawnBurst_Instantaneous(
    Spawn Count=&RandomRangeInt(Minimum=5, Maximum=7, Evaluation Type=::NewEnumerator1)
    Use Spawn Probability=false
   )
   &SpawnRate(SpawnRate=55, Use Spawn Probability=false) {enabled=false}
   &Grid2D_CreateUnitToWorldTransform(
    Camera Facing=true
    Local Pivot:Vector2f=<0.5 0.1>
    Offset:NiagaraPosition=$Engine.Owner.Position
    Rotation:Quat4f=$Engine.Owner.Rotation
    WorldGridExtents:Vector2f=$Emitter.Grid2D_SetResolution.WorldGridExtents
   ) {enabled=false}
  ]
  particleSpawnScript [
   /Niagara/Modules/Spawn/Initialization/InitializeParticle(
    Color=<rgba 1111 13.6009 0 1>
    Lifetime=&RandomRangeFloat(Minimum=0.8, Maximum=1.4)
    Mass=&RandomRangeFloat(Minimum=0.75, Maximum=1)
    Module.Write Color=true
    Module.Write Lifetime=true
    Module.Write Mass=true
    Module.Write Material Random=true
    Module.Write Position=true
    Module.Write Scale=false
    Module.Write SpriteRotation=false
    Module.Write SpriteSize=true
    Sprite Size=&Vector2DFromFloat(Value=1)
   )
   &AddVelocityInCone(
    Cone Angle=60
    Cone Axis=<0 0 1>
    Cone Axis Coordinate Space=::World
    Module.Use Velocity Falloff On Cone Axis=true
    Velocity Distribution Along Cone Axis=0
    Velocity Falloff Away From Cone Axis=0
    Velocity Strength=&Multiply_Float(A=&RandomRangeFloat(Minimum=2000, Maximum=1800), B=$User.Global Scale)
   ) {enabled=false}
   &ShapeLocation(MaskSpawnGroup=false, Override Seed=false, Ring Radius=30, Shape Primitive=::NewEnumerator4, UseOwnerScale=true)
   &AddVelocity(
    Module.Use Falloff Distance=false
    Origin Offset=&Multiply_VectorByFloat(Vector=&RandomRangeVector(Minimum=<0 0 -45>, Maximum=<0 0 -60>), Float=$User.Global Scale)
    Origin Offset Coordinate Space=::World
    Override Seed=false
    Velocity Mode=::NewEnumerator1
    Velocity Speed=&Multiply_Float(A=&RandomRangeFloat(Minimum=1250, Maximum=1700), B=$User.Global Scale)
   )
   &SphereLocation(
    MaskSpawnGroup=false
    Override Seed=false
    Sphere Radius=&Multiply_Float(
     A=&Lerp_Float(A=25, B=50, Alpha=&Sine(Normalized Angle=$Emitter.Age, Period=4, Scale=0.5, Bias=1))
     B=$User.Global Scale
    )
    Surface Only=false
    UseNonUniformScale=true
    UseOffset=true
    UseSphereOrientation=false
   ) {enabled=false}
   /Niagara/Modules/Update/Forces/CurlNoiseForce(
    Noise Frequency:NiagaraFloat=&Divide_Float(A=0.04, B=$User.Global Scale)
    Noise Strength=&Multiply_Float(A=250, B=$User.Global Scale)
    Pan Noise=true
    Pan Noise Field=<0.3 0.5 0.3>
   ) {enabled=false}
   &InheritVelocity(Module.Limit Speed=true, Velocity Limit=1500, Velocity Scale=<0 0 0>) {enabled=false}
   &AddVelocityFromPoint(
    Module.Use Falloff Distance=false
    Origin Offset Coordinate Space=::World
    Velocity Origin:Vector3f=&SimulationPosition()
    Velocity Strength=&Multiply_Float(A=&UniformRangedFloat(Minimum=50, Maximum=88), B=$User.Global Scale)
   ) {enabled=false, module=AddVelocityFromPoint001}
   &ApplyInitialForces()
   &TorusLocation(
    Align Torus With Axis=true
    Large Radius:NiagaraFloat=&Multiply_Float(A=60, B=$User.Global Scale)
    MaskSpawnGroup=false
    Torus Axis=<1 0 1>
    Torus Distribution Mode:ENiagaraTorusDistributionMode=::NewEnumerator1
    Torus Mode=::NewEnumerator2
    U Position:NiagaraFloat=&Subtract_Float(A=&ReturnNormalizedExecIndex(), B=2)
   ) {enabled=false}
  ]
  particleUpdateScript [
   &ParticleState()
   &CurlNoiseForce(Noise Frequency=0.1, Noise Quality / Cost=::NewEnumerator0, Noise Strength=1000, Pan Noise=true, Pan Noise Field=<0.1 0.5 0.2>) {module=CurlNoiseForce001}
   &GravityForce(Gravity=&Multiply_VectorByFloat(Vector=<0 0 -1500>, Float=$User.Global Scale))
   &Drag(Drag=&Multiply_Float(A=1.4, B=$User.Global Scale), Use Linear Drag=true, Use Rotational Drag=false)
   &ScaleColor(
    Scale Alpha=&FloatFromCurve(FloatCurve=DataInterface() {propertyValues {Curve {keys [<at -0.0046 1.0113 RCIM_Cubic>,<at 1.0011 0.002 RCIM_Cubic>]}}})
    ScaleA=true
    ScaleRGB=true
   )
   &Collision(
    Age Colliding Particles=false
    Collision Enabled=&SetBoolByFloatComparison(A=$Particles.Age, B=0.4)
    Enable Rest State=false
    Friction=0.2
    Friction During a Bounce=0.3
    Kill Offscreen Particles=true
    Maximum time offscreen=99
    Module.EnableMaxCollisionCount=false
    Module.Randomize Collision Normal Vector=false
    Restitution=0.2
   ) {module=Collision001}
   &SolveForcesAndVelocity(Clamp Velocity=false, Limit Acceleration=false)
   /NiagaraFluids/Modules/Particles/Fluids_Gas_Source(
    Density:NiagaraFloat=&RandomRangeFloat(Minimum=0.2, Maximum=1.2)
    NoiseIntensity=3
    NoiseScale=3
    Radius:NiagaraFloat=&FloatFromCurve(
     FloatCurve=DataInterface() {propertyValues {Curve {keys [
         <at 0 0>
         {interpMode=RCIM_Cubic, leaveTangent=-2.282, tangentMode=RCTM_Break, time=0.0942, value=0.9956}
         {arriveTangent=-0.8762, interpMode=RCIM_Cubic, leaveTangent=-0.8762, time=0.8179, value=0.4775}
         <at 1.0022 0.2>
        ]}}}
     Scale Curve=&Multiply_Float(A=&RandomRangeFloat(Minimum=14, Maximum=16), B=$User.Global Scale)
    )
    Temperature:NiagaraFloat=&FloatFromCurve(
     FloatCurve=DataInterface() {propertyValues {Curve {keys [<at 0 1>,<at 0.228 0.772>,<at 0.8 0>]}}}
     Scale Curve=&RandomRangeFloat(Minimum=15, Maximum=40)
    )
    VelocityScale:NiagaraFloat=&RandomRangeFloat(Minimum=0.5, Maximum=1)
   )
   /SetVariables(Particles.SpriteSize=&Vector2DFromFloat(Value=&Multiply_Float(A=$Particles.fluids_source_radius, B=&Multiply_Float(A=1.4, B=1.1)))) {enabled=false, isSetParameters=true}
  ]
  rendererClasses [&NiagaraSpriteRendererProperties]
  renderers [{rendererClass=&NiagaraSpriteRendererProperties}]
 }
 Sparks {rendererClasses [], renderers []}
}
summary {
 systemName=NS_Grenade_Explosion_Asteria_LW
 emitters [
  {emitterName=SourceParticles, simTarget=GPUComputeSim, rendererClasses [&NiagaraSpriteRendererProperties]}
  {emitterName=NE_Sparks_GPU, simTarget=GPUComputeSim, rendererClasses [&NiagaraSpriteRendererProperties]}
  {emitterName=LW_Flash_wide, rendererClasses []}
  {emitterName=LW_Intro_Flash, rendererClasses []}
  {emitterName=LW_GlowSphereMesh1, rendererClasses []}
  {emitterName=LW_GlowSphereMesh2, rendererClasses []}
  {emitterName=LW_Streaks_Fat, rendererClasses []}
  {emitterName=LW_Blast_Stylized, rendererClasses []}
  {emitterName=LW_Dark_Chunks, rendererClasses []}
  {emitterName=LW_Smoke_Ground, rendererClasses []}
  {emitterName=Light001, rendererClasses []}
  {emitterName=Sparks, rendererClasses []}
  {emitterName=LW_Explosion_Main, rendererClasses []}
  {emitterName=Grid2D_Gas_Control_Emitter001, simTarget=GPUComputeSim, rendererClasses [&NiagaraSpriteRendererProperties]}
  {emitterName=NE_Grid2D_Smoke_Trail_HIGH, simTarget=GPUComputeSim, rendererClasses [&NiagaraSpriteRendererProperties]}
  {emitterName=Mesh_Sparks001, rendererClasses []}
 ]
 userVariables [
  User.Additional Fire Color {type=LinearColor, description="", defaultValue {struct=LinearColor, value=<rgba 2 1 1 1>}}
  User.Explosion Smoke Color {type=LinearColor, description="", defaultValue {struct=LinearColor, value=<rgba 0.3229 0.3028 0.2791 1>}}
  User.Fire Intensity {type=NiagaraFloat, defaultValue=1, description=""}
  User.Global Color {type=LinearColor, description="", defaultValue {struct=LinearColor, value=<rgba 4 0.5 0 1>}}
  User.Global Scale {type=NiagaraFloat, defaultValue=1, description=""}
  User.Heat Dissipation {type=NiagaraFloat, defaultValue=0.5, description=""}
  User.Use Iconic Blast {type=NiagaraBool, defaultValue=true}
 ]
}
